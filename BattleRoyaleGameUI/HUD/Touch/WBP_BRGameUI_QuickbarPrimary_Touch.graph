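# Touch HUD widget for the Battle Royale primary quickbar, showing six item slots and a pickup preview.
# Inventory changes create mapped "gizmo" widgets for slots 6 and up; focus changes swap them with the pickaxe slot by reparenting between overlays.
# A replicated loose gameplay tag enables pickaxe toggling while such slots apply.
# Interaction updates position the preview over slots a pickup would occupy, and event-router messages add or remove per-slot overlays.
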WidgetBlueprint /BattleRoyaleGameUI/HUD/Touch/WBP_BRGameUI_QuickbarPrimary_Touch
graphs [
 EventGraph {nodes [
   n1 = OnQuickBarSlotFocusChanged_Bind!() -> (OutputDelegate, then, QuickbarIndex:EFortQuickBars Enum=Primary, Slot:int) {is=title K2Node_CustomEvent}
   > UpdateItemPreview()
   > Switch on EFortQuickBars!(execute, Selection:EFortQuickBars Enum=@n1.QuickbarIndex) -> (Primary:Exec, Secondary:Exec, Creative:Exec, Max_None:Exec) {is=title K2Node_SwitchEnum}
   > Set CurrentWeaponSlot(CurrentWeaponSlot:int=@n1.Slot) -> (Output_Get:int=0)
   > n2 = Branch(Condition="integer >= integer"(A=@n1.Slot, B=6))
   n3 = GetSelectedGizmoSlot!(CurrentSlot:int=@n1.Slot) -> (GizmoSlot:WBP BRGame UI Quickbar Slot Primary Object Reference)
   > Set GizmoSlot(GizmoSlot:WBP BRGame UI Quickbar Slot Primary Object Reference=@n3) -> (Output_Get:WBP BRGame UI Quickbar Slot Primary Object Reference)
   > Set HasAdditionalSlot(HasAdditionalSlot:bool=IsValid(Object=@n3)) -> (Output_Get:bool)
   > n4 = Swap Slots 6 to 1()
   n5 = Get CurrentWeaponSlot() -> (CurrentWeaponSlot:int=0)
   n6 = OR Boolean(A="Equal (Integer)"(A=@n5), B="integer >= integer"(A=@n5, B=6))
   HandleGizmoSlotPreview!(execute=[@n4.then, @n7.then], self, ShowGizmoPreview:bool=@n6) -> (then)
   > HandleAdditionalSlotTag!(execute, self, IsAddTag:bool=AND Boolean(A=HasAnyValidAdditionalSlot!() -> (HasAnyAdditionalSlot:bool), B=@n6)) -> (then)
   > Set PreviousWeaponSlot(PreviousWeaponSlot:int=Get CurrentWeaponSlot() -> (CurrentWeaponSlot=0)) -> (Output_Get:int=0)
   n7 = Swap Slots 1 to 6(execute=@n2.else)
   QuickbarSwapSlottable() {is=title K2Node_CustomEvent}
   > Branch(Condition=Get HasAdditionalSlot() -> (HasAdditionalSlot:bool))
   > n8 = Branch(Condition=Get IsGizmoSlotFocused() -> (IsGizmoSlotFocused:bool))
   > Swap Slots 1 to 6()
   Swap Slots 6 to 1(execute=@n8.else)
   Swap Slots 1 to 6() {is=title K2Node_CustomEvent}
   > HandlePickaxeAdditionalSlotSwap!(execute, self, IsSwap1To6:bool=true) -> (then)
   Swap Slots 6 to 1() {is=title K2Node_CustomEvent}
   > HandlePickaxeAdditionalSlotSwap(IsSwap1To6=false)
   n9 = Event OnQuickbarContentsChanged!() -> (OutputDelegate, then, QuickbarIndex:EFortQuickBars Enum=Primary, ChangedSlots:Array of Integers) {is=title K2Node_Event}
   > Parent: OnQuickbarContentsChanged!(execute, QuickbarIndex:EFortQuickBars Enum=@n9.QuickbarIndex, ChangedSlots:Array of Integers=@n9.ChangedSlots) -> (then) {is=title K2Node_CallParentFunction}
   > UpdateItemPreview()
   Event Destruct()
   > StopListeningToEvent(self:Async Action Start Listening To Event Object Reference=Get Async Task Ref() -> (Async Task Ref:Async Action Start Listening To Event Object Reference))
   Event Construct()
   > n10 = Sequence!(execute) -> (then_0, then_1, then_2:Exec, then_3:Exec, then_4:Exec, then_5:Exec)
   Parent: Construct(execute=@n10.then_1)
   n11 = GetGlobalEventRouter()
   > n12 = StartListeningToEvent!(
    execute:Exec
    Target:Gameplay Event Router Component Object Reference=@n11
    EventType:Script Struct Object Reference=/BattleRoyaleGameUI/Data/ST_AddOverlayToQuickBarData
    Context:Object Reference
    EventBubblingRule:EEventBubblingRule Enum
    None:%=@n11
   ) -> (
    then:Exec
    AsyncTaskProxy:Async Action Start Listening To Event Object Reference
    OnEventReceived:Exec
    Payload:ST Add Overlay To Quick Bar Data Structure
   )
   > n13 = Get!(execute:Exec) -> (then:Exec, else:Exec, Async Task Ref:Async Action Start Listening To Event Object Reference)
   > StopListeningToEvent(self=@n13)
   Set Async Task Ref(execute=@n13.else, Async Task Ref:Async Action Start Listening To Event Object Reference=@n12) -> (Output_Get:Async Action Start Listening To Event Object Reference)
   n14 = Break ST Add Overlay to Quick Bar Data!(ST_AddOverlayToQuickBarData:"ST Add Overlay To Quick Bar Data Structure (by ref)"=@n12.Payload) -> (Widget_16_#1:User Widget Class Reference, QuickBarType_5_#2:EFortQuickBars Enum) {is=title K2Node_BreakStruct}
   Branch(
    execute=@n12.OnEventReceived
    Condition=AND Boolean(
     A=IsValidClass!(Class:Object Class Reference=@n14) -> (ReturnValue:bool)
     B="Equal (Enum)"(A:EFortQuickBars Enum=@n14.QuickBarType_5_#2, B:EFortQuickBars Enum=Primary)
    )
   )
   > HandleQuickBarOverlay!(execute, self, Data:"ST Add Overlay To Quick Bar Data Structure (by ref)"=@n12.Payload) -> (then)
   n15 = Set QuickbarSlots(
    execute=@n10.then_0
    QuickbarSlots:Array of Fort Quick Bar Slot Base Object References=Make Array!(
     "[0]":Fort Quick Bar Slot Base Object Reference=Get QuickbarSlot1() -> (QuickbarSlot1:WBP BRGame UI Quickbar Slot Primary Object Reference)
     "[1]":Fort Quick Bar Slot Base Object Reference=Get QuickbarSlot2() -> (QuickbarSlot2:WBP BRGame UI Quickbar Slot Primary Object Reference)
     "[2]":Fort Quick Bar Slot Base Object Reference=Get QuickbarSlot3() -> (QuickbarSlot3:WBP BRGame UI Quickbar Slot Primary Object Reference)
     "[3]":Fort Quick Bar Slot Base Object Reference=Get QuickbarSlot4() -> (QuickbarSlot4:WBP BRGame UI Quickbar Slot Primary Object Reference)
     "[4]":Fort Quick Bar Slot Base Object Reference=Get QuickbarSlot5() -> (QuickbarSlot5:WBP BRGame UI Quickbar Slot Primary Object Reference)
     "[5]":Fort Quick Bar Slot Base Object Reference=Get QuickbarSlot6() -> (QuickbarSlot6:WBP BRGame UI Quickbar Slot Primary Object Reference)
    ) -> (Array:Array of Fort Quick Bar Slot Base Object References)
   ) -> (Output_Get:Array of Fort Quick Bar Slot Base Object References)
   > n16 = For Each Loop(Array:Array of Fort Quick Bar Slot Base Object References=@n15) -> (Array Element:Fort Quick Bar Slot Base Object Reference)
   > n17 = Cast To WBP_BRGameUI_QuickbarSlot_Primary!(execute, Object=@n16) -> (then, CastFailed, AsWBP BRGame UI Quickbar Slot Primary:WBP BRGame UI Quickbar Slot Primary Object Reference) {is=title K2Node_DynamicCast}
   > InitializeDataModel!(
    execute
    self:WBP BRGame UI Quickbar Slot Primary Object Reference=@n17
    SlotNum:int=Get SlotIndex!(self:Fort Quick Bar Slot Base Object Reference=@n17) -> (SlotIndex:int=0)
    Quick Bar Type:EFortQuickBars Enum=Primary
   ) -> (then)
   n18 = Create Event()
   n19 = Get FortHUDContext!() -> (ReturnValue:Fort HUDContext Object Reference) {is=title K2Node_GetBlueprintContextEx}
   Bind Event to On Player Can Interact(execute=@n10.then_4, self:Fort HUDContext Object Reference=@n19, Delegate=@n18) {is=title K2Node_AddDelegate}
   > Bind Event to On Interact Updated(self:Fort HUDContext Object Reference=@n19, Delegate=@n18) {is=title K2Node_AddDelegate}
   Clear!(
    execute=@n10.then_5
    TargetMap:Map of Integers to WBP BRGame UI Quickbar Slot Primary Object References=Get GizmoSlotMap() -> (GizmoSlotMap:Map of Integers to WBP BRGame UI Quickbar Slot Primary Object References)
   ) -> (then) {is=title K2Node_CallFunction}
   > n20 = Set PlayerController(PlayerController:Player Controller Object Reference=GetOwningPlayer()) -> (Output_Get:Player Controller Object Reference)
   > n21 = Cast To FortPlayerController(Object=@n20)
   > Set FortPlayerController(FortPlayerController:Fort Player Controller Object Reference=@n21) -> (Output_Get:Fort Player Controller Object Reference)
   n22 = GetQuickbarFocus!(self:Fort Inventory Context Object Reference=Get FortInventoryContext!() -> (ReturnValue:Fort Inventory Context Object Reference) {is=title K2Node_GetBlueprintContextEx}) -> (OutQuickBar:EFortQuickBars Enum=Primary, OutSlot:int=0, OutSecondarySlot:int=0, OutPreviousFocusedSlot:int=0)
   OnQuickbarSlotFocusChanged!(execute=@n10.then_2, self:Fort Quickbar Base Object Reference, QuickbarIndex:EFortQuickBars Enum=@n22, SlotIndex:int=@n22.OutSlot) -> (then)
   n23 = Get FortInventoryContext()
   Assign On Quickbar Slot Focus Changed!(execute:Exec=@n10.then_3, self:Fort Inventory Context Object Reference=@n23, Delegate:"Delegate (by ref)"=Create Event()) -> (then:Exec) {is=title K2Node_AssignDelegate}
   > Bind Event to On Quickbar Contents Changed(self:Fort Inventory Context Object Reference=@n23, Delegate=Create Event()) {is=title K2Node_AddDelegate}
  ]}
 Get Slot Display Index {nodes [
   n1 = Get Slot Display Index!() -> (then, SlotIndex:int) {is=title K2Node_FunctionEntry}
   n2 = SlotAsGridSlot(Widget=GetSpecificSlot!(self:Fort Quickbar Base Object Reference, SlotNumber:int=@n1) -> (ReturnValue:Fort Quick Bar Slot Base Object Reference))
   > n3 = Is Valid(InputObject=@n2)
   Return Node!(execute=@n3.Is Not Valid, ColumnIndex:int=0)
   > Return Node(ColumnIndex=Get Column!(self:Grid Slot Object Reference=@n2) -> (Column:int=0))
  ]}
 GetFirstGizmoWithQuantity {nodes [
   GetFirstGizmoWithQuantity() {is=title K2Node_FunctionEntry}
   > n1 = For Loop with Break!(Execute:Exec, FirstIndex:int=6, LastIndex:int=int - int(A=Get TotalSlots() -> (TotalSlots:int=0)), Break:Exec=@n5.then) -> (LoopBody:Exec, Index:int, Completed:Exec) {is=title K2Node_MacroInstance}
   n2 = Get GizmoSlotMap()
   > Branch(Condition="integer > integer"(A=Length!(TargetMap:Map of Integers to WBP BRGame UI Quickbar Slot Primary Object References=@n2) -> (ReturnValue:int=0) {is=title K2Node_CallFunction}))
   n3 = Find(TargetMap:Map of Integers to WBP BRGame UI Quickbar Slot Primary Object References=Get GizmoSlotMap(), Key:"Integer (by ref)"=@n1) -> (Value:WBP BRGame UI Quickbar Slot Primary Object Reference)
   > Is Valid(InputObject=@n3)
   n4 = Get InstituteChirpQuickbarSlotViewModel!(self:WBP BRGame UI Quickbar Slot Primary Object Reference=@n3) -> (InstituteChirpQuickbarSlotViewModel:Institute Chirp - HUD Controller Quickbar Slot Viewmodel Object Reference)
   > Is Valid(InputObject=@n4)
   > Branch(Condition="integer > integer"(A=Stack Count!(self:Device - HUD Controller Quickbar Slot Viewmodel Object Reference=@n4) -> (ReturnValue:int=0)))
   > Set GizmoSlot(GizmoSlot=@n3)
   > n5 = Set CurrentPreviewGizmoSlot(CurrentPreviewGizmoSlot:WBP BRGame UI Quickbar Slot Primary Object Reference=@n3) -> (Output_Get:WBP BRGame UI Quickbar Slot Primary Object Reference)
   Return Node!(
    execute=@n1.Completed
    GizmoSlot:WBP BRGame UI Quickbar Slot Primary Object Reference=Get GizmoSlot() -> (GizmoSlot:WBP BRGame UI Quickbar Slot Primary Object Reference)
   )
  ]}
 GetFirstWidgetOfType {nodes [
   n1 = GetFirstWidgetOfType!() -> (then, Type:User Widget Class Reference, Panel:Panel Widget Object Reference) {is=title K2Node_FunctionEntry}
   > n2 = For Each Loop(Array:Array of Widget Object References=GetAllChildren(self=@n1.Panel)) -> (Array Element:Widget Object Reference)
   Return Node!(execute=@n2.Completed, Content:Widget Object Reference)
   > Branch(Condition=IsA!(Object:Object Reference=@n2, ObjectClass:Object Class Reference=@n1) -> (ReturnValue:bool))
   > Return Node(Content=@n2)
  ]}
 GetSelectedGizmoSlot {nodes [
   n1 = GetSelectedGizmoSlot!() -> (then, CurrentSlot:int) {is=title K2Node_FunctionEntry}
   n2 = Get GizmoSlotMap()
   > n3 = Branch(Condition="integer > integer"(A=Length(TargetMap=@n2) -> (ReturnValue=0)))
   > Return Node!(execute, GizmoSlot:WBP BRGame UI Quickbar Slot Primary Object Reference=Find(TargetMap=@n2, Key=@n1))
   Return Node(execute=@n3.else)
  ]}
 GetSpecificSlot {nodes [
   n1 = GetSpecificSlot!() -> (then, SlotNumber:int=0) {is=title K2Node_FunctionEntry}
   n2 = Make Array(
    "[0]":WBP BRGame UI Quickbar Slot Primary Object Reference=Get QuickbarSlot1()
    "[1]":WBP BRGame UI Quickbar Slot Primary Object Reference=Get QuickbarSlot2()
    "[2]":WBP BRGame UI Quickbar Slot Primary Object Reference=Get QuickbarSlot3()
    "[3]":WBP BRGame UI Quickbar Slot Primary Object Reference=Get QuickbarSlot4()
    "[4]":WBP BRGame UI Quickbar Slot Primary Object Reference=Get QuickbarSlot5()
    "[5]":WBP BRGame UI Quickbar Slot Primary Object Reference=Get QuickbarSlot6()
   ) -> (Array:Array of WBP BRGame UI Quickbar Slot Primary Object References)
   > Branch(Condition=Is Valid Index!(TargetArray:Array of WBP BRGame UI Quickbar Slot Primary Object References=@n2, IndexToTest:int=@n1) -> (ReturnValue:bool))
   > Return Node!(
    execute
    ReturnValue:Fort Quick Bar Slot Base Object Reference="Get (a copy)"(Array:Array of WBP BRGame UI Quickbar Slot Primary Object References=@n2, Dimension 1=@n1) -> (Output:WBP BRGame UI Quickbar Slot Primary Object Reference)
   )
  ]}
 Handle Interaction Updated {nodes [
   n1 = Handle Interaction Updated!() -> (then, ContextInfo:Fort Interact Context Info Object Reference) {is=title K2Node_FunctionEntry}
   > n2 = Is Valid(InputObject=@n1)
   > n3 = Cast To FortPickup!(execute, Object=Get ReceivingActor!(self:Fort Interact Context Info Object Reference=@n1) -> (ReceivingActor:Actor Weak object)) -> (then, CastFailed, AsFort Pickup:Fort Pickup Object Reference) {is=title K2Node_DynamicCast}
   > n4 = Set Current Pickup Preview Item(Current Pickup Preview Item:Fort Item Object Reference=Get PrimaryPickupDummyItem!(self:Fort Pickup Object Reference=@n3) -> (PrimaryPickupDummyItem:Fort Item Object Reference)) -> (Output_Get:Fort Item Object Reference)
   UpdateItemPreview(execute=[@n4.then, @n5.then])
   n5 = Set Current Pickup Preview Item(execute=[@n3.CastFailed, @n2.Is Not Valid])
  ]}
 HandleAdditionalSlotPadding {nodes [
   n1 = HandleAdditionalSlotPadding!() -> (then, HasAdditionalSlot:bool=~) {is=title K2Node_FunctionEntry}
   n2 = Select(
    Option 0:Margin Structure=Get DefaultSlotPadding() -> (DefaultSlotPadding:Margin Structure)
    Option 1:Margin Structure=Get AdditionalSlotPadding() -> (AdditionalSlotPadding:Margin Structure)
    Index=@n1
   ) -> (ReturnValue:Margin Structure)
   > SetPadding!(
    execute
    self:Overlay Slot Object Reference=SlotAsOverlaySlot(Widget=Get CommonVisibilitySwitcher_30() -> (CommonVisibilitySwitcher_30:Common Visibility Switcher Object Reference))
    InPadding:Margin Structure=@n2
   ) -> (then)
   > SetPadding(self=SlotAsOverlaySlot(Widget=Get PickupPreview() -> (PickupPreview:Size Box Object Reference)), InPadding=@n2)
  ]}
 HandleAdditionalSlotTag {nodes [
   n1 = HandleAdditionalSlotTag!() -> (then, IsAddTag:bool=~) {is=title K2Node_FunctionEntry}
   > n2 = Delta Gate Wildcard!(execute:Exec, Variable:"Boolean (by ref)"=Get IsAdditionalSlotTagAdded() -> (IsAdditionalSlotTagAdded:bool), New Value:bool=@n1) -> (Success:Exec, Updated Variable:"Boolean (by ref)", Previous Variable:bool=~) {is=title K2Node_MacroInstance}
   > n3 = Branch(Condition=@n2)
   > n4 = Branch(Condition=HasAdditionalSlotTag!() -> (HasTag:bool))
   AddLooseGameplayTags!(
    execute=@n4.else
    Actor:Actor Object Reference=Get PlayerController() -> (PlayerController:Player Controller Object Reference)
    GameplayTags:"Gameplay Tag Container Structure (by ref)"=Get AllowTogglePickAxeTag() -> (AllowTogglePickAxeTag:Gameplay Tag Container Structure)
    bShouldReplicate:bool=true
   ) -> (then, ReturnValue:bool)
   RemoveLooseGameplayTags!(
    execute=@n3.else
    Actor:Actor Object Reference=Get PlayerController()
    GameplayTags:"Gameplay Tag Container Structure (by ref)"=Get AllowTogglePickAxeTag()
    bShouldReplicate:bool=true
   ) -> (then, ReturnValue:bool)
  ]}
 HandleGizmoSlotPreview {nodes [
   n1 = HandleGizmoSlotPreview!() -> (then, ShowGizmoPreview:bool=~) {is=title K2Node_FunctionEntry}
   > n2 = Delta Gate Wildcard(Variable=Get IsGizmoSlotPreviewed() -> (IsGizmoSlotPreviewed:bool), New Value=@n1) -> (Previous Variable=~)
   > Is Valid(InputObject=Get GizmoSlot())
   n3 = Get GizmoSlot()
   > n4 = Cast To WBP_BRGameUI_QuickbarSlot_Slottable!(execute, Object=@n3) -> (then, CastFailed, AsWBP BRGame UI Quickbar Slot Slottable:WBP BRGame UI Quickbar Slot Slottable Object Reference) {is=title K2Node_DynamicCast}
   > Set SlottableWidget(SlottableWidget:WBP BRGame UI Quickbar Slot Slottable Object Reference=@n4) -> (Output_Get:WBP BRGame UI Quickbar Slot Slottable Object Reference)
   > n5 = Branch(Condition=@n2)
   n6 = Get SlottableWidget() -> (SlottableWidget:WBP BRGame UI Quickbar Slot Slottable Object Reference)
   > StopAnimation(
    self=@n6
    InAnimation=Get Anim_Show!(self:WBP BRGame UI Quickbar Slot Slottable Object Reference=@n6) -> (Anim_Show:Widget Animation Object Reference)
   )
   n7 = Get SizeBox_TileContents!(self:WBP BRGame UI Quickbar Slot Primary Object Reference=@n3) -> (SizeBox_TileContents:Size Box Object Reference)
   > SetRenderOpacity(self=@n7, InOpacity=1)
   > SetSwappableVisuals!(
    execute
    self
    QuickbarSlot:WBP BRGame UI Quickbar Slot Primary Object Reference=Get SlottableWidget()
    IsFocused:bool="integer >= integer"(A=Get CurrentWeaponSlot() -> (CurrentWeaponSlot=0), B=6)
   ) -> (then)
   n8 = "integer >= integer"(A=Get PreviousWeaponSlot() -> (PreviousWeaponSlot:int=0), B=6)
   n9 = Get CurrentWeaponSlot() -> (CurrentWeaponSlot=0)
   > Branch(Condition=AND Boolean(A="Equal (Integer)"(A=@n9), B=NOT Boolean(A=@n8)))
   n10 = Get SlottableWidget()
   > PlayAnimationForward(self=@n10, InAnimation=Get Anim_Show(self=@n10))
   n11 = Get CurrentWeaponSlot() -> (CurrentWeaponSlot=0)
   Branch(execute=@n5.else, Condition=AND Boolean(A="integer > integer"(A=@n11), B="integer < integer"(A=@n11, B=6)))
   n12 = Get SlottableWidget()
   n13 = Get Anim_Show(self=@n12)
   > StopAnimation(self=Get SlottableWidget(), InAnimation=@n13)
   > SetRenderOpacity(self=@n7, InOpacity=0)
   > PlayAnimationReverse(self=@n12, InAnimation=@n13)
  ]}
 HandlePickaxeAdditionalSlotSwap {nodes [
   n1 = HandlePickaxeAdditionalSlotSwap!() -> (then, IsSwap1To6:bool=~) {is=title K2Node_FunctionEntry}
   > n2 = Branch(Condition=@n1)
   > HideGizmoSlots()
   > AddChild(self=Get QuickbarSlot1_Overlay() -> (QuickbarSlot1_Overlay:Overlay Object Reference), Content=Get QuickbarSlot1())
   > SetSwappableVisuals(QuickbarSlot=Get QuickbarSlot1(), IsFocused=true)
   > Set IsGizmoSlotFocused(IsGizmoSlotFocused:bool=false) -> (Output_Get:bool)
   n3 = Get GizmoSlot()
   > Is Valid(InputObject=@n3)
   > AddChild(self=Get SwappableSlot_Overlay() -> (SwappableSlot_Overlay:Overlay Object Reference), Content=@n3)
   n4 = Get GizmoSlot()
   > SetSlotSelected!(execute, self:Fort Quick Bar Slot Base Object Reference=@n4, bSelected:bool=false) -> (then)
   > SetSwappableVisuals(QuickbarSlot=Get GizmoSlot(), IsFocused=false)
   n5 = Get CurrentWeaponSlot() -> (CurrentWeaponSlot=0)
   n6 = "integer >= integer"(A=@n5, B=6)
   n7 = "Equal (Integer)"(A=@n5)
   > SetRenderOpacity(self=Get SizeBox_TileContents(self=@n4), InOpacity="To Float (Boolean)"(InBool=OR Boolean(A=@n7, B=@n6)))
   HideGizmoSlots(execute=@n2.else)
   > AddChild(self=Get SwappableSlot_Overlay(), Content=Get QuickbarSlot1())
   > SetSwappableVisuals(QuickbarSlot=Get QuickbarSlot1(), IsFocused=false)
   > Set IsGizmoSlotFocused(IsGizmoSlotFocused=true)
   > Is Valid(InputObject=Get GizmoSlot())
   > AddChild(self=Get QuickbarSlot1_Overlay(), Content=Get GizmoSlot())
   n8 = Get GizmoSlot()
   > SetSlotSelected(self=@n8, bSelected=false)
   > SetSlotSelected(self=Get GizmoSlot(), bSelected=true)
   > SetSwappableVisuals(QuickbarSlot=Get GizmoSlot(), IsFocused=true)
   n9 = Get CurrentWeaponSlot() -> (CurrentWeaponSlot=0)
   n10 = "integer >= integer"(A=@n9, B=6)
   n11 = "Equal (Integer)"(A=@n9)
   > SetRenderOpacity(self=Get SizeBox_TileContents(self=@n8), InOpacity="To Float (Boolean)"(InBool=OR Boolean(A=@n11, B=@n10)))
  ]}
 HandleQuickBarOverlay {nodes [
   n1 = HandleQuickBarOverlay!() -> (then, Data:"ST Add Overlay To Quick Bar Data Structure (by ref)") {is=title K2Node_FunctionEntry}
   n2 = Break ST Add Overlay to Quick Bar Data!(ST_AddOverlayToQuickBarData=@n1) -> (
    AddOrRemove_12_#3:bool=~
    Widget_16_#1
    Slot_8_#4:int
   )
   n3 = Get Slot1Overlay() -> (Slot1Overlay:Overlay Object Reference)
   n4 = Get Slot2Overlay() -> (Slot2Overlay:Overlay Object Reference)
   n5 = Get Slot3Overlay() -> (Slot3Overlay:Overlay Object Reference)
   n6 = Get Slot4Overlay() -> (Slot4Overlay:Overlay Object Reference)
   n7 = Get Slot5Overlay() -> (Slot5Overlay:Overlay Object Reference)
   n8 = Get Slot6Overlay() -> (Slot6Overlay:Overlay Object Reference)
   > Set CurrentSlot(CurrentSlot:Overlay Object Reference="Get (a copy)"(
     Array:Array of Overlay Object References=Make Array(
      "[0]":Overlay Object Reference=@n3
      "[1]":Overlay Object Reference=@n4
      "[2]":Overlay Object Reference=@n5
      "[3]":Overlay Object Reference=@n6
      "[4]":Overlay Object Reference=@n7
      "[5]":Overlay Object Reference=@n8
     ) -> (Array:Array of Overlay Object References)
     Dimension 1=int - int(A=@n2.Slot_8_#4)
    ) -> (Output:Overlay Object Reference)) -> (Output_Get:Overlay Object Reference)
   > n9 = Branch(Condition=@n2)
   > n10 = Create Widget!(
    execute:Exec
    Class:User Widget Class Reference=@n2.Widget_16_#1
    OwningPlayer:Player Controller Object Reference=GetOwningPlayer()
   ) -> (then:Exec, ReturnValue:User Widget Object Reference)
   n11 = Get CurrentSlot() -> (CurrentSlot:Overlay Object Reference)
   > n12 = AddChildToOverlay!(execute, self:Overlay Object Reference=@n11, Content:Widget Object Reference=@n10) -> (then, ReturnValue:Overlay Slot Object Reference)
   > SetVerticalAlignment(self=@n12, InVerticalAlignment=VAlign_Fill)
   > SetHorizontalAlignment(self=@n12, InHorizontalAlignment=HAlign_Fill)
   n13 = GetFirstWidgetOfType!(self, Type:User Widget Class Reference=@n2.Widget_16_#1, Panel:Panel Widget Object Reference=Get CurrentSlot()) -> (Content:Widget Object Reference) {is=title K2Node_CallFunction}
   Is Valid(exec=@n9.else, InputObject=@n13)
   > RemoveChild(self=@n11, Content=@n13)
  ]}
 HasAdditionalSlotTag {nodes [
   HasAdditionalSlotTag() {is=title K2Node_FunctionEntry}
   n1 = Get PlayerController()
   > Return Node!(
    execute
    HasTag:bool=HasMatchingGameplayTag!(
     self:Gameplay Tag Asset Interface Interface=To GameplayTagAssetInterface!(InObject:Object Reference=@n1) -> (ReturnValue:Gameplay Tag Asset Interface Interface)
     TagToCheck:Gameplay Tag Structure=(TagName="Gameplay.Status.TogglePickAxeEnabled")
    ) -> (ReturnValue:bool)
   )
  ]}
 HasAnyValidAdditionalSlot {nodes [
   HasAnyValidAdditionalSlot() {is=title K2Node_FunctionEntry}
   > n1 = "For Each Loop (Map)"!(execute:Exec, MapPin:Map of Integers to WBP BRGame UI Quickbar Slot Primary Object References=Get GizmoSlotMap(), BreakPin:Exec=@n3.then) -> (then:Exec, KeyPin:int, ValuePin:WBP BRGame UI Quickbar Slot Primary Object Reference, CompletedPin:Exec)
   n2 = IsValid(Object=@n1.ValuePin)
   > n3 = Branch(Condition=@n2)
   Return Node!(execute=@n1.CompletedPin, HasAnyAdditionalSlot:bool=@n2)
  ]}
 HideGizmoSlots {nodes [
   HideGizmoSlots() {is=title K2Node_FunctionEntry}
   > n1 = "For Each Loop (Map)"(MapPin=Get GizmoSlotMap())
   n2 = GetFocusedSlot!(
    self:Fort Player Controller Object Reference=Get FortPlayerController() -> (FortPlayerController:Fort Player Controller Object Reference)
    InQuickbar:EFortQuickBars Enum=Primary
   ) -> (ReturnValue:int=0)
   > n3 = Branch(Condition="Equal (Object)"(A=@n1.ValuePin, B=GetSelectedGizmoSlot(CurrentSlot=@n2) {is=title K2Node_CallFunction}))
   SetRenderOpacity(execute=@n3.else, self=Get SizeBox_TileContents(self=@n1.ValuePin), InOpacity=0)
  ]}
 OnQuickbarContentsChanged_Bind {nodes [
   n1 = OnQuickbarContentsChanged_Bind!() -> (then, QuickbarIndex:EFortQuickBars Enum=Primary, ChangedSlots:Array of Integers) {is=title K2Node_FunctionEntry}
   > n2 = GetNumQuickbarSlots!(execute, self:Fort Player Controller Object Reference=Get FortPlayerController(), QuickBarType:EFortQuickBars Enum=Primary) -> (then, ReturnValue:int=0)
   > Set TotalSlots(TotalSlots:int=@n2) -> (Output_Get:int=0)
   > Switch on EFortQuickBars(Selection=@n1)
   > n3 = For Each Loop(Array:Array of Integers=Get ChangedSlots() -> (ChangedSlots:Array of Integers)) -> (Array Element:int)
   HideGizmoSlots(execute=@n3.Completed) {is=title K2Node_CallFunction}
   n4 = Set CurrentSlot(execute=@n3.LoopBody, CurrentSlot:int=@n3) -> (Output_Get:int=0)
   > Branch(Condition="integer >= integer"(A=@n4, B=6))
   n5 = Get FortInventoryContext()
   n6 = Get QuickbarIndex() -> (QuickbarIndex:EFortQuickBars Enum=Primary)
   n7 = Get CurrentSlot() -> (CurrentSlot:int=0)
   > Set HasItemInCurrentSlot(HasItemInCurrentSlot:bool=IsValid(Object=GetQuickBarSlottedItem!(self:Fort Inventory Context Object Reference=@n5, InQuickBar:EFortQuickBars Enum=@n6, Slot:int=@n7) -> (ReturnValue:Fort Item Object Reference))) -> (Output_Get:bool)
   n8 = IsValid(Object=GetSelectedGizmoSlot(CurrentSlot=Get CurrentSlot() -> (CurrentSlot=0)))
   > n9 = Branch(Condition=AND Boolean(A=Get HasItemInCurrentSlot() -> (HasItemInCurrentSlot:bool), B=NOT Boolean(A=@n8)))
   > HandleAdditionalSlotPadding!(execute, self, HasAdditionalSlot:bool=true) -> (then) {is=title K2Node_CallFunction}
   > n10 = Is Valid(InputObject=Find(TargetMap=Get GizmoSlotMap(), Key=Get CurrentSlot() -> (CurrentSlot=0)))
   n11 = Branch(execute=[@n17.then, @n10.Is Valid], Condition=IsValid(Object=Get GizmoSlot()))
   n12 = Find(TargetMap=Get GizmoSlotMap(), Key=Get CurrentSlot() -> (CurrentSlot=0))
   Set bKeybindsHidden!(execute=[@n15.then, @n11.then], bKeybindsHidden:bool=true, self:Fort Quick Bar Slot Base Object Reference=@n12) -> (then, Output_Get:bool)
   > AddChild(self=Get SwappableSlot_Overlay(), Content=@n12)
   > SetRenderOpacity(self=Get SizeBox_TileContents(self=@n12), InOpacity=0)
   > InitializeDataModel(self=@n12, SlotNum=Get CurrentSlot() -> (CurrentSlot=0), Quick Bar Type=Get QuickbarIndex() -> (QuickbarIndex=Primary))
   > SetQuickbarIndexRuntime!(execute, self:Fort Quick Bar Slot Base Object Reference=@n12, Index:int=Get CurrentSlot() -> (CurrentSlot=0), bRefresh:bool=true) -> (then)
   > Set HasAdditionalSlot(HasAdditionalSlot=true)
   > SetVisibility(self=Get Overlay_AdditionalSlot() -> (Overlay_AdditionalSlot:Overlay Object Reference), InVisibility=HitTestInvisible)
   > HideGizmoSlots()
   > Set IsGizmoSlotPreviewed(IsGizmoSlotPreviewed:bool=false) -> (Output_Get:bool)
   n13 = "Equal (Integer)"(A=Get CurrentWeaponSlot() -> (CurrentWeaponSlot=0))
   > HandleGizmoSlotPreview(ShowGizmoPreview=@n13) {is=title K2Node_CallFunction}
   > Branch(Condition=@n13)
   > AddLooseGameplayTags(Actor=GetOwningPlayer(), GameplayTags=Get AllowTogglePickAxeTag(), bShouldReplicate=true)
   n14 = Get GizmoSlotMap()
   Branch(execute=@n11.else, Condition="integer > integer"(A=Length(TargetMap=@n14) -> (ReturnValue=0)))
   > n15 = Set GizmoSlot(GizmoSlot=@n12)
   n16 = Create Widget!(
    execute=@n10.Is Not Valid
    Class=Get AdditionalQuickbarSlotClass() -> (AdditionalQuickbarSlotClass:WBP BRGame UI Quickbar Slot Primary Class Reference)
    OwningPlayer=GetOwningPlayer()
    InstituteChirpQuickbarSlotViewModel:Institute Chirp - HUD Controller Quickbar Slot Viewmodel Object Reference
   ) -> (then, ReturnValue:WBP BRGame UI Quickbar Slot Primary Object Reference)
   > n17 = Add!(
    execute
    TargetMap:Map of Integers to WBP BRGame UI Quickbar Slot Primary Object References=Get GizmoSlotMap()
    Key:"Integer (by ref)"=Get CurrentSlot() -> (CurrentSlot=0)
    Value:"WBP BRGame UI Quickbar Slot Primary Object Reference (by ref)"=@n16
   ) -> (then) {is=title K2Node_CallFunction}
   n18 = Get HasItemInCurrentSlot()
   Branch(execute=@n9.else, Condition=AND Boolean(A=NOT Boolean(A=@n18), B=Get HasAdditionalSlot()))
   n19 = Get GizmoSlotMap()
   > n20 = Branch(Condition="integer > integer"(A=Length(TargetMap=@n19) -> (ReturnValue=0)))
   n21 = Get CurrentSlot() -> (CurrentSlot=0)
   n22 = Get GizmoSlotMap()
   > RemoveFromParent(self=[Get GizmoSlot(), Find(TargetMap=@n22, Key=@n21)])
   > n23 = Remove!(execute, TargetMap:Map of Integers to WBP BRGame UI Quickbar Slot Primary Object References=@n22, Key:"Integer (by ref)"=@n21) -> (then, ReturnValue:bool)
   Set GizmoSlot(execute=[@n23.then, @n20.else])
   n24 = HasAnyValidAdditionalSlot() {is=title K2Node_CallFunction}
   > Set HasAdditionalSlot(HasAdditionalSlot=@n24)
   > SetVisibility(
    self=Get Overlay_AdditionalSlot()
    InVisibility=Select(Option 0:ESlateVisibility Enum=Collapsed, Option 1:ESlateVisibility Enum=HitTestInvisible, Index=@n24) -> (ReturnValue:ESlateVisibility Enum)
   )
   > Set IsGizmoSlotPreviewed(IsGizmoSlotPreviewed=false)
   > GetFirstGizmoWithQuantity!(execute, self) -> (then, GizmoSlot:WBP BRGame UI Quickbar Slot Primary Object Reference) {is=title K2Node_CallFunction}
   > HandleGizmoSlotPreview(ShowGizmoPreview=@n24)
   > HandleAdditionalSlotPadding(HasAdditionalSlot=@n24)
   > SetRenderOpacity(self=Get GizmoSlot(), InOpacity=SelectFloat(A=1, B=0, bPickA=@n24))
   > AddChild(self=Get QuickbarSlot1_Overlay(), Content=Get QuickbarSlot1())
   > Set IsGizmoSlotFocused(IsGizmoSlotFocused=false)
   > Swap Slots 1 to 6() {is=title K2Node_CallFunction}
   > n25 = Branch(Condition=HasAnyValidAdditionalSlot())
   RemoveLooseGameplayTags(execute=@n25.else, Actor=GetOwningPlayer(), GameplayTags=Get AllowTogglePickAxeTag(), bShouldReplicate=true)
  ]}
 Refresh {nodes [
   Refresh() {is=title K2Node_FunctionEntry}
   > n1 = For Each Loop(Array:Array of Widget Object References=GetAllChildren(self=Get GridPanel_Slots() -> (GridPanel_Slots:Grid Panel Object Reference))) -> (Array Element:Widget Object Reference)
   > Set CurrentIndex(CurrentIndex:int=@n1.Array Index) -> (Output_Get:int=0)
   > n2 = Set CurrentGridSlot(CurrentGridSlot:Grid Slot Object Reference=SlotAsGridSlot(Widget=@n1)) -> (Output_Get:Grid Slot Object Reference)
   > Is Valid(InputObject=@n2)
   > SetRow!(execute, self:Grid Slot Object Reference=Get CurrentGridSlot() -> (CurrentGridSlot:Grid Slot Object Reference), InRow:int=0) -> (then)
   > n3 = Branch(Condition=Get IsForward() -> (IsForward:bool))
   > SetColumn!(execute, self:Grid Slot Object Reference=Get CurrentGridSlot(), InColumn:int=Get CurrentIndex() -> (CurrentIndex:int=0)) -> (then)
   n4 = Get QuickbarSlots() -> (QuickbarSlots:Array of Fort Quick Bar Slot Base Object References)
   n5 = Get CurrentIndex() -> (CurrentIndex=0)
   SetColumn(
    execute=@n3.else
    self=Get CurrentGridSlot()
    InColumn=int - int(A=Length(TargetArray:Array of Fort Quick Bar Slot Base Object References=@n4) {is=title K2Node_CallArrayFunction}, B=int + int(A=@n5))
   )
  ]}
 ResetSwappableSlotToDefault {nodes [
   ResetSwappableSlotToDefault() {is=title K2Node_FunctionEntry}
   > SetRenderTranslation(self=Get GizmoSlot(), Translation=(X=0,Y=0))
   > n1 = GetDynamicMaterial(self=Get Image_TileImage!(self:WBP BRGame UI Quickbar Slot Primary Object Reference=Get GizmoSlot()) -> (Image_TileImage:Image Object Reference))
   > SetScalarParameterValue(self=@n1, ParameterName=IsMasked, Value=0)
   > SetScalarParameterValue(self=@n1, ParameterName=IconYOffset, Value=0)
  ]}
 SetSwappableVisuals {nodes [
   n1 = SetSwappableVisuals!() -> (then, QuickbarSlot:WBP BRGame UI Quickbar Slot Primary Object Reference, IsFocused:bool=~) {is=title K2Node_FunctionEntry}
   > SetRenderTranslation(
    self=Get SizeBox_TileContents(self=@n1)
    Translation=Select(Option 0:Vector 2D Structure=(X=0,Y=-35), Option 1:Vector 2D Structure=(X=0,Y=0), Index=Get IsFocused() -> (IsFocused:bool)) -> (ReturnValue:Vector 2D Structure)
   )
   > n2 = GetDynamicMaterial(self=Get Image_TileImage(self=Get QuickbarSlot() -> (QuickbarSlot:WBP BRGame UI Quickbar Slot Primary Object Reference)))
   > SetScalarParameterValue(self=@n2, ParameterName=IsMasked, Value=SelectFloat(A=1, B=0, bPickA=Get IsFocused()))
   > SetScalarParameterValue(self=@n2, ParameterName=IconYOffset, Value=SelectFloat(A=0, B=0.15, bPickA=Get IsFocused()))
  ]}
 UpdateItemPreview {nodes [
   UpdateItemPreview() {is=title K2Node_FunctionEntry}
   n1 = Get Current Pickup Preview Item() -> (Current Pickup Preview Item:Fort Item Object Reference)
   > n2 = Is Valid(InputObject=@n1)
   > n3 = PreviewItemPickUp!(
    execute
    self:Fort HUDContext Object Reference=Get FortHUDContext()
    InItem:Fort Item Object Reference=@n1
    InQuickBarType:EFortQuickBars Enum=Primary
   ) -> (then, OutSlotIndex:int=0, OutItemsToPush:Map of Integers to Integers, OutItemsToDrop:Array of Integers, ReturnValue:bool)
   > n4 = Branch(Condition="integer > integer"(A=@n3))
   > n5 = Branch(Condition=@n3.ReturnValue)
   n6 = Break Widget Transform!(WidgetTransform:"Widget Transform Structure (by ref)"=Get RenderTransform!(self:Widget Object Reference=Get PickupPreview()) -> (RenderTransform:Widget Transform Structure)) -> (Translation:Vector 2D Structure, Scale:Vector 2D Structure, Shear:Vector 2D Structure, Angle:float=~) {is=title K2Node_BreakStruct}
   n7 = BreakVector2D(InVec=@n6)
   n8 = GetNumberOfSlotsTaken!(self:Fort Item Object Reference=Get Current Pickup Preview Item()) -> (ReturnValue:Byte=0)
   n9 = int + int(A=@n3, B="To Integer (Byte)"!(InByte:Byte=Byte - Byte!(A:Byte=@n8, B:Byte=1) -> (ReturnValue:Byte=0)) -> (ReturnValue:int=0)) -> (ReturnValue=0) {is=title K2Node_CommutativeAssociativeBinaryOperator}
   n10 = int * float!(
    A:int="Min (Integer)"(
     A=Get Slot Display Index!(SlotIndex:int=@n3) -> (ColumnIndex:int=0) {is=title K2Node_CallFunction}
     B=Get Slot Display Index(SlotIndex=@n9) -> (ColumnIndex=0)
    ) -> (ReturnValue=0)
    B:double=Get Item Preview Width() -> (Item Preview Width:double)
   ) -> (ReturnValue:double)
   > SetRenderTransform!(
    execute
    self:Widget Object Reference=Get PickupPreview()
    InTransform:Widget Transform Structure=Make Widget Transform!(
     Translation:Vector 2D Structure=MakeVector2D(X=@n10, Y=@n7.Y)
     Scale:Vector 2D Structure=@n6.Scale
     Shear:Vector 2D Structure=@n6.Shear
     Angle:float=@n6.Angle
    ) -> (WidgetTransform:Widget Transform Structure) {is=title K2Node_MakeStruct}
   ) -> (then)
   n11 = GetNumberOfSlotsTaken(self=Get Current Pickup Preview Item()) -> (ReturnValue=0)
   > SetWidthOverride(self=Get PickupPreview(), InWidthOverride=int * float(A="To Integer (Byte)"(InByte=@n11) -> (ReturnValue=0), B=Get Item Preview Width()))
   > SetVisibility(self=Get PickupPreview(), InVisibility=HitTestInvisible)
   SetVisibility(execute=[@n5.else, @n4.else, @n2.Is Not Valid], self=Get PickupPreview(), InVisibility=Hidden)
  ]}
]
variables [
 AdditionalQuickbarSlotClass {type=Class, default=/BattleRoyaleGameUI/HUD/QuickBar/WBP_BRGameUI_QuickbarSlot_Slottable.WBP_BRGameUI_QuickbarSlot_Slottable_C}
 AdditionalSlotPadding {type=Margin, default=<ltrb 0 20 0 20>}
 AllowTogglePickAxeTag {type=GameplayTagContainer, default {gameplayTags [{tagName=Gameplay.Status.TogglePickAxeEnabled}]}}
 Async Task Ref {type=AsyncAction_StartListeningToEvent, default=null}
 Current Pickup Preview Item {type=FortItem, default=null}
 CurrentPreviewGizmoSlot {type=WBP_BRGameUI_QuickbarSlot_Primary_C, default=null}
 CurrentWeaponSlot {type=integer}
 DefaultSlotPadding {type=Margin, default=<ltrb 0 0 0 0>}
 FortPlayerController {type=FortPlayerController, default=null}
 GizmoSlot {type=WBP_BRGameUI_QuickbarSlot_Primary_C, default=null}
 GizmoSlotMap {default {}}
 HasAdditionalSlot {type=boolean}
 IsAdditionalSlotTagAdded {type=boolean}
 IsGizmoSlotFocused {type=boolean}
 IsGizmoSlotPreviewed {type=boolean}
 IsMaxSlotsSet {type=boolean}
 IsRefreshChildren {type=boolean, default=true}
 Item Preview Width {type=number, default=94.5}
 PlayerController {type=PlayerController, default=null}
 PreviousWeaponSlot {type=integer}
 TotalSlots {type=integer}
]
widgets {
 info {
  inheritedWidgetCount=2
  parentClass=/Game/Athena/HUD/Bacchus/BacchusDynamicQuickbar.BacchusDynamicQuickbar_C
  rootWidgetClass=&Overlay
  widgetCount=25
 }
 widgets [
  Overlay_18 {widgetClassPath=&Overlay, widget=Overlay_18}
  CommonVisibilitySwitcher_30 {
   widgetClassPath=&CommonVisibilitySwitcher
   bIsVariable=true
   parent=Overlay_18
   widget=CommonVisibilitySwitcher_30
   slot {class=OverlaySlot}
  }
  GridPanel_Slots {
   widgetClassPath=&GridPanel
   bIsVariable=true
   parent=CommonVisibilitySwitcher_30
   widget=GridPanel_Slots
   properties {columnFill [1]}
   slot {class=CommonVisibilitySwitcherSlot}
  }
  Slot1Overlay {widgetClassPath=&Overlay, bIsVariable=true, parent=GridPanel_Slots, widget=Slot1Overlay, slot {class=GridSlot}}
  QuickbarSlot1_Overlay {widgetClassPath=&Overlay, bIsVariable=true, parent=Slot1Overlay, widget=QuickbarSlot1_Overlay, slot {class=OverlaySlot}}
  Overlay_AdditionalSlot {
   widgetClassPath=&Overlay
   bIsVariable=true
   parent=QuickbarSlot1_Overlay
   widget=Overlay_AdditionalSlot
   properties {bOverride_Cursor=true}
   slot {class=OverlaySlot, properties {horizontalAlignment=HAlign_Fill, verticalAlignment=VAlign_Fill}}
  }
  SwappableSlot_SizeBox {
   widgetClassPath=&SizeBox
   parent=Overlay_AdditionalSlot
   widget=SwappableSlot_SizeBox
   properties {bOverride_WidthOverride=true, heightOverride=160, widthOverride=90}
   slot {class=OverlaySlot, properties {horizontalAlignment=HAlign_Fill, verticalAlignment=VAlign_Fill}}
  }
  SwappableSlot_Overlay {widgetClassPath=&Overlay, bIsVariable=true, parent=SwappableSlot_SizeBox, widget=SwappableSlot_Overlay, slot {class=SizeBoxSlot}}
  QuickbarSlot1 {
   widgetClassPath=/BattleRoyaleGameUI/HUD/QuickBar/WBP_BRGameUI_QuickbarSlot_Primary
   bIsVariable=true
   parent=QuickbarSlot1_Overlay
   widget=QuickbarSlot1
   properties {
    bCheckVisibilityBasedOffInventoryCapacity=true
    cooldownTypesSupported [Death]
    slotSizeBySlotUsage [<88 88>, <88 88>, <179 88>, <170 88>, <361 88>]
   }
   slot {class=OverlaySlot, properties {horizontalAlignment=HAlign_Fill, verticalAlignment=VAlign_Fill}}
  }
  Slot2Overlay {
   widgetClassPath=&Overlay
   bIsVariable=true
   parent=GridPanel_Slots
   widget=Slot2Overlay
   slot {class=GridSlot, properties {column=1}}
  }
  QuickbarSlot2 {
   widgetClassPath=&WBP_BRGameUI_QuickbarSlot_Primary
   bIsVariable=true
   parent=Slot2Overlay
   widget=QuickbarSlot2
   properties {
    bCheckVisibilityBasedOffInventoryCapacity=true
    slotIndex=1
    cooldownTypesSupported [Death]
    slotSizeBySlotUsage [<88 88>, <88 88>, <179 88>, <170 88>, <361 88>]
   }
   slot {class=OverlaySlot}
  }
  Slot3Overlay {
   widgetClassPath=&Overlay
   bIsVariable=true
   parent=GridPanel_Slots
   widget=Slot3Overlay
   slot {class=GridSlot, properties {column=2}}
  }
  QuickbarSlot3 {
   widgetClassPath=&WBP_BRGameUI_QuickbarSlot_Primary
   bIsVariable=true
   parent=Slot3Overlay
   widget=QuickbarSlot3
   properties {
    bCheckVisibilityBasedOffInventoryCapacity=true
    slotIndex=2
    cooldownTypesSupported [Death]
    slotSizeBySlotUsage [<88 88>, <88 88>, <179 88>, <170 88>, <361 88>]
   }
   slot {class=OverlaySlot}
  }
  Slot4Overlay {
   widgetClassPath=&Overlay
   bIsVariable=true
   parent=GridPanel_Slots
   widget=Slot4Overlay
   slot {class=GridSlot, properties {column=3}}
  }
  QuickbarSlot4 {
   widgetClassPath=&WBP_BRGameUI_QuickbarSlot_Primary
   bIsVariable=true
   parent=Slot4Overlay
   widget=QuickbarSlot4
   properties {
    bCheckVisibilityBasedOffInventoryCapacity=true
    slotIndex=3
    cooldownTypesSupported [Death]
    slotSizeBySlotUsage [<88 88>, <88 88>, <179 88>, <170 88>, <361 88>]
   }
   slot {class=OverlaySlot}
  }
  Slot5Overlay {
   widgetClassPath=&Overlay
   bIsVariable=true
   parent=GridPanel_Slots
   widget=Slot5Overlay
   slot {class=GridSlot, properties {column=4}}
  }
  QuickbarSlot5 {
   widgetClassPath=&WBP_BRGameUI_QuickbarSlot_Primary
   bIsVariable=true
   parent=Slot5Overlay
   widget=QuickbarSlot5
   properties {
    bCheckVisibilityBasedOffInventoryCapacity=true
    slotIndex=4
    cooldownTypesSupported [Death]
    slotSizeBySlotUsage [<88 88>, <88 88>, <179 88>, <170 88>, <361 88>]
   }
   slot {class=OverlaySlot}
  }
  Slot6Overlay {
   widgetClassPath=&Overlay
   bIsVariable=true
   parent=GridPanel_Slots
   widget=Slot6Overlay
   slot {class=GridSlot, properties {column=5}}
  }
  QuickbarSlot6 {
   widgetClassPath=&WBP_BRGameUI_QuickbarSlot_Primary
   bIsVariable=true
   parent=Slot6Overlay
   widget=QuickbarSlot6
   properties {
    bCheckVisibilityBasedOffInventoryCapacity=true
    slotIndex=5
    cooldownTypesSupported [Death]
    slotSizeBySlotUsage [<88 88>, <88 88>, <179 88>, <170 88>, <361 88>]
   }
   slot {class=OverlaySlot}
  }
  HorizontalBox_80 {
   widgetClassPath=&HorizontalBox
   parent=CommonVisibilitySwitcher_30
   widget=HorizontalBox_80
   slot {class=CommonVisibilitySwitcherSlot}
  }
  DynamicQuickbarBox {
   widgetClassPath=&DynamicEntryBox
   bIsVariable=true
   parent=CommonVisibilitySwitcher_30
   widget=DynamicQuickbarBox
   properties {
    entryWidgetClass=/BattleRoyaleGameUI/HUD/QuickBar/WBP_BRGameUI_QuickbarSlot_Primary.WBP_BRGameUI_QuickbarSlot_Primary_C
    numDesignerPreviewEntries=6
   }
   slot {class=CommonVisibilitySwitcherSlot}
  }
  PickupPreview {
   widgetClassPath=&SizeBox
   bIsVariable=true
   parent=Overlay_18
   widget=PickupPreview
   properties {bOverride_HeightOverride=true, bOverride_WidthOverride=true, heightOverride=110, visibility=Hidden, widthOverride=95}
   slot {class=OverlaySlot}
  }
  Border_23 {
   widgetClassPath=&Border
   parent=PickupPreview
   widget=Border_23
   properties {
    visibility=SelfHitTestInvisible
    background {
     bIsDynamicallyLoaded=false
     drawAs=Box
     imageType=NoImage
     mirroring=NoMirror
     resourceName=None
     resourceObject=/Game/UI/InGame/Quickbar/M_UI_FramePreview_Inst
     tiling=NoTile
     imageSize=<80 80>
     margin=<ltrb 0 0 0 0>
     outlineSettings {
      bUseBrushTransparency=false
      roundingType=HalfHeightRadius
      width=0
      color {colorUseRule=UseColor_Specified, specifiedColor=<rgba 0 0 0 0>}
      cornerRadii=<0 0 0 0>
     }
     tintColor {colorUseRule=UseColor_Specified, specifiedColor=<rgba 1 1 1 1>}
     uVRegion {bIsValid=false, max=<0 0>, min=<0 0>}
    }
    padding=<ltrb -200 -200 -200 -200>
    renderTransformPivot=<0 0>
   }
   slot {class=SizeBoxSlot, properties {padding=<ltrb -10 -10 -10 -10>}}
  }
  KeybindWidget_SwitchModeKeybind {widgetClassPath=/Script/FortniteUI.FortKeybindWidget, bInherited=true, widget=null}
  TutorialHighlight {widgetClassPath=&TutorialHighlight}
 ]
}
