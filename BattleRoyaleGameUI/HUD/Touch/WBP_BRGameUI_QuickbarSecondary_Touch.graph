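# Touch HUD quickbar for Battle Royale's secondary bar, with five build and trap slots, extending BacchusDynamicQuickbar.
# Construct initializes and maximizes the slots, tracks focused slot changes and refreshes visibility on item pickup.
# Under HUD tags, no-build or traps-in-no-build playlists, it collapses unless CanDisplayTrapSlot passes.
# If shown there, it collapses the first four slots, leaving the trap slot, though smart build skips that per-slot toggle.
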WidgetBlueprint /BattleRoyaleGameUI/HUD/Touch/WBP_BRGameUI_QuickbarSecondary_Touch
graphs [
 CanDisplayTrapSlot {nodes [
   CanDisplayTrapSlot() {is=title K2Node_FunctionEntry}
   n1 = GetOwningPlayerPawn!(self:User Widget Object Reference) -> (ReturnValue:Pawn Object Reference)
   > n2 = Cast To FortInventoryOwnerInterface!(execute, Object=GetLocalViewingPlayerController!(self:Pawn Object Reference=@n1) -> (ReturnValue:Player Controller Object Reference)) -> (then, CastFailed, AsFort Inventory Owner Interface:Fort Inventory Owner Interface Interface) {is=title K2Node_DynamicCast}
   > n3 = GetInventoryItemWithTag!(
    execute
    InventoryOwner:Fort Inventory Owner Interface Interface=@n2
    TagToFind:"Gameplay Tag Structure (by ref)"=MakeLiteralGameplayTag!(Value:Gameplay Tag Structure=(TagName="Trap")) -> (ReturnValue:Gameplay Tag Structure)
    bExactMatch:bool=false
   ) -> (then, ReturnValue:Fort Item Object Reference)
   n4 = HasAnyMatchingGameplayTags!(
    self:Gameplay Tag Asset Interface Interface=GetAbilitySystemComponent!(Actor:Actor Object Reference=@n1) -> (ReturnValue:Ability System Component Object Reference)
    TagContainer:"Gameplay Tag Container Structure (by ref)"=Get GenericHideTags() -> (GenericHideTags:Gameplay Tag Container Structure)
   ) -> (ReturnValue:bool)
   n5 = PlaylistHasTag!(
    self:Fort Game State Athena Object Reference=GetGameStateAthenaFromContext!(WorldContextObject:Object Reference=@n1) -> (ReturnValue:Fort Game State Athena Object Reference)
    InTag:"Gameplay Tag Structure (by ref)"=Get TrapsInNoBuildPlaylistTag() -> (TrapsInNoBuildPlaylistTag:Gameplay Tag Structure)
   ) -> (ReturnValue:bool)
   n6 = IsValid(Object=@n3)
   > Return Node!(execute, HasTrap:bool=AND Boolean(A=OR Boolean(A=@n5, B=@n6), B=NOT Boolean(A=@n4)))
  ]}
 EventGraph {nodes [
   n1 = UpdateSlots!() -> (OutputDelegate, then, NewItem:Fort World Item Object Reference, Count:int) {is=title K2Node_CustomEvent}
   > Update Visibility with Tags()
   n2 = OnQuickBarSlotFocusChanged_Bind!() -> (OutputDelegate, then, QuickbarIndex:EFortQuickBars Enum=Primary, Slot:int) {is=title K2Node_CustomEvent}
   > Branch(Condition="Equal (Enum)"(A:EFortQuickBars Enum=@n2.QuickbarIndex, B:EFortQuickBars Enum=Secondary))
   > Set CurrentBuildingSlot(CurrentBuildingSlot:int=@n2.Slot) -> (Output_Get:int=0)
   Event Construct()
   > Parent: Construct()
   > n3 = Sequence!(execute) -> (then_0, then_1, then_2:Exec)
   Parent: Construct(execute=@n3.then_1)
   Assign On Quickbar Slot Focus Changed!(
    execute:Exec=@n3.then_2
    self:Fort Inventory Context Object Reference=Get FortInventoryContext!() -> (ReturnValue:Fort Inventory Context Object Reference) {is=title K2Node_GetBlueprintContextEx}
    Delegate:"Delegate (by ref)"=Create Event()
   ) -> (then:Exec) {is=title K2Node_AssignDelegate}
   > Bind Event to On Item Picked Up(self:Fort Inventory Context Object Reference=Get FortInventoryContext(), Delegate=@n1) {is=title K2Node_AddDelegate}
   n4 = Set QuickbarSlots(
    execute=@n3.then_0
    QuickbarSlots:Array of Fort Quick Bar Slot Base Object References=Make Array!(
     "[0]":Fort Quick Bar Slot Base Object Reference=Get QuickbarSlotF1() -> (QuickbarSlotF1:WBP BRGame UI Quickbar Slot Secondary Object Reference)
     "[1]":Fort Quick Bar Slot Base Object Reference=Get QuickbarSlotF2() -> (QuickbarSlotF2:WBP BRGame UI Quickbar Slot Secondary Object Reference)
     "[2]":Fort Quick Bar Slot Base Object Reference=Get QuickbarSlotF3() -> (QuickbarSlotF3:WBP BRGame UI Quickbar Slot Secondary Object Reference)
     "[3]":Fort Quick Bar Slot Base Object Reference=Get QuickbarSlotF4() -> (QuickbarSlotF4:WBP BRGame UI Quickbar Slot Secondary Object Reference)
     "[4]":Fort Quick Bar Slot Base Object Reference=Get QuickbarSlotF5() -> (QuickbarSlotF5:WBP BRGame UI Quickbar Slot Secondary Object Reference)
    ) -> (Array:Array of Fort Quick Bar Slot Base Object References)
   ) -> (Output_Get:Array of Fort Quick Bar Slot Base Object References)
   > n5 = For Each Loop(Array:Array of Fort Quick Bar Slot Base Object References=@n4) -> (Array Element:Fort Quick Bar Slot Base Object Reference)
   > n6 = Cast To WBP_BRGameUI_QuickbarSlot_Secondary!(execute, Object=@n5) -> (then, CastFailed, AsWBP BRGame UI Quickbar Slot Secondary:WBP BRGame UI Quickbar Slot Secondary Object Reference) {is=title K2Node_DynamicCast}
   > InitializeDataModel!(
    execute
    self:WBP BRGame UI Quickbar Slot Secondary Object Reference=@n6
    SlotNum:int=@n5.Array Index
    Quick Bar Type:EFortQuickBars Enum=Secondary
   ) -> (then)
   > MaximizeSlot!(execute, self:Fort Quick Bar Slot Base Object Reference=@n6, bShouldSkipAnimation:bool=true) -> (then)
   Event Minimize() {is=title K2Node_Event}
   Event Maximize() {is=title K2Node_Event}
  ]}
 GetSpecificSlot {nodes [
   n1 = GetSpecificSlot!() -> (then, SlotNumber:int=0) {is=title K2Node_FunctionEntry}
   n2 = Make Array(
    "[0]":WBP BRGame UI Quickbar Slot Secondary Object Reference=Get QuickbarSlotF1()
    "[1]":WBP BRGame UI Quickbar Slot Secondary Object Reference=Get QuickbarSlotF2()
    "[2]":WBP BRGame UI Quickbar Slot Secondary Object Reference=Get QuickbarSlotF3()
    "[3]":WBP BRGame UI Quickbar Slot Secondary Object Reference=Get QuickbarSlotF4()
    "[4]":WBP BRGame UI Quickbar Slot Secondary Object Reference=Get QuickbarSlotF5()
   ) -> (Array:Array of WBP BRGame UI Quickbar Slot Secondary Object References)
   > Branch(Condition=Is Valid Index!(TargetArray:Array of WBP BRGame UI Quickbar Slot Secondary Object References=@n2, IndexToTest:int=@n1) -> (ReturnValue:bool))
   > Return Node!(
    execute
    ReturnValue:Fort Quick Bar Slot Base Object Reference="Get (a copy)"(Array:Array of WBP BRGame UI Quickbar Slot Secondary Object References=@n2, Dimension 1=@n1) -> (Output:WBP BRGame UI Quickbar Slot Secondary Object Reference)
   )
  ]}
 ToggleDisplayNonTrapSlots {nodes [
   n1 = ToggleDisplayNonTrapSlots!() -> (then, bSetToVisible:bool=~) {is=title K2Node_FunctionEntry}
   > n2 = Cast To FortPlayerController(Object=GetOwningPlayer())
   > n3 = Branch(Condition=GetSmartBuildEnabled!(self:Fort Client Settings Record Object Reference=GetClientSettings!(self:Fort Player Controller Object Reference=@n2) -> (ReturnValue:Fort Client Settings Record Object Reference)) -> (ReturnValue:bool))
   n4 = Branch(execute=@n3.else, Condition=@n1)
   n5 = Get QuickbarSlotF1()
   n6 = Get QuickbarSlotF2()
   n7 = Get QuickbarSlotF3()
   n8 = Get QuickbarSlotF4()
   > SetVisibility(self=[@n5, @n6, @n7, @n8], InVisibility=Visible)
   SetVisibility(execute=@n4.else, self=[@n5, @n6, @n7, @n8], InVisibility=Collapsed)
  ]}
 Update Visibility with Tags {nodes [
   Update Visibility with Tags() {is=title K2Node_FunctionEntry}
   n1 = GetOwningPlayer()
   > n2 = Is Valid(InputObject=@n1)
   n3 = GetAbilitySystemComponent(Actor=@n1)
   > n4 = Is Valid(InputObject=@n3)
   n5 = GetGameStateAthenaFromContext(WorldContextObject=@n1)
   > n6 = Is Valid(InputObject=@n5)
   n7 = Get HUDElementTag() -> (HUDElementTag:Gameplay Tag Container Structure)
   n8 = Get TrapsInNoBuildPlaylistTag()
   > n9 = Branch(Condition=OR Boolean!(
     A=HasAnyMatchingGameplayTags(self=@n3, TagContainer=@n7)
     B=IsNoBuildBRPlaylist!(self:Fort Game State Athena Object Reference=@n5) -> (ReturnValue:bool)
     C:bool=PlaylistHasTag(self=@n5, InTag=@n8)
    ) -> (ReturnValue))
   > n10 = Branch(Condition=CanDisplayTrapSlot!(self) -> (HasTrap:bool) {is=title K2Node_CallFunction})
   > SetVisibility(InVisibility=SelfHitTestInvisible)
   > ToggleDisplayNonTrapSlots!(execute, self, bSetToVisible:bool=false) -> (then) {is=title K2Node_CallFunction}
   SetVisibility(execute=@n10.else, InVisibility=Collapsed)
   SetVisibility(execute=[@n9.else, @n6.Is Not Valid, @n4.Is Not Valid, @n2.Is Not Valid], InVisibility=SelfHitTestInvisible)
   > ToggleDisplayNonTrapSlots(bSetToVisible=true)
  ]}
]
variables [
 CurrentBuildingSlot {type=integer}
 GenericHideTags {
  type=GameplayTagContainer
  default {gameplayTags [
    {tagName=Athena.Creative.HideForLoading}
    {tagName=Athena.Creative.HUDController.HideBuildMenu}
    {tagName=Athena.HideForCapture}
    {tagName=HUD.Athena.Hide.WithAboveQuickbarOverride}
    {tagName=HUD.Athena.Hide.WithInventory}
    {tagName=HUD.Athena.Hide.WithRCPawn}
    {tagName=HUD.Athena.HideForAbility}
    {tagName=HUD.Athena.HideForConversation}
    {tagName=HUD.Athena.HideForGadget}
    {tagName=HUD.Athena.HideForPostGame}
    {tagName=HUD.Athena.HideforRadial}
    {tagName=HUD.Athena.HideforVehicle}
    {tagName=HUD.Athena.HideOnBus}
    {tagName=Hud.Athena.HideSecondaryForAbility}
    {tagName=HUD.Event.HideForSpecialEvent}
    {tagName=HUD.Hide.WhileLoading}
    {tagName=HUD.Hide.CreativeJoiningLobby}
    {tagName=HUD.LTM.HideForGoose}
    {tagName=HUD.LTM.HideForTryHard}
    {tagName=Gameplay.Action.Player.BlockBuildMode}
   ]}
 }
 TrapsInNoBuildPlaylistTag {type=GameplayTag, default {tagName=Athena.Playlist.TrapsInNoBuild}}
]
widgets {
 info {
  inheritedWidgetCount=2
  parentClass=/Game/Athena/HUD/Bacchus/BacchusDynamicQuickbar.BacchusDynamicQuickbar_C
  rootWidgetClass=&SizeBox
  widgetCount=12
 }
 widgets [
  SizeBox_77 {widgetClassPath=&SizeBox, widget=SizeBox_77, properties {bOverride_MinDesiredWidth=true, minDesiredWidth=500}}
  HorizontalBox_140 {
   widgetClassPath=&HorizontalBox
   parent=SizeBox_77
   widget=HorizontalBox_140
   slot {class=SizeBoxSlot, properties {padding=<ltrb 34 0 0 0>}}
  }
  "[Horizontal Box] Camera Mode Content" {
   widgetClassPath=&HorizontalBox
   bIsVariable=true
   parent=HorizontalBox_140
   widget="[Horizontal Box] Camera Mode Content"
   slot {class=HorizontalBoxSlot, properties {size {sizeRule=Fill, value=1}}}
  }
  Overlay_0 {
   widgetClassPath=&Overlay
   parent="[Horizontal Box] Camera Mode Content"
   widget=Overlay_0
   slot {class=HorizontalBoxSlot, properties {padding=<ltrb 8 0 8 0>}}
  }
  GridPanel_0 {widgetClassPath=&GridPanel, parent=Overlay_0, widget=GridPanel_0, slot {class=OverlaySlot}}
  QuickbarSlotF1 {
   widgetClassPath=/BattleRoyaleGameUI/HUD/QuickBar/WBP_BRGameUI_QuickbarSlot_Secondary
   bIsVariable=true
   parent=GridPanel_0
   widget=QuickbarSlotF1
   properties {bHideIfNoItemAssigned=true, quickBarType=Secondary}
   slot {class=GridSlot, properties {horizontalAlignment=HAlign_Left, verticalAlignment=VAlign_Top}}
  }
  QuickbarSlotF2 {
   widgetClassPath=&WBP_BRGameUI_QuickbarSlot_Secondary
   bIsVariable=true
   parent=GridPanel_0
   widget=QuickbarSlotF2
   properties {bHideIfNoItemAssigned=true, quickBarType=Secondary, slotIndex=1}
   slot {class=GridSlot, properties {column=1, horizontalAlignment=HAlign_Left, verticalAlignment=VAlign_Top}}
  }
  QuickbarSlotF3 {
   widgetClassPath=&WBP_BRGameUI_QuickbarSlot_Secondary
   bIsVariable=true
   parent=GridPanel_0
   widget=QuickbarSlotF3
   properties {bHideIfNoItemAssigned=true, quickBarType=Secondary, slotIndex=2}
   slot {class=GridSlot, properties {column=3, horizontalAlignment=HAlign_Left, verticalAlignment=VAlign_Top}}
  }
  QuickbarSlotF4 {
   widgetClassPath=&WBP_BRGameUI_QuickbarSlot_Secondary
   bIsVariable=true
   parent=GridPanel_0
   widget=QuickbarSlotF4
   properties {bHideIfNoItemAssigned=true, quickBarType=Secondary, slotIndex=3}
   slot {class=GridSlot, properties {column=4, horizontalAlignment=HAlign_Left, verticalAlignment=VAlign_Top}}
  }
  QuickbarSlotF5 {
   widgetClassPath=&WBP_BRGameUI_QuickbarSlot_Secondary
   bIsVariable=true
   parent=GridPanel_0
   widget=QuickbarSlotF5
   properties {bHideIfNoItemAssigned=true, quickBarType=Secondary, slotIndex=4}
   slot {class=GridSlot, properties {column=6, horizontalAlignment=HAlign_Left, verticalAlignment=VAlign_Top}}
  }
  KeybindWidget_SwitchModeKeybind {widgetClassPath=/Script/FortniteUI.FortKeybindWidget, bInherited=true, widget=null}
  TutorialHighlight {widgetClassPath=&TutorialHighlight}
 ]
}
