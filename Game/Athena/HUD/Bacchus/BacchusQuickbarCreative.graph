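# Creative-mode HUD quickbar widget, derived from BacchusDynamicQuickbar, showing eight Creative quickbar slots in a single grid row.
# Construct fills QuickbarSlots before calling the parent Construct, then applies the current inventory-context focus and binds to focus changes.
# The focus handler records CurrentWeaponSlot only for nonzero slots on the Primary quickbar, although the slot widgets are typed Creative.
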WidgetBlueprint /Game/Athena/HUD/Bacchus/BacchusQuickbarCreative
graphs [EventGraph {nodes [
   n1 = OnQuickBarSlotFocusChanged_Bind!() -> (OutputDelegate, then, QuickbarIndex:EFortQuickBars Enum=Primary, Slot:int) {is=title K2Node_CustomEvent}
   > Branch(Condition=AND Boolean(
     A="Equal (Enum)"(A:EFortQuickBars Enum=@n1.QuickbarIndex, B:EFortQuickBars Enum=Primary)
     B="Not Equal (Integer)"!(A:int=@n1.Slot, B:int=0) -> (ReturnValue:bool) {is=title K2Node_CallFunction}
    ))
   > Set CurrentWeaponSlot(CurrentWeaponSlot:int=@n1.Slot) -> (Output_Get:int=0)
   Event Construct()
   > n2 = Sequence!(execute) -> (then_0, then_1, then_2:Exec, then_3:Exec, then_4:Exec)
   Parent: Construct(execute=@n2.then_1)
   Set QuickbarSlots(
    execute=@n2.then_0
    QuickbarSlots:Array of Fort Quick Bar Slot Base Object References=Make Array!(
     "[0]":Fort Quick Bar Slot Base Object Reference=Get QuickbarSlot1() -> (QuickbarSlot1:Bacchus Quickbar Slot Object Reference)
     "[1]":Fort Quick Bar Slot Base Object Reference=Get QuickbarSlot2() -> (QuickbarSlot2:Bacchus Quickbar Slot Object Reference)
     "[2]":Fort Quick Bar Slot Base Object Reference=Get QuickbarSlot3() -> (QuickbarSlot3:Bacchus Quickbar Slot Object Reference)
     "[3]":Fort Quick Bar Slot Base Object Reference=Get QuickbarSlot4() -> (QuickbarSlot4:Bacchus Quickbar Slot Object Reference)
     "[4]":Fort Quick Bar Slot Base Object Reference=Get QuickbarSlot5() -> (QuickbarSlot5:Bacchus Quickbar Slot Object Reference)
     "[5]":Fort Quick Bar Slot Base Object Reference=Get QuickbarSlot6() -> (QuickbarSlot6:Bacchus Quickbar Slot Object Reference)
     "[6]":Fort Quick Bar Slot Base Object Reference=Get QuickbarSlot7() -> (QuickbarSlot7:Bacchus Quickbar Slot Object Reference)
     "[7]":Fort Quick Bar Slot Base Object Reference=Get QuickbarSlot8() -> (QuickbarSlot8:Bacchus Quickbar Slot Object Reference)
    ) -> (Array:Array of Fort Quick Bar Slot Base Object References)
   ) -> (Output_Get:Array of Fort Quick Bar Slot Base Object References)
   n3 = GetQuickbarFocus!(self:Fort Inventory Context Object Reference=Get FortInventoryContext!() -> (ReturnValue:Fort Inventory Context Object Reference) {is=title K2Node_GetBlueprintContextEx}) -> (OutQuickBar:EFortQuickBars Enum=Primary, OutSlot:int=0, OutSecondarySlot:int=0, OutPreviousFocusedSlot:int=0)
   OnQuickbarSlotFocusChanged!(execute=@n2.then_2, self:Fort Quickbar Base Object Reference, QuickbarIndex:EFortQuickBars Enum=@n3, SlotIndex:int=@n3.OutSlot) -> (then)
   Assign On Quickbar Slot Focus Changed!(
    execute:Exec=@n2.then_3
    self:Fort Inventory Context Object Reference=Get FortInventoryContext()
    Delegate:"Delegate (by ref)"=Create Event()
   ) -> (then:Exec) {is=title K2Node_AssignDelegate}
  ]}]
variables [CurrentWeaponSlot {type=integer, default=1}]
widgets {
 info {
  inheritedWidgetCount=2
  parentClass=&/BacchusDynamicQuickbar.BacchusDynamicQuickbar_C
  rootWidgetClass=&HorizontalBox
  widgetCount=13
 }
 widgets [
  HorizontalBox_140 {widgetClassPath=&HorizontalBox, widget=HorizontalBox_140}
  Overlay_1 {
   widgetClassPath=&Overlay
   parent=HorizontalBox_140
   widget=Overlay_1
   slot {class=HorizontalBoxSlot, properties {padding=<ltrb 0 0 8 0>}}
  }
  GridPanel_0 {widgetClassPath=&GridPanel, parent=Overlay_1, widget=GridPanel_0, slot {class=OverlaySlot}}
  QuickbarSlot1 {
   widgetClassPath=&/BacchusQuickBarSlot.BacchusQuickbarSlot
   bIsVariable=true
   parent=GridPanel_0
   widget=QuickbarSlot1
   properties {
    bCheckVisibilityBasedOffInventoryCapacity=false
    bConsumePointerInput=false
    quickBarType=Creative
    shouldCollapseItemWidgetBorder=true
    slotIndex=1
    use Native=true
    cooldownTypesSupported [Death]
   }
   slot {class=GridSlot}
  }
  QuickbarSlot2 {
   widgetClassPath=&BacchusQuickbarSlot
   bIsVariable=true
   parent=GridPanel_0
   widget=QuickbarSlot2
   properties {
    bCheckVisibilityBasedOffInventoryCapacity=false
    bConsumePointerInput=false
    quickBarType=Creative
    shouldCollapseItemWidgetBorder=true
    slotIndex=2
    use Native=true
    cooldownTypesSupported [Death, WeaponReloading, ItemActivation, AthenaWeaponFireCooldown]
   }
   slot {class=GridSlot, properties {column=1}}
  }
  QuickbarSlot3 {
   widgetClassPath=&BacchusQuickbarSlot
   bIsVariable=true
   parent=GridPanel_0
   widget=QuickbarSlot3
   properties {
    bCheckVisibilityBasedOffInventoryCapacity=false
    bConsumePointerInput=false
    quickBarType=Creative
    shouldCollapseItemWidgetBorder=true
    slotIndex=3
    use Native=true
    cooldownTypesSupported [Death, WeaponReloading, ItemActivation, AthenaWeaponFireCooldown]
   }
   slot {class=GridSlot, properties {column=2}}
  }
  QuickbarSlot4 {
   widgetClassPath=&BacchusQuickbarSlot
   bIsVariable=true
   parent=GridPanel_0
   widget=QuickbarSlot4
   properties {
    bCheckVisibilityBasedOffInventoryCapacity=false
    bConsumePointerInput=false
    quickBarType=Creative
    shouldCollapseItemWidgetBorder=true
    slotIndex=4
    use Native=true
    cooldownTypesSupported [Death, WeaponReloading, ItemActivation, AthenaWeaponFireCooldown]
   }
   slot {class=GridSlot, properties {column=3}}
  }
  QuickbarSlot5 {
   widgetClassPath=&BacchusQuickbarSlot
   bIsVariable=true
   parent=GridPanel_0
   widget=QuickbarSlot5
   properties {
    bCheckVisibilityBasedOffInventoryCapacity=false
    bConsumePointerInput=false
    quickBarType=Creative
    shouldCollapseItemWidgetBorder=true
    slotIndex=5
    use Native=true
    cooldownTypesSupported [Death, WeaponReloading, ItemActivation, AthenaWeaponFireCooldown]
   }
   slot {class=GridSlot, properties {column=4}}
  }
  QuickbarSlot6 {
   widgetClassPath=&BacchusQuickbarSlot
   bIsVariable=true
   parent=GridPanel_0
   widget=QuickbarSlot6
   properties {
    bCheckVisibilityBasedOffInventoryCapacity=false
    bConsumePointerInput=false
    quickBarType=Creative
    shouldCollapseItemWidgetBorder=true
    slotIndex=6
    use Native=true
    cooldownTypesSupported [Death, WeaponReloading, ItemActivation, AthenaWeaponFireCooldown]
   }
   slot {class=GridSlot, properties {column=5}}
  }
  QuickbarSlot7 {
   widgetClassPath=&BacchusQuickbarSlot
   bIsVariable=true
   parent=GridPanel_0
   widget=QuickbarSlot7
   properties {
    bCheckVisibilityBasedOffInventoryCapacity=false
    bConsumePointerInput=false
    quickBarType=Creative
    shouldCollapseItemWidgetBorder=true
    slotIndex=7
    use Native=true
    cooldownTypesSupported [Death, WeaponReloading, ItemActivation, AthenaWeaponFireCooldown]
   }
   slot {class=GridSlot, properties {column=6}}
  }
  QuickbarSlot8 {
   widgetClassPath=&BacchusQuickbarSlot
   bIsVariable=true
   parent=GridPanel_0
   widget=QuickbarSlot8
   properties {
    bCheckVisibilityBasedOffInventoryCapacity=false
    bConsumePointerInput=false
    quickBarType=Creative
    shouldCollapseItemWidgetBorder=true
    slotIndex=8
    use Native=true
    cooldownTypesSupported [Death, WeaponReloading, ItemActivation, AthenaWeaponFireCooldown]
   }
   slot {class=GridSlot, properties {column=7}}
  }
  KeybindWidget_SwitchModeKeybind {widgetClassPath=/Script/FortniteUI.FortKeybindWidget, bInherited=true, widget=null}
  TutorialHighlight {widgetClassPath=&TutorialHighlight}
 ]
}
