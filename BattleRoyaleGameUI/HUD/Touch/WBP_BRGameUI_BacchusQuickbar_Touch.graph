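# Touch HUD quickbar container for Battle Royale, extending BacchusQuickbar with a switcher over primary combat, secondary build and creative bars.
# Refresh scales ResizeBox by slot size relative to a default, then UpdateDirection passes the inverted bIsBackwards flag to each bar and refreshes it.
# Refresh and HUD state updates re-evaluate the secondary bar's tag-based visibility.
# Quickbar changes stop all slot use on the secondary bar, because the Select index is fixed to false.
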WidgetBlueprint /BattleRoyaleGameUI/HUD/Touch/WBP_BRGameUI_BacchusQuickbar_Touch
graphs [
 EventGraph {nodes [
   n1 = Event PreConstruct()
   > Branch(Condition=@n1.IsDesignTime)
   > Refresh(self:Bacchus Quickbar Object Reference)
   Event OnQuickbarChanged!() -> (OutputDelegate, then, NewBar:Fort Quickbar Base Object Reference, OldBar:Fort Quickbar Base Object Reference) {is=title K2Node_Event}
   > n2 = Cast To QuickbarBase!(
    execute
    Object=Select(
     Option 0:Fort Quickbar Base Object Reference=Get BacchusQuickbarSecondary() -> (BacchusQuickbarSecondary:Fort Quickbar Base Object Reference)
     Option 1:Fort Quickbar Base Object Reference=Get BacchusQuickbarPrimary() -> (BacchusQuickbarPrimary:Fort Quickbar Base Object Reference)
     Index=false
    ) -> (ReturnValue:Fort Quickbar Base Object Reference)
   ) -> (then, CastFailed, AsQuickbar Base:Quickbar Base Object Reference) {is=title K2Node_DynamicCast}
   > StopUsingAllSlots(self:Quickbar Base Object Reference=@n2)
   Event Refresh() {is=title K2Node_Event}
   > SetUserSpecifiedScale!(
    execute
    self:Scale Box Object Reference=Get ResizeBox() -> (ResizeBox:Scale Box Object Reference)
    InUserSpecifiedScale:float=float / float(A=Get SlotSize() -> (SlotSize:float), B=Get Default Slot Size() -> (Default Slot Size:double)) {is=title K2Node_CallFunction}
   ) -> (then)
   > UpdateDirection()
   > n3 = Parent: Refresh() {is=title K2Node_CallParentFunction}
   n4 = Cast To WBP_BRGameUI_QuickbarSecondary_Touch!(execute=[@n3.then, @n5.then], Object=Get BacchusQuickbarSecondary()) -> (then, CastFailed, AsWBP BRGame UI Quickbar Secondary Touch:WBP BRGame UI Quickbar Secondary Touch Object Reference) {is=title K2Node_DynamicCast}
   > Update Visibility with Tags(self:WBP BRGame UI Quickbar Secondary Touch Object Reference=@n4)
   n5 = Event OnHUDStateUpdate!() -> (OutputDelegate, then, NewState:"Fort HUD State Structure (by ref)") {is=title K2Node_Event}
  ]}
 UpdateDirection {nodes [
   UpdateDirection() {is=title K2Node_FunctionEntry}
   > n1 = Sequence!(execute) -> (then_0, then_1, then_2:Exec)
   > n2 = Cast To BacchusDynamicQuickbar!(execute, Object=Get BacchusQuickbarPrimary()) -> (then, CastFailed, AsBacchus Dynamic Quickbar:Bacchus Dynamic Quickbar Object Reference) {is=title K2Node_DynamicCast}
   n3 = NOT Boolean(A=Get bIsBackwards() -> (bIsBackwards:bool))
   > Set IsForward!(execute, IsForward:bool=@n3, self:Bacchus Dynamic Quickbar Object Reference=@n2) -> (then, Output_Get:bool)
   > n4 = Cast To WBP_BRGameUI_QuickbarPrimary_Touch!(execute, Object=@n2) -> (then, CastFailed, AsWBP BRGame UI Quickbar Primary Touch:WBP BRGame UI Quickbar Primary Touch Object Reference) {is=title K2Node_DynamicCast}
   > n5 = Branch(Condition=Get IsRefreshChildren!(self:WBP BRGame UI Quickbar Primary Touch Object Reference=@n4) -> (IsRefreshChildren:bool))
   > Refresh(self:Bacchus Dynamic Quickbar Object Reference=@n4)
   Refresh(execute=@n5.else, self=@n2)
   n6 = Cast To BacchusDynamicQuickbar(execute=@n1.then_1, Object=Get BacchusQuickbarSecondary())
   > Set IsForward(IsForward=@n3, self=@n6)
   > Refresh(self=@n6)
   n7 = Cast To BacchusDynamicQuickbar(execute=@n1.then_2, Object=Get BacchusQuickbarCreative() -> (BacchusQuickbarCreative:Fort Quickbar Base Object Reference))
   > Set IsForward(IsForward=@n3, self=@n7)
   Refresh(self=@n7)
  ]}
]
variables [
 BacchusQuickbarViewModel {type=MVVMViewModelBase, default=null}
 Default Slot Size {type=number, default=92}
]
widgets {
 info {inheritedWidgetCount=7, parentClass=/Script/FortniteUI.BacchusQuickbar, rootWidgetClass=&ScaleBox, widgetCount=9}
 widgets [
  ScaleBox_HUDWidgetScale {widgetClassPath=&ScaleBox, bInherited=true, widget=ScaleBox_HUDWidgetScale, properties {stretch=UserSpecified}}
  ResizeBox {
   widgetClassPath=&ScaleBox
   bIsVariable=true
   parent=ScaleBox_HUDWidgetScale
   widget=ResizeBox
   properties {stretch=UserSpecified, userSpecifiedScale=1.4}
   slot {class=ScaleBoxSlot}
  }
  Overlay_163 {widgetClassPath=&Overlay, parent=ResizeBox, widget=Overlay_163, slot {class=ScaleBoxSlot}}
  BarSwitcher {
   widgetClassPath=&CommonVisibilitySwitcher
   bInherited=true
   bIsVariable=true
   parent=Overlay_163
   widget=BarSwitcher
   slot {class=OverlaySlot}
  }
  BacchusQuickbarPrimary {
   widgetClassPath=&/WBP_BRGameUI_QuickbarPrimary_Touch
   bInherited=true
   bIsVariable=true
   parent=BarSwitcher
   widget=BacchusQuickbarPrimary
   properties {
    elementTag {tagName=HUD.Quickbar.Combat}
    hUDElementTag {gameplayTags [
      {tagName=HUD.Athena.HideOnBus}
      {tagName=Athena.HideForCapture}
      {tagName=HUD.Athena.HideForGadget}
      {tagName=HUD.Athena.Hide.WithInventory}
      {tagName=HUD.Athena.Hide.WithRCPawn}
      {tagName=HUD.Athena.HideForAbility}
      {tagName=HUD.Event.HideForSpecialEvent}
      {tagName=Athena.Creative.HideForLoading}
      {tagName=HUD.Hide.WhileLoading}
      {tagName=HUD.Athena.HideForPostGame}
      {tagName=HUD.Athena.HideforRadial}
      {tagName=HUD.Athena.HideforVehicle}
      {tagName=HUD.Athena.Hide.WithMap}
      {tagName=Athena.Creative.HUDController.HidePlayerInventory}
      {tagName=HUD.Hide.CreativeJoiningLobby}
     ]}
    renderTransform {angle=0, scale=<0.95 0.95>, shear=<0 0>, translation=<0 0>}
   }
   slot {class=CommonVisibilitySwitcherSlot, properties {horizontalAlignment=HAlign_Left, verticalAlignment=VAlign_Top, padding=<ltrb 20 0 0 0>}}
  }
  BacchusQuickbarSecondary {
   widgetClassPath=&/WBP_BRGameUI_QuickbarSecondary_Touch
   bInherited=true
   bIsVariable=true
   parent=BarSwitcher
   widget=BacchusQuickbarSecondary
   properties {
    elementTag {tagName=HUD.Quickbar.Build}
    hUDElementTag {gameplayTags [
      {tagName=HUD.Athena.HideOnBus}
      {tagName=Athena.HideForCapture}
      {tagName=HUD.Athena.HideForGadget}
      {tagName=HUD.Athena.Hide.WithInventory}
      {tagName=HUD.Athena.Hide.WithRCPawn}
      {tagName=HUD.Athena.HideForAbility}
      {tagName=HUD.LTM.HideForTryHard}
      {tagName=HUD.Event.HideForSpecialEvent}
      {tagName=HUD.Athena.HideSecondaryForAbility}
      {tagName=Athena.Creative.HideForLoading}
      {tagName=HUD.LTM.HideForGoose}
      {tagName=HUD.Hide.WhileLoading}
      {tagName=Gameplay.Action.Player.BlockBuildMode}
      {tagName=HUD.Athena.HideForPostGame}
      {tagName=HUD.Athena.HideforRadial}
      {tagName=HUD.Athena.HideforNoBuilding}
      {tagName=HUD.Athena.HideforVehicle}
      {tagName=HUD.Athena.HideForConversation}
      {tagName=Athena.Creative.HUDController.HideBuildMenu}
      {tagName=HUD.Hide.CreativeJoiningLobby}
      {tagName=HUD.Athena.Hide.WithAboveQuickbarOverride}
     ]}
   }
   slot {class=CommonVisibilitySwitcherSlot, properties {padding=<ltrb 15 10 0 0>}}
  }
  BacchusQuickbarCreative {
   widgetClassPath=/Game/Athena/HUD/Bacchus/BacchusQuickbarCreative
   bInherited=true
   bIsVariable=true
   parent=BarSwitcher
   widget=BacchusQuickbarCreative
   slot {class=CommonVisibilitySwitcherSlot}
  }
  TutorialHighlight {
   widgetClassPath=&TutorialHighlight
   bIsVariable=true
   parent=Overlay_163
   widget=TutorialHighlight
   properties {visibility=HitTestInvisible}
   slot {class=OverlaySlot, properties {horizontalAlignment=HAlign_Fill, verticalAlignment=VAlign_Fill}}
  }
  Image_Background {widgetClassPath=&Image, bInherited=true, widget=null}
 ]
}
