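# Battle Royale HUD widget showing the equipped item's ammo, durability, name and rarity beside a bracket.
# When visibility is set, it reads the focused Primary quickbar item a frame later, tints materials by rarity and plays the switch animation.
# Focusing a non-Primary quickbar fades it out.
# It registers two gameplay-tag widget hosts for injected item and ammo content, and restyles for compact touch UI.
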WidgetBlueprint /BattleRoyaleGameUI/HUD/QuickBar/WBP_BRGameUI_EquippedItemWidget
animations [
 Anim_SwitchItem {
  binding_count=5
  length_seconds=4
  object=&self:Anim_SwitchItem
  sequence {bindings [
    HorizontalBox_3 {
     id=#1
     possessed_class=HorizontalBox
     tracks [
      {
       type=MovieSceneFloatTrack
       display_name=Render Opacity
       sections [{type=MovieSceneFloatSection, end_frame=80, start_frame=0, channels [None {type=MovieSceneScriptingFloatChannel, keys [[0,0],[4,1]]}]}]
      }
      {
       type=MovieScene2DTransformTrack
       sections [{
         type=MovieScene2DTransformSection
         end_frame=80
         start_frame=0
         channels [
          Translation.X {type=MovieSceneScriptingFloatChannel, default=0}
          Translation.Y {type=MovieSceneScriptingFloatChannel, default=0}
          Angle {type=MovieSceneScriptingFloatChannel, default=0}
          Scale.X {type=MovieSceneScriptingFloatChannel, keys [[0,0.8],[9,1]]}
          Scale.Y {type=MovieSceneScriptingFloatChannel, keys [[0,0.8],[9,1]]}
          Shear.X {type=MovieSceneScriptingFloatChannel, default=0}
          Shear.Y {type=MovieSceneScriptingFloatChannel, default=0}
         ]
        }]
      }
     ]
    }
    Text_RarityLabel {
     id=#2
     possessed_class=CommonTextBlock
     tracks [
      {
       type=MovieSceneFloatTrack
       display_name=Render Opacity
       sections [{type=MovieSceneFloatSection, channels [None {type=MovieSceneScriptingFloatChannel, keys [[4,0],[7,1]]}]}]
      }
      {
       type=MovieScene2DTransformTrack
       sections [{
         type=MovieScene2DTransformSection
         channels [
          Translation.X {type=MovieSceneScriptingFloatChannel, keys [[4,32],[10,0]]}
          Translation.Y {type=MovieSceneScriptingFloatChannel, const [0,1]}
         ]
        }]
      }
     ]
    }
    RarityLabel {
     id=#3
     possessed_class=Border
     tracks [
      {
       type=MovieSceneFloatTrack
       display_name=Render Opacity
       sections [{type=MovieSceneFloatSection, channels [None {type=MovieSceneScriptingFloatChannel, keys [[4,0],[8,1],[40,1],[80,0]]}]}]
      }
      {
       type=MovieScene2DTransformTrack
       sections [{
         type=MovieScene2DTransformSection
         channels [
          Translation.X {type=MovieSceneScriptingFloatChannel, keys [[4,32],[8,0]]}
          Translation.Y {type=MovieSceneScriptingFloatChannel, const [0,1]}
          Angle {type=MovieSceneScriptingFloatChannel, const [0,1]}
          Scale.X {type=MovieSceneScriptingFloatChannel, const [1,1]}
          Scale.Y {type=MovieSceneScriptingFloatChannel, const [1,1]}
          Shear.X {type=MovieSceneScriptingFloatChannel, const [0,1]}
          Shear.Y {type=MovieSceneScriptingFloatChannel, const [0,1]}
         ]
        }]
      }
      {
       type=MovieSceneWidgetMaterialTrack
       sections [{type=MovieSceneParameterSection, channels [Anim_Color {type=MovieSceneScriptingFloatChannel, keys [[0,0],[4,0],[8,1]]}]}]
      }
     ]
    }
    Text_ItemName {
     id=#4
     possessed_class=CommonTextBlock
     tracks [
      {
       type=MovieScene2DTransformTrack
       sections [{
         type=MovieScene2DTransformSection
         end_frame=80
         start_frame=0
         channels [
          Translation.X {type=MovieSceneScriptingFloatChannel, keys [[1,60],[8,0]]}
          Translation.Y {type=MovieSceneScriptingFloatChannel, default=0}
          Angle {type=MovieSceneScriptingFloatChannel, default=0}
          Scale.X {type=MovieSceneScriptingFloatChannel, default=1}
          Scale.Y {type=MovieSceneScriptingFloatChannel, default=1}
          Shear.X {type=MovieSceneScriptingFloatChannel, default=0}
          Shear.Y {type=MovieSceneScriptingFloatChannel, default=0}
         ]
        }]
      }
      {
       type=MovieSceneFloatTrack
       display_name=Render Opacity
       sections [{
         type=MovieSceneFloatSection
         end_frame=80
         start_frame=0
         channels [None {type=MovieSceneScriptingFloatChannel, keys [[1,0],[8,1],[40,1],[80,0.5]]}]
        }]
      }
     ]
    }
    HBox_AmmoExtensionWidgets {
     id=#5
     possessed_class=HorizontalBox
     tracks [
      {
       type=MovieSceneFloatTrack
       display_name=Render Opacity
       sections [{type=MovieSceneFloatSection, end_frame=80, start_frame=0, channels [None {type=MovieSceneScriptingFloatChannel, keys [[1,0],[6,1]]}]}]
      }
      {
       type=MovieScene2DTransformTrack
       sections [{
         type=MovieScene2DTransformSection
         end_frame=80
         start_frame=0
         channels [
          Translation.X {type=MovieSceneScriptingFloatChannel, keys [[1,45],[6,0]]}
          Translation.Y {type=MovieSceneScriptingFloatChannel, default=0}
          Angle {type=MovieSceneScriptingFloatChannel, default=0}
          Scale.X {type=MovieSceneScriptingFloatChannel, default=1}
          Scale.Y {type=MovieSceneScriptingFloatChannel, default=1}
          Shear.X {type=MovieSceneScriptingFloatChannel, default=0}
          Shear.Y {type=MovieSceneScriptingFloatChannel, default=0}
         ]
        }]
      }
     ]
    }
   ]}
 }
 Anim_FadeOut {
  binding_count=1
  length_seconds=0.0583
  object=&self:Anim_FadeOut
  sequence {
   bindings [HBox_Container {
     id=#6
     possessed_class=HorizontalBox
     tracks [
      {
       type=MovieScene2DTransformTrack
       sections [{
         type=MovieScene2DTransformSection
         end_frame=7
         start_frame=0
         channels [
          Translation.X {type=MovieSceneScriptingFloatChannel, keys [[0,0],[7,4]]}
          Translation.Y {type=MovieSceneScriptingFloatChannel, default=0}
          Angle {type=MovieSceneScriptingFloatChannel, default=0}
          Scale.X {type=MovieSceneScriptingFloatChannel, keys [[0,1],[7,0.95]]}
          Scale.Y {type=MovieSceneScriptingFloatChannel, keys [[0,1],[7,0.95]]}
          Shear.X {type=MovieSceneScriptingFloatChannel, default=0}
          Shear.Y {type=MovieSceneScriptingFloatChannel, default=0}
         ]
        }]
      }
      {
       type=MovieSceneFloatTrack
       display_name=Render Opacity
       sections [{type=MovieSceneFloatSection, end_frame=7, start_frame=0, channels [None {type=MovieSceneScriptingFloatChannel, keys [[0,1],[7,0]]}]}]
      }
     ]
    }]
   display_rate {numerator=120}
  }
 }
]
graphs [
 AddContentWidget {nodes [
   n1 = AddContentWidget!() -> (then, WidgetSlotTag:Gameplay Tag Structure, ContentWidget:Widget Object Reference) {is=title K2Node_FunctionEntry}
   > n2 = Branch(Condition=MatchesTag!(
     TagOne:Gameplay Tag Structure=@n1
     TagTwo:Gameplay Tag Structure=Get ItemInfoSlotTag() -> (ItemInfoSlotTag:Gameplay Tag Structure)
     bExactMatch:bool=true
    ) -> (ReturnValue:bool))
   > n3 = AddChildToHorizontalBox!(
    execute
    self:Horizontal Box Object Reference=Get HBox_ExtensionWidgets() -> (HBox_ExtensionWidgets:Horizontal Box Object Reference)
    Content:Widget Object Reference=Get ContentWidget() -> (ContentWidget:Widget Object Reference)
   ) -> (then, ReturnValue:Horizontal Box Slot Object Reference)
   > SetHorizontalAlignment(self:Horizontal Box Slot Object Reference=@n3, InHorizontalAlignment=HAlign_Fill)
   > SetVerticalAlignment(self:Horizontal Box Slot Object Reference=@n3, InVerticalAlignment=VAlign_Fill)
   Branch(
    execute=@n2.else
    Condition=MatchesTag(
     TagOne=Get WidgetSlotTag() -> (WidgetSlotTag:Gameplay Tag Structure)
     TagTwo=Get ItemAmmInfoSlotTag() -> (ItemAmmInfoSlotTag:Gameplay Tag Structure)
     bExactMatch=true
    )
   )
   > n4 = AddChildToHorizontalBox(self=Get HBox_AmmoExtensionWidgets() -> (HBox_AmmoExtensionWidgets:Horizontal Box Object Reference), Content=Get ContentWidget())
   > SetHorizontalAlignment(self=@n4, InHorizontalAlignment=HAlign_Fill)
   > SetVerticalAlignment(self=@n4, InVerticalAlignment=VAlign_Fill)
  ]}
 EventGraph {nodes [
   n1 = OnQuickBarSlotFocusChanged_Bind!() -> (OutputDelegate, then, QuickbarIndex:EFortQuickBars Enum=Primary, Slot:int) {is=title K2Node_CustomEvent}
   > n2 = Branch(Condition="Equal (Enum)"(A:EFortQuickBars Enum=@n1.QuickbarIndex, B:EFortQuickBars Enum=Primary))
   Branch(execute=@n2.else, Condition=Get IsShown() -> (IsShown:bool))
   > Set IsShown(IsShown:bool=false) -> (Output_Get:bool)
   > PlayAnimationForward(InAnimation=Get Anim_FadeOut() -> (Anim_FadeOut:Widget Animation Object Reference))
   n3 = Event Add Content To Host Widget!() -> (
    OutputDelegate
    then
    WidgetHostTag:"Gameplay Tag Structure (by ref)"
    ContentWidget:Widget Object Reference
    Padding:"Margin Structure (by ref)"
    bFill:bool
   ) {is=title K2Node_Event}
   > AddContentWidget!(execute, self, WidgetSlotTag:Gameplay Tag Structure=@n3.WidgetHostTag, ContentWidget:Widget Object Reference=@n3.ContentWidget) -> (then) {is=title K2Node_CallFunction}
   n4 = Event BP_HandleHasBulletsPerClipChanged!() -> (OutputDelegate, then, bHasBulletsPerClip:bool) {is=title K2Node_Event}
   > SetVisibility(
    self=Get Image_AmmoSplit() -> (Image_AmmoSplit:Image Object Reference)
    InVisibility=Select(
     Option 0=Collapsed
     Option 1=Get WeaponEquipVisibility() -> (WeaponEquipVisibility:ESlateVisibility Enum=Visible)
     Index=@n4.bHasBulletsPerClip
    )
   )
   Event Handle Resource Type Changed!() -> (OutputDelegate, then, Item:Fort Resource Item Definition Object Reference, ResourceCount:int=0) {is=title K2Node_Event}
   > Native_UpdateWidgetColor!(execute, self:Athena Equipped Item Widget Base Object Reference, bHasAmmo:bool=true) -> (then)
   Event PreConstruct()
   > SetMobileStyle() {is=title K2Node_CallFunction}
   Event Destruct()
   > UnregisterWidgetSlots()
   n5 = Event Handle Resource Count Changed!() -> (OutputDelegate, then, ResourceCount:int=0) {is=title K2Node_Event}
   > PrintString!(
    execute
    InString=Append(A=Resource Count:, B="To String (Integer)"(InInt=@n5.ResourceCount))
    bPrintToScreen
    bPrintToLog
    TextColor
    Duration
    Key
   ) -> (then)
   Event OnVisibilitySetEvent!() -> (OutputDelegate, then, InVisibility:ESlateVisibility Enum=Visible) {is=title K2Node_Event}
   > ResetIntroAnim()
   > n6 = Cast To FortPlayerController(Object=GetOwningPlayer())
   > DelayUntilNextFrame!(execute) -> (then)
   > n7 = GetItemInQuickbarSlot!(
    execute
    self:Fort Player Controller Object Reference=@n6
    QuickBarType:EFortQuickBars Enum=Primary
    SlotIndex:int=GetFocusedSlot!(self:Fort Player Controller Object Reference=@n6, InQuickbar:EFortQuickBars Enum=Primary) -> (ReturnValue:int=0)
   ) -> (then, ReturnValue:Fort Item Object Reference)
   > Is Valid(InputObject=@n7)
   > "SetText (Text)"(
    self=Get Text_ItemName() -> (Text_ItemName:Common Text Object Reference)
    InText=GetDisplayName!(self:Fort Item Object Reference=@n7) -> (ReturnValue:Text)
   )
   > n8 = GetRarityData!(execute, Rarity:EFortRarity Enum=GetRarity!(self:Fort Item Object Reference=@n7) -> (ReturnValue:EFortRarity Enum=Common)) -> (then, ReturnValue:Fort Rarity Item Data Structure)
   > n9 = GetDynamicMaterial(self=Get Image_BracketRight() -> (Image_BracketRight:Image Object Reference))
   n10 = Break Fort Rarity Item Data!(FortRarityItemData:"Fort Rarity Item Data Structure (by ref)"=@n8) -> (
    Name:Text
    Color1:Linear Color Structure
    Color2:Linear Color Structure
    Color3:Linear Color Structure
    Color4:Linear Color Structure
    Color5:Linear Color Structure
    Radius:float=~
    Falloff:float=~
    Brightness:float=~
    Roughness:float=~
    Glow:float=~
   ) {is=title K2Node_BreakStruct}
   > SetVectorParameterValue(self=@n9, ParameterName=Color, Value=@n10.Color1)
   > SetVectorParameterValue(self=@n9, ParameterName=Color_2, Value=@n10.Color2)
   n11 = Get Text_RarityLabel() -> (Text_RarityLabel:Common Text Object Reference)
   > "SetText (Text)"(self=@n11, InText=@n10)
   > n12 = GetDynamicFontMaterial!(execute, self:Text Object Reference=@n11) -> (then, ReturnValue:Material Instance Dynamic Object Reference)
   > SetVectorParameterValue(self=@n12, ParameterName=Color_1, Value=@n10.Color1)
   > SetVectorParameterValue(self=@n12, ParameterName=Color_2, Value=@n10.Color2)
   > n13 = GetDynamicMaterial(self:Border Object Reference=Get RarityLabel() -> (RarityLabel:Border Object Reference))
   > SetVectorParameterValue(self=@n13, ParameterName=Color, Value=@n10.Color4)
   > PlayAnimation(InAnimation=Get Anim_SwitchItem() -> (Anim_SwitchItem:Widget Animation Object Reference))
   > n14 = Branch(Condition=Get IsShown())
   Set IsShown(execute=@n14.else, IsShown=true)
   > PlayAnimationReverse(InAnimation=Get Anim_FadeOut())
   Event Construct()
   > SetDesiredSizeOverride!(
    execute
    self:Image Object Reference=[
     Get Image_AmmoIcon() -> (Image_AmmoIcon:Image Object Reference)
     Get Image_RechargeableIcon() -> (Image_RechargeableIcon:Image Object Reference)
    ]
    DesiredSize:Vector 2D Structure=Get SmallBrushSize() -> (SmallBrushSize:Vector 2D Structure)
   ) -> (then)
   > Assign On Quickbar Slot Focus Changed!(
    execute:Exec
    self:Fort Inventory Context Object Reference=Get FortInventoryContext!() -> (ReturnValue:Fort Inventory Context Object Reference) {is=title K2Node_GetBlueprintContextEx}
    Delegate:"Delegate (by ref)"=Create Event()
   ) -> (then:Exec) {is=title K2Node_AssignDelegate}
   > RegisterWidgetSlots()
  ]}
 RegisterWidgetSlots {nodes [
   RegisterWidgetSlots() {is=title K2Node_FunctionEntry}
   > RegisterWidgetHost!(
    execute
    WidgetHost:Widget Object Reference
    WidgetHostTag:"Gameplay Tag Structure (by ref)"=Get ItemInfoSlotTag()
    IsPlayerSpecific:bool=true
   ) -> (then)
   > RegisterWidgetHost(WidgetHostTag=Get ItemAmmInfoSlotTag(), IsPlayerSpecific=true)
  ]}
 RemoveContentFromWidgetHost {nodes [
   n1 = Remove Content From Host Widget!() -> (
    then
    WidgetHostTag:"Gameplay Tag Structure (by ref)"
    ContentWidget:Widget Object Reference
    RemovalType:EWidgetSlottingRemovalType Enum=Unparent
   ) {is=title K2Node_FunctionEntry}
   > n2 = Branch(Condition="Equal (GameplayTag)"(A=@n1, B=Get ItemInfoSlotTag()))
   > n3 = RemoveChild(self=Get HBox_ExtensionWidgets(), Content=Get ContentWidget())
   > Return Node!(execute, ReturnValue:bool=@n3)
   n4 = Branch(
    execute=@n2.else
    Condition=MatchesTag(TagOne=Get WidgetHostTag() -> (WidgetHostTag:"Gameplay Tag Structure (by ref)"), TagTwo=Get ItemAmmInfoSlotTag(), bExactMatch=true)
   )
   > n5 = RemoveChild(self=Get HBox_AmmoExtensionWidgets(), Content=Get ContentWidget())
   > Return Node(ReturnValue=@n5)
   Return Node(execute=@n4.else, ReturnValue=false)
  ]}
 ResetIntroAnim {nodes [
   ResetIntroAnim() {is=title K2Node_FunctionEntry}
   n1 = Get Anim_SwitchItem()
   > PlayAnimation(InAnimation=@n1, StartAtTime=GetEndTime(self=@n1), PlayMode=Reverse)
  ]}
 SetMobileStyle {nodes [
   SetMobileStyle() {is=title K2Node_FunctionEntry}
   n1 = GetOwningPlayer()
   > Branch(Condition=AND Boolean(
     A=IsUsingCompactUI()
     B=IsUsingInputType!(InputType:ECommonInputType Enum=Touch, PlayerController:Player Controller Object Reference=@n1) -> (ReturnValue:bool)
    ))
   n2 = Get SizeBox_ItemName() -> (SizeBox_ItemName:Size Box Object Reference)
   > ClearMinDesiredWidth(self:Size Box Object Reference=@n2)
   > SetMaxDesiredHeight!(execute, self:Size Box Object Reference=@n2, InMaxDesiredHeight:float=40) -> (then)
   > SetMinDesiredWidth!(execute, self:Size Box Object Reference=@n2, InMinDesiredWidth:float=245) -> (then)
   > SetMaxDesiredWidth!(execute, self:Size Box Object Reference=@n2, InMaxDesiredWidth:float=245) -> (then)
   > SetMinDesiredHeight!(execute, self:Size Box Object Reference=@n2, InMinDesiredHeight:float=40) -> (then)
   n3 = Get Text_ItemName()
   > SetWrapTextWidth!(execute, self:Common Text Object Reference=@n3, InWrapTextAt:int=245) -> (then)
   > SetLineHeightPercentage!(execute, self:Common Text Object Reference=@n3, InLineHeightPercentage:float=0.8) -> (then)
   > SetTextOverflowPolicy!(execute, self:Text Object Reference=@n3, InOverflowPolicy:ETextOverflowPolicy Enum=Ellipsis) -> (then)
   n4 = Get RarityLabel()
   > AddChild(
    self=Get CommonHardwareVisibility_RarityTouch() -> (CommonHardwareVisibility_RarityTouch:Common Hardware Visibility Border Object Reference)
    Content=@n4
   )
   > SetPadding!(execute, self:Border Object Reference=@n4, InPadding_Left:float=0, InPadding_Top:float=0, InPadding_Right:float=0, InPadding_Bottom:float=0) -> (then)
   > SetPadding(
    self:Border Slot Object Reference=SlotAsBorderSlot!(Widget:Widget Object Reference=Get Text_RarityLabel()) -> (ReturnValue:Border Slot Object Reference)
    InPadding_Left=5
    InPadding_Top=2
    InPadding_Right=5
    InPadding_Bottom=2
   )
   > SetPadding!(
    execute
    self:Size Box Slot Object Reference=SlotAsSizeBoxSlot!(Widget:Widget Object Reference=Get Bracket() -> (Bracket:Size Box Object Reference)) -> (ReturnValue:Size Box Slot Object Reference)
    InPadding:Margin Structure
   ) -> (then)
   n5 = Get Text_ItemName()
   n6 = SlotAsScaleBoxSlot!(Widget:Widget Object Reference=@n5) -> (ReturnValue:Scale Box Slot Object Reference)
   > SetHorizontalAlignment(self:Scale Box Slot Object Reference=@n6, InHorizontalAlignment=HAlign_Right)
   > Set Justification!(execute, Justification:ETextJustify Enum=Right, self:Text Layout Widget Object Reference=@n5) -> (then, Output_Get:ETextJustify Enum=Left)
   > SetVerticalAlignment(self:Scale Box Slot Object Reference=@n6, InVerticalAlignment=VAlign_Fill)
  ]}
 UnregisterWidgetSlots {nodes [
   UnregisterWidgetSlots() {is=title K2Node_FunctionEntry}
   > UnregisterWidgetHost!(execute, WidgetHost:Widget Object Reference, WidgetHostTag:"Gameplay Tag Structure (by ref)"=Get ItemInfoSlotTag()) -> (then)
   > UnregisterWidgetHost(WidgetHostTag=Get ItemAmmInfoSlotTag())
  ]}
]
variables [
 HasHideItemInfoTag {type=boolean}
 IsShown {type=boolean, default=true}
 ItemAmmInfoSlotTag {type=GameplayTag, default {tagName=DynamicUI.HUD.Fort.Athena.EquippedItemAmmoInfo.Slot}}
 ItemInfoSlotTag {type=GameplayTag, default {tagName=DynamicUI.HUD.Fort.Athena.EquippedItemInfo.Slot}}
 LargeBrushSize {type=Vector2D, default=<40 40>}
 SmallBrushSize {type=Vector2D, default=<24 24>}
]
widgets {
 info {
  inheritedWidgetCount=24
  namedSlotCount=1
  parentClass=/Script/FortniteUI.AthenaHideableAmmoEquippedItemWidget
  rootWidgetClass=&HorizontalBox
  widgetCount=42
 }
 widgets [
  HBox_Container {
   widgetClassPath=&HorizontalBox
   bIsVariable=true
   widget=HBox_Container
   properties {pixelSnapping=Disabled, renderTransformPivot=<1 0.5>}
  }
  Overlay_0 {
   widgetClassPath=&Overlay
   parent=HBox_Container
   widget=Overlay_0
   slot {class=HorizontalBoxSlot, properties {horizontalAlignment=HAlign_Right, verticalAlignment=VAlign_Center, padding=<ltrb 0 4 0 4>}}
  }
  VerticalBox_0 {
   widgetClassPath=&VerticalBox
   parent=Overlay_0
   widget=VerticalBox_0
   slot {class=OverlaySlot, properties {horizontalAlignment=HAlign_Right, verticalAlignment=VAlign_Center}}
  }
  CommonHardwareVisibility_RarityTouch {
   widgetClassPath=&CommonHardwareVisibilityBorder
   bIsVariable=true
   parent=VerticalBox_0
   widget=CommonHardwareVisibility_RarityTouch
   properties {visibilityQuery {
     autoDescription=" ANY( Input.Touch )"
     tokenStreamVersion=0
     userDescription=Touch only
     queryTokenStream [0,1,1,1,0]
     tagDictionary [{tagName=Input.Touch}]
    }}
   slot {class=VerticalBoxSlot, properties {horizontalAlignment=HAlign_Right, verticalAlignment=VAlign_Center}}
  }
  HorizontalBox_3 {
   widgetClassPath=&HorizontalBox
   parent=VerticalBox_0
   widget=HorizontalBox_3
   properties {renderTransformPivot=<1 0.5>}
   slot {class=VerticalBoxSlot, properties {horizontalAlignment=HAlign_Right, verticalAlignment=VAlign_Center}}
  }
  FortHUDElementWrapper {
   widgetClassPath=/Game/Athena/HUD/FortHUDElementWrapper
   bIsVariable=true
   parent=HorizontalBox_3
   widget=FortHUDElementWrapper
   properties {elementTag {tagName=HUD.Athena.EquippedAmmoIndicator}, hUDElementTag {gameplayTags [{tagName=HUD.Athena.Hide.EquippedAmmoIndicator}]}}
   slot {class=HorizontalBoxSlot}
  }
  Overlay_28 {widgetClassPath=&Overlay, namedSlotHost=FortHUDElementWrapper, widget=Overlay_28}
  Switcher_Content {
   widgetClassPath=&CommonVisibilitySwitcher
   bInherited=true
   parent=Overlay_28
   widget=Switcher_Content
   properties {activeWidgetIndex=1}
   slot {class=OverlaySlot, properties {horizontalAlignment=HAlign_Fill, verticalAlignment=VAlign_Center, padding=<ltrb 0 -4 0 -2>}}
  }
  Text_ResourceCount {
   widgetClassPath=&CommonTextBlock
   bInherited=true
   bIsVariable=true
   parent=Switcher_Content
   widget=Text_ResourceCount
   properties {
    lineHeightPercentage=0.9
    text=999
    font {
     bForceMonospaced=false
     bMaterialIsStencil=false
     fontMaterial=/BattleRoyaleGameUI/HUD/PlayerHitPointInfo/Assets/M_Font_PlayerVitals_Value
     fontObject=&Burbank
     letterSpacing=0
     monospacedWidth=1
     size=33
     skewAmount=0
     typefaceFontName=BurbankRegular-Black
     outlineSettings {
      bApplyOutlineToDropShadows=true
      bMiteredCorners=false
      bSeparateFillAlpha=true
      outlineMaterial=&M-FontShadow
      outlineSize=2
      outlineColor=<rgba 0 0 0 0.25>
     }
    }
    margin=<ltrb 0 -8 0 -2>
    shadowOffset=<0 0>
   }
   slot {class=CommonVisibilitySwitcherSlot, properties {horizontalAlignment=HAlign_Right, verticalAlignment=VAlign_Center}}
  }
  Panel_AmmoContent {
   widgetClassPath=&HorizontalBox
   bInherited=true
   parent=Switcher_Content
   widget=Panel_AmmoContent
   slot {class=CommonVisibilitySwitcherSlot, properties {horizontalAlignment=HAlign_Right, verticalAlignment=VAlign_Center}}
  }
  Text_MagazineAmmoCount {
   widgetClassPath=&CommonTextBlock
   bInherited=true
   bIsVariable=true
   parent=Panel_AmmoContent
   widget=Text_MagazineAmmoCount
   properties {
    lineHeightPercentage=0.9
    text=0
    font {
     bForceMonospaced=false
     bMaterialIsStencil=false
     fontMaterial=/BattleRoyaleGameUI/Shared/M_UI_Font_White
     fontObject=&Burbank
     letterSpacing=0
     monospacedWidth=1
     size=33
     skewAmount=0
     typefaceFontName=BurbankRegular-Black
     outlineSettings {
      bApplyOutlineToDropShadows=true
      bMiteredCorners=false
      bSeparateFillAlpha=true
      outlineMaterial=/BattleRoyaleGameUI/Shared/MI_UI_Font_SoftShadow_LargeText
      outlineSize=12
      outlineColor=<rgba 1 1 1 1>
     }
    }
    margin=<ltrb -10 -16 -10 -12>
    shadowOffset=<0 0>
   }
   slot {
    class=HorizontalBoxSlot
    properties {horizontalAlignment=HAlign_Right, verticalAlignment=VAlign_Center, padding=<ltrb 0 0 0 2>, size {sizeRule=Automatic, value=0.75}}
   }
  }
  Image_AmmoSplit {
   widgetClassPath=&Image
   bInherited=true
   bIsVariable=true
   parent=Panel_AmmoContent
   widget=Image_AmmoSplit
   properties {
    visibility=SelfHitTestInvisible
    brush {
     bIsDynamicallyLoaded=false
     drawAs=Image
     imageType=FullColor
     mirroring=NoMirror
     resourceName=None
     resourceObject=/BattleRoyaleGameUI/Shared/FrameDecor/M_UI_SimpleDivider_Vertical
     tiling=NoTile
     imageSize=<12 40>
     margin=<ltrb 0 0 0 0>
     outlineSettings {
      bUseBrushTransparency=false
      roundingType=HalfHeightRadius
      width=0
      color {colorUseRule=UseColor_Specified, specifiedColor=<rgba 0 0 0 0>}
      cornerRadii=<0 0 0 0>
     }
     tintColor {colorUseRule=UseColor_Specified, specifiedColor=<rgba 1 1 1 1>}
     uVRegion {bIsValid=false, max=<0 0>, min=<0 0>}
    }
    renderTransform {angle=0, scale=<1 1>, shear=<0 0>, translation=<0 3>}
   }
   slot {class=HorizontalBoxSlot, properties {verticalAlignment=VAlign_Center, padding=<ltrb 2 0 2 0>}}
  }
  Text_SurplusAmmoCount {
   widgetClassPath=&CommonTextBlock
   bInherited=true
   bIsVariable=true
   parent=Panel_AmmoContent
   widget=Text_SurplusAmmoCount
   properties {
    mobileFontSizeMultiplier=1.2
    text=0
    font {
     bForceMonospaced=false
     bMaterialIsStencil=false
     fontMaterial=&M_UI_Font_White
     fontObject=&Burbank
     letterSpacing=0
     monospacedWidth=1
     size=16.5
     skewAmount=0
     typefaceFontName=Burbank-PlayerPower
     outlineSettings {
      bApplyOutlineToDropShadows=true
      bMiteredCorners=false
      bSeparateFillAlpha=true
      outlineMaterial=/BattleRoyaleGameUI/Shared/MI_UI_Font_SoftShadow_SmallText
      outlineSize=4
      outlineColor=<rgba 1 1 1 1>
     }
    }
    margin=<ltrb -2 -4 -2 -4>
    renderTransform {angle=0, scale=<1 1>, shear=<0 0>, translation=<0 3>}
    shadowOffset=<0 0>
   }
   slot {class=HorizontalBoxSlot, properties {horizontalAlignment=HAlign_Left, verticalAlignment=VAlign_Center}}
  }
  Image_AmmoIcon {
   widgetClassPath=&Image
   bInherited=true
   bIsVariable=true
   parent=Panel_AmmoContent
   widget=Image_AmmoIcon
   properties {
    visibility=Collapsed
    brush {
     bIsDynamicallyLoaded=false
     drawAs=Image
     imageType=NoImage
     mirroring=NoMirror
     resourceName=None
     resourceObject=/Game/UI/Foundation/Textures/Icons/Ammo/T-Icon-Ammo-Medium-64
     tiling=NoTile
     imageSize=<24 24>
     margin=<ltrb 0 0 0 0>
     outlineSettings {
      bUseBrushTransparency=false
      roundingType=HalfHeightRadius
      width=0
      color {colorUseRule=UseColor_Specified, specifiedColor=<rgba 0 0 0 0>}
      cornerRadii=<0 0 0 0>
     }
     tintColor {colorUseRule=UseColor_Specified, specifiedColor=<rgba 1 1 1 1>}
     uVRegion {bIsValid=false, max=<0 0>, min=<0 0>}
    }
   }
   slot {class=HorizontalBoxSlot, properties {horizontalAlignment=HAlign_Left, verticalAlignment=VAlign_Center, padding=<ltrb 4 0 0 0>}}
  }
  Panel_UtilityContent {
   widgetClassPath=&HorizontalBox
   bInherited=true
   parent=Switcher_Content
   widget=Panel_UtilityContent
   slot {class=CommonVisibilitySwitcherSlot, properties {horizontalAlignment=HAlign_Right, verticalAlignment=VAlign_Center}}
  }
  Text_UtilityAmmoCount {
   widgetClassPath=&CommonTextBlock
   bInherited=true
   bIsVariable=true
   parent=Panel_UtilityContent
   widget=Text_UtilityAmmoCount
   properties {
    lineHeightPercentage=0.9
    mobileFontSizeMultiplier=0.8
    text=0%
    font {
     bForceMonospaced=false
     bMaterialIsStencil=false
     fontMaterial=&M_Font_PlayerVitals_Value
     fontObject=&Burbank
     letterSpacing=0
     monospacedWidth=1
     size=33
     skewAmount=0
     typefaceFontName=BurbankRegular-Black
     outlineSettings {
      bApplyOutlineToDropShadows=true
      bMiteredCorners=false
      bSeparateFillAlpha=true
      outlineMaterial=&MI_UI_Font_SoftShadow_LargeText
      outlineSize=12
      outlineColor=<rgba 1 1 1 1>
     }
    }
    margin=<ltrb -10 -16 -10 -12>
    shadowOffset=<0 0>
   }
   slot {
    class=HorizontalBoxSlot
    properties {horizontalAlignment=HAlign_Right, verticalAlignment=VAlign_Center, size {sizeRule=Automatic, value=0.75}}
   }
  }
  Image_UtilitySplit {
   widgetClassPath=&Image
   bInherited=true
   bIsVariable=true
   parent=Panel_UtilityContent
   widget=Image_UtilitySplit
   properties {
    visibility=HitTestInvisible
    brush {
     bIsDynamicallyLoaded=false
     drawAs=Image
     imageType=FullColor
     mirroring=NoMirror
     resourceName=None
     resourceObject=&M_UI_SimpleDivider_Vertical
     tiling=NoTile
     imageSize=<12 40>
     margin=<ltrb 0 0 0 0>
     outlineSettings {
      bUseBrushTransparency=false
      roundingType=HalfHeightRadius
      width=0
      color {colorUseRule=UseColor_Specified, specifiedColor=<rgba 0 0 0 0>}
      cornerRadii=<0 0 0 0>
     }
     tintColor {colorUseRule=UseColor_Specified, specifiedColor=<rgba 1 1 1 1>}
     uVRegion {bIsValid=false, max=<0 0>, min=<0 0>}
    }
    renderTransform {angle=0, scale=<1 1>, shear=<0 0>, translation=<0 3>}
   }
   slot {class=HorizontalBoxSlot, properties {verticalAlignment=VAlign_Center, padding=<ltrb 4 0 4 0>}}
  }
  Text_SurplusUtilityAmmoCount {
   widgetClassPath=&CommonTextBlock
   bInherited=true
   bIsVariable=true
   parent=Panel_UtilityContent
   widget=Text_SurplusUtilityAmmoCount
   properties {
    mobileFontSizeMultiplier=1.2
    text=0
    visibility=HitTestInvisible
    font {
     bForceMonospaced=false
     bMaterialIsStencil=false
     fontMaterial=&M_Font_PlayerVitals_Value
     fontObject=&Burbank
     letterSpacing=0
     monospacedWidth=1
     size=16.5
     skewAmount=0
     typefaceFontName=Burbank-PlayerPower
     outlineSettings {
      bApplyOutlineToDropShadows=true
      bMiteredCorners=false
      bSeparateFillAlpha=true
      outlineMaterial=&MI_UI_Font_SoftShadow_SmallText
      outlineSize=6
      outlineColor=<rgba 1 1 1 1>
     }
    }
    margin=<ltrb -4 0 -4 0>
    renderTransform {angle=0, scale=<1 1>, shear=<0 0>, translation=<0 3>}
    shadowOffset=<0 0>
   }
   slot {class=HorizontalBoxSlot, properties {horizontalAlignment=HAlign_Left, verticalAlignment=VAlign_Center}}
  }
  Text_OutOfAmmo {
   widgetClassPath=&CommonTextBlock
   bInherited=true
   bIsVariable=true
   parent=Overlay_28
   widget=Text_OutOfAmmo
   properties {
    mobileFontSizeMultiplier=1.3
    text=0
    visibility=Collapsed
    font {
     bForceMonospaced=false
     bMaterialIsStencil=false
     fontMaterial=/BattleRoyaleGameUI/HUD/QuickBar/Assets/MI_UI_ItemDetailsNoAmmo_Font
     fontObject=&Burbank
     letterSpacing=0
     monospacedWidth=1
     size=18
     skewAmount=0
     typefaceFontName=BurbankRegular-Black
     outlineSettings {
      bApplyOutlineToDropShadows=true
      bMiteredCorners=false
      bSeparateFillAlpha=true
      outlineMaterial=/BattleRoyaleGameUI/Shared/MI_UI_Font_SoftShadow_LargeText_CriticalGlow
      outlineSize=16
      outlineColor=<rgba 1 1 1 1>
     }
    }
    margin=<ltrb 0 -12 -16 -12>
    shadowOffset=<0 0>
   }
   slot {class=OverlaySlot, properties {horizontalAlignment=HAlign_Right, verticalAlignment=VAlign_Bottom}}
  }
  TutorialHighlight {
   widgetClassPath=&TutorialHighlight
   bIsVariable=true
   parent=Overlay_28
   widget=TutorialHighlight
   properties {
    customHighlightWidget=/Game/Athena/HUD/Tutorial/Tutorial_Outline_Square_Static.Tutorial_Outline_Square_Static_C
    visibility=SelfHitTestInvisible
   }
   slot {class=OverlaySlot, properties {horizontalAlignment=HAlign_Fill, verticalAlignment=VAlign_Fill}}
  }
  HBox_AmmoExtensionWidgets {
   widgetClassPath=&HorizontalBox
   bIsVariable=true
   parent=HorizontalBox_3
   widget=HBox_AmmoExtensionWidgets
   slot {class=HorizontalBoxSlot, properties {horizontalAlignment=HAlign_Right, padding=<ltrb 4 0 0 0>}}
  }
  Image_RechargeableIcon {
   widgetClassPath=&FortLazyImage
   bInherited=true
   bIsVariable=true
   parent=HBox_AmmoExtensionWidgets
   widget=Image_RechargeableIcon
   properties {
    bIsVolatile=true
    visibility=Collapsed
    brush {
     bIsDynamicallyLoaded=false
     drawAs=Image
     imageType=NoImage
     mirroring=NoMirror
     resourceName=None
     resourceObject=/Game/UI/Foundation/Textures/Icons/Ammo/T-Icon-RechargeAmmo-64
     tiling=NoTile
     imageSize=<32 32>
     margin=<ltrb 0 0 0 0>
     outlineSettings {
      bUseBrushTransparency=false
      roundingType=HalfHeightRadius
      width=0
      color {colorUseRule=UseColor_Specified, specifiedColor=<rgba 0 0 0 0>}
      cornerRadii=<0 0 0 0>
     }
     tintColor {colorUseRule=UseColor_Specified, specifiedColor=<rgba 1 1 1 1>}
     uVRegion {bIsValid=false, max=<0 0>, min=<0 0>}
    }
    loadingBackgroundBrush {
     bIsDynamicallyLoaded=false
     drawAs=NoDrawType
     imageType=NoImage
     mirroring=NoMirror
     resourceName=None
     resourceObject=null
     tiling=NoTile
     imageSize=<48 48>
     margin=<ltrb 0 0 0 0>
     outlineSettings {
      bUseBrushTransparency=false
      roundingType=HalfHeightRadius
      width=0
      color {colorUseRule=UseColor_Specified, specifiedColor=<rgba 0 0 0 0>}
      cornerRadii=<0 0 0 0>
     }
     tintColor {colorUseRule=UseColor_Specified, specifiedColor=<rgba 1 1 1 1>}
     uVRegion {bIsValid=false, max=<0 0>, min=<0 0>}
    }
   }
   slot {class=HorizontalBoxSlot, properties {horizontalAlignment=HAlign_Left, verticalAlignment=VAlign_Center, padding=<ltrb 4 0 4 0>}}
  }
  VBox_ItemInfo {
   widgetClassPath=&VerticalBox
   bIsVariable=true
   parent=VerticalBox_0
   widget=VBox_ItemInfo
   properties {visibility=HitTestInvisible}
   slot {class=VerticalBoxSlot, properties {horizontalAlignment=HAlign_Right, verticalAlignment=VAlign_Center}}
  }
  AthenaEquippedItemDurabilityWidget {
   widgetClassPath=/Game/Athena/HUD/AthenaEquippedItemDurabilityWidget
   parent=VBox_ItemInfo
   widget=AthenaEquippedItemDurabilityWidget
   properties {visibility=HitTestInvisible}
   slot {class=VerticalBoxSlot, properties {horizontalAlignment=HAlign_Right}}
  }
  HorizontalBox_2 {
   widgetClassPath=&HorizontalBox
   parent=VBox_ItemInfo
   widget=HorizontalBox_2
   properties {clipping=ClipToBounds}
   slot {class=VerticalBoxSlot}
  }
  RarityLabel {
   widgetClassPath=&Border
   bIsVariable=true
   parent=HorizontalBox_2
   widget=RarityLabel
   properties {
    clipping=ClipToBounds
    visibility=SelfHitTestInvisible
    background {
     bIsDynamicallyLoaded=false
     drawAs=Image
     imageType=FullColor
     mirroring=NoMirror
     resourceName=None
     resourceObject=/BattleRoyaleGameUI/HUD/QuickBar/Assets/M_UI_RarityLabel_Backing
     tiling=NoTile
     imageSize=<32 32>
     margin=<ltrb 0 0 0 0>
     outlineSettings {
      bUseBrushTransparency=false
      roundingType=HalfHeightRadius
      width=0
      color {colorUseRule=UseColor_Specified, specifiedColor=<rgba 0 0 0 0>}
      cornerRadii=<0 0 0 0>
     }
     tintColor {colorUseRule=UseColor_Specified, specifiedColor=<rgba 1 1 1 1>}
     uVRegion {bIsValid=false, max=<0 0>, min=<0 0>}
    }
    padding=<ltrb 3 1 3 1>
   }
   slot {class=HorizontalBoxSlot, properties {horizontalAlignment=HAlign_Right, verticalAlignment=VAlign_Center, padding=<ltrb 0 0 2 0>}}
  }
  Text_RarityLabel {
   widgetClassPath=&CommonTextBlock
   bIsVariable=true
   parent=RarityLabel
   widget=Text_RarityLabel
   properties {
    justification=Right
    mobileFontSizeMultiplier=1.35
    text="{Rarity}"
    textTransformPolicy=ToUpper
    font {
     bForceMonospaced=false
     bMaterialIsStencil=false
     fontMaterial=/BattleRoyaleGameUI/HUD/QuickBar/Assets/M_UI_RarityTag_Font
     fontObject=&BurbankSmall
     letterSpacing=0
     monospacedWidth=1
     size=9
     skewAmount=0
     typefaceFontName=Black
     outlineSettings {
      bApplyOutlineToDropShadows=false
      bMiteredCorners=false
      bSeparateFillAlpha=false
      outlineMaterial=null
      outlineSize=0
      outlineColor=<rgba 0 0 0 1>
     }
    }
    shadowOffset=<0 0>
   }
   slot {class=BorderSlot, properties {padding=<ltrb 3 1 3 1>}}
  }
  SizeBox_ItemName {
   widgetClassPath=&SizeBox
   bIsVariable=true
   parent=HorizontalBox_2
   widget=SizeBox_ItemName
   properties {maxDesiredWidth=250}
   slot {class=HorizontalBoxSlot, properties {horizontalAlignment=HAlign_Right}}
  }
  Text_ItemName {
   widgetClassPath=&CommonTextBlock
   bIsVariable=true
   parent=SizeBox_ItemName
   widget=Text_ItemName
   properties {
    bIsScrollingEnabled=false
    justification=Right
    mobileFontSizeMultiplier=1.3
    text="[PH] Item Name"
    textTransformPolicy=ToUpper
    font {
     bForceMonospaced=false
     bMaterialIsStencil=false
     fontMaterial=/BattleRoyaleGameUI/HUD/QuickBar/Assets/M_UI_ItemDetails_Font
     fontObject=&Burbank
     letterSpacing=0
     monospacedWidth=1
     size=11.25
     skewAmount=0
     typefaceFontName=BurbankRegular-Black
     outlineSettings {
      bApplyOutlineToDropShadows=true
      bMiteredCorners=false
      bSeparateFillAlpha=true
      outlineMaterial=&MI_UI_Font_SoftShadow_SmallText
      outlineSize=3
      outlineColor=<rgba 1 1 1 1>
     }
    }
    margin=<ltrb -1 -4 0 -4>
    shadowOffset=<0 0>
   }
   slot {class=SizeBoxSlot}
  }
  HBox_ExtensionWidgets {
   widgetClassPath=&HorizontalBox
   bIsVariable=true
   parent=VBox_ItemInfo
   widget=HBox_ExtensionWidgets
   properties {visibility=HitTestInvisible, renderTransform {angle=0, scale=<1 1>, shear=<0 0>, translation=<0 4>}}
   slot {class=VerticalBoxSlot, properties {horizontalAlignment=HAlign_Right, verticalAlignment=VAlign_Center}}
  }
  Bracket {
   widgetClassPath=&SizeBox
   bIsVariable=true
   parent=HBox_Container
   widget=Bracket
   properties {bOverride_MinDesiredHeight=true, minDesiredHeight=48}
   slot {class=HorizontalBoxSlot, properties {padding=<ltrb 0 0 2 0>, size {sizeRule=Fill, value=1}}}
  }
  Image_BracketRight {
   widgetClassPath=&Image
   bIsVariable=true
   parent=Bracket
   widget=Image_BracketRight
   properties {brush {
     bIsDynamicallyLoaded=false
     drawAs=Image
     imageType=FullColor
     mirroring=NoMirror
     resourceName=None
     resourceObject=/BattleRoyaleGameUI/Shared/FrameDecor/MI_UI_ItemWidgetInfo_Bracket
     tiling=NoTile
     imageSize=<20 24>
     margin=<ltrb 0 0 0 0>
     outlineSettings {
      bUseBrushTransparency=false
      roundingType=HalfHeightRadius
      width=0
      color {colorUseRule=UseColor_Specified, specifiedColor=<rgba 0 0 0 0>}
      cornerRadii=<0 0 0 0>
     }
     tintColor {colorUseRule=UseColor_Specified, specifiedColor=<rgba 1 1 1 1>}
     uVRegion {bIsValid=false, max=<0 0>, min=<0 0>}
    }}
   slot {class=SizeBoxSlot}
  }
  SwitchPopup {widgetClassPath=&VerticalBox, bInherited=true, widget=null}
  SwitchPopupText {widgetClassPath=&CommonTextBlock, bInherited=true, widget=null}
  SimpleAnimations {widgetClassPath=/Script/FortniteUI.FortSimpleWidgetAnimationsPanel, bInherited=true, widget=null}
  Panel_ResourceContent {widgetClassPath=/Script/UMG.PanelWidget, bInherited=true, widget=null}
  Image_UtilityIcon {widgetClassPath=&Image, bInherited=true, widget=null}
  Image_ResourceIcon {widgetClassPath=&Image, bInherited=true, widget=null}
  Overlay_UtilityAmmoIcon {widgetClassPath=&Overlay, bInherited=true, widget=null}
  RootScaleBox {widgetClassPath=&ScaleBox, bInherited=true, widget=null}
  Image_Background {widgetClassPath=&Image, bInherited=true, widget=null}
  ScaleBox_HUDWidgetScale {widgetClassPath=&ScaleBox, bInherited=true, widget=null}
 ]
}
