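# Mobile touch HUD element for Battle Royale that displays the shared equipped item widget, named as the touch reload widget.
# Subclasses FortMobileHUDElement and nests WBP_BRGameUI_EquippedItemWidget in a fixed-height, right-aligned SizeBox under the inherited user-scaled, hit-test-invisible ScaleBox, with a small render offset.
# The event graph is empty; variables declare an EquippedItemInfo slot tag and a bow weapon tag container, with no use shown here.
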WidgetBlueprint /BattleRoyaleGameUI/HUD/Touch/WBP_BRGameUI_EquippedItemReloadWidget_Touch
graphs [EventGraph {links [], nodes []}]
variables [
 ItemInfoSlotTag {type=GameplayTag, default {tagName=DynamicUI.HUD.Fort.Athena.EquippedItemInfo.Slot}}
 Tag Container {type=GameplayTagContainer, default {gameplayTags [{tagName=Weapon.Ranged.Bow}]}}
]
widgets {
 info {inheritedWidgetCount=3, parentClass=/Script/FortniteUI.FortMobileHUDElement, rootWidgetClass=&ScaleBox, widgetCount=5}
 widgets [
  ScaleBox_HUDWidgetScale {
   widgetClassPath=&ScaleBox
   bInherited=true
   widget=ScaleBox_HUDWidgetScale
   properties {stretch=UserSpecified, visibility=HitTestInvisible}
  }
  SizeBox_2 {
   widgetClassPath=&SizeBox
   parent=ScaleBox_HUDWidgetScale
   widget=SizeBox_2
   properties {bOverride_HeightOverride=true, heightOverride=105}
   slot {class=ScaleBoxSlot, properties {horizontalAlignment=HAlign_Right, verticalAlignment=VAlign_Fill}}
  }
  WBP_BRGameUI_EquippedItemWidget {
   widgetClassPath=/BattleRoyaleGameUI/HUD/QuickBar/WBP_BRGameUI_EquippedItemWidget
   bIsVariable=true
   parent=SizeBox_2
   widget=WBP_BRGameUI_EquippedItemWidget
   properties {renderTransform {angle=0, scale=<1 1>, shear=<0 0>, translation=<-15 7>}}
   slot {class=SizeBoxSlot}
  }
  Image_Background {widgetClassPath=&Image, bInherited=true, widget=null}
  TutorialHighlight {widgetClassPath=&TutorialHighlight}
 ]
}
