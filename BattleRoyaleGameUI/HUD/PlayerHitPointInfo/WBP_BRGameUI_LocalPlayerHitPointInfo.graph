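# Local player HUD widget stacking shield, health and vehicle vital bars in tag-controlled wrappers, plus a storm bar and extension slots.
# It registers as a HealthMeterTag widget host, placing injected content in an overlay SizeBox.
# With Fade on Full, pawn health changes play FadeStatus when health and shield are full and unfaded, otherwise reverse it.
# It collapses while the Creative quickbar is equipped; PreConstruct, SetSize and FadeVehicle are unwired.
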WidgetBlueprint /BattleRoyaleGameUI/HUD/PlayerHitPointInfo/WBP_BRGameUI_LocalPlayerHitPointInfo
animations [
 FadeStatus {
  binding_count=1
  length_seconds=4
  object=&self:FadeStatus
  sequence {bindings [VerticalBox_23 {
     id=#1
     possessed_class=VerticalBox
     tracks [{
       type=MovieSceneFloatTrack
       display_name=Render Opacity
       sections [{type=MovieSceneFloatSection, channels [None {type=MovieSceneScriptingFloatChannel, keys [[0,1],[40,1],[80,0]]}]}]
      }]
    }]}
 }
 FadeVehicle {
  binding_count=1
  length_seconds=4
  object=&self:FadeVehicle
  sequence {bindings [FortHUDElementWrapper_Vehicle {
     id=#2
     possessed_class=FortHUDElementWrapper_C
     tracks [{
       type=MovieSceneFloatTrack
       display_name=Render Opacity
       sections [{type=MovieSceneFloatSection, channels [None {type=MovieSceneScriptingFloatChannel, keys [[0,1],[40,1],[80,0]]}]}]
      }]
    }]}
 }
]
graphs [
 EventGraph {nodes [
   n1 = OnPawnHealthChanged_Event() {is=title K2Node_CustomEvent}
   n2 = GetPawnFromPlayerStateZone!(
    execute=@n10.then
    InActor:Actor Object Reference=GetPlayerStateZone!(self:Fort HUDContext Object Reference=Get FortHUDContext!() -> (ReturnValue:Fort HUDContext Object Reference) {is=title K2Node_GetBlueprintContextEx}) -> (ReturnValue:Fort Player State Zone Object Reference)
   ) -> (then, OutPawn:Fort Player Pawn Object Reference, PawnReturned:bool)
   n3 = GetMaxHealth!(self:Fort Pawn Object Reference=@n2) -> (ReturnValue:float)
   n4 = GetHealth!(self:Fort Pawn Object Reference=@n2) -> (ReturnValue:float)
   n5 = GetShield!(self:Fort Pawn Object Reference=@n2) -> (ReturnValue:float)
   n6 = GetMaxShield!(self:Fort Pawn Object Reference=@n2) -> (ReturnValue:float)
   n7 = Get isFaded() -> (isFaded:bool)
   n8 = Branch(
    execute=@n1.then
    Condition=AND Boolean!(
     A="Equal (Float)"!(A:double=@n4, B:double=@n3) -> (ReturnValue:bool) {is=title K2Node_CallFunction}
     B="Equal (Float)"(A=@n5, B=@n6)
     C:bool=NOT Boolean(A=@n7)
    ) -> (ReturnValue)
   )
   n9 = Get FadeStatus() -> (FadeStatus:Widget Animation Object Reference)
   > StopAnimation(InAnimation=@n9)
   > PlayAnimation(InAnimation=@n9)
   > Set isFaded(isFaded:bool=true) -> (Output_Get:bool)
   StopAnimation(execute=@n8.else, InAnimation=@n9)
   > PlayAnimationReverse(InAnimation=@n9, PlaybackSpeed=3)
   > Set isFaded(isFaded=false)
   Event Construct()
   > RegisterWidgetHost!(
    execute
    WidgetHost:Widget Object Reference
    WidgetHostTag:"Gameplay Tag Structure (by ref)"=Get HealthMeterTag() -> (HealthMeterTag:Gameplay Tag Structure)
    IsPlayerSpecific:bool=true
   ) -> (then)
   > Bind Event to On Creative Quickbar Equipped Changed(
    self:Creative HUDLContext Object Reference=Get CreativeHUDLContext!() -> (ReturnValue:Creative HUDLContext Object Reference) {is=title K2Node_GetBlueprintContextEx}
    Delegate=Create Event()
   ) {is=title K2Node_AddDelegate}
   > n10 = Branch(Condition=Get Fade on Full() -> (Fade on Full:bool))
   n11 = Event Add Content To Host Widget!() -> (
    OutputDelegate
    then
    WidgetHostTag:"Gameplay Tag Structure (by ref)"
    ContentWidget:Widget Object Reference
    Padding:"Margin Structure (by ref)"
    bFill:bool
   ) {is=title K2Node_Event}
   > Branch(Condition=MatchesTag!(TagOne:Gameplay Tag Structure=@n11.WidgetHostTag, TagTwo:Gameplay Tag Structure=Get HealthMeterTag(), bExactMatch:bool=true) -> (ReturnValue:bool))
   > AddChild(self=Get SizeBox_HealthbarOverlay() -> (SizeBox_HealthbarOverlay:Size Box Object Reference), Content=@n11.ContentWidget)
   Event Destruct()
   > UnregisterWidgetHost!(execute, WidgetHost:Widget Object Reference, WidgetHostTag:"Gameplay Tag Structure (by ref)"=Get HealthMeterTag()) -> (then)
   Assign On Pawn Health Changed!(execute:Exec=@n2.then, self:Fort Pawn Object Reference=@n2, Delegate:"Delegate (by ref)"=@n1) -> (then:Exec) {is=title K2Node_AssignDelegate}
   SetSize!(
    execute
    self:WBP BRGame UI Player Vital Bar Object Reference=[
     Get HitPointBarHealth() -> (HitPointBarHealth:WBP BRGame UI Player Vital Bar Object Reference)
     Get HitPointBarShield() -> (HitPointBarShield:WBP BRGame UI Player Vital Bar Object Reference)
    ]
    useLargeSize:bool=Get Use Large Format() -> (Use Large Format:bool)
   ) -> (then)
   Event PreConstruct()
  ]}
 OnCreativeQuickbarEquippedChanged {nodes [
   n1 = OnCreativeQuickbarEquippedChanged!() -> (then, bIsEquipped:bool=~) {is=title K2Node_FunctionEntry}
   > n2 = Branch(Condition=@n1)
   > SetVisibility(InVisibility=Collapsed)
   SetVisibility(execute=@n2.else, InVisibility=SelfHitTestInvisible)
  ]}
 RemoveContentFromWidgetHost {nodes [
   n1 = Remove Content From Host Widget!() -> (
    then
    WidgetHostTag:"Gameplay Tag Structure (by ref)"
    ContentWidget:Widget Object Reference
    RemovalType:EWidgetSlottingRemovalType Enum=Unparent
   ) {is=title K2Node_FunctionEntry}
   > n2 = Branch(Condition=MatchesTag(TagOne=@n1, TagTwo=Get HealthMeterTag(), bExactMatch=false))
   > n3 = RemoveChild(self=Get SizeBox_HealthbarOverlay(), Content=Get ContentWidget() -> (ContentWidget:Widget Object Reference))
   Return Node!(execute=[@n3.then, @n2.else], ReturnValue:bool=@n3)
  ]}
]
variables [
 Fade on Full {type=boolean}
 HealthMeterTag {type=GameplayTag, default {tagName=DynamicUI.HUD.Fort.Athena.PlayerHitPointInfo.HealthMeter.Slot}}
 inVehicle {type=boolean}
 isFaded {type=boolean}
 RankInfo {type=FortHabaneroDisplayData, default=/Game/UI/Habanero/HabaneroDisplayData}
 Use Large Format {type=boolean}
 VehicleHealthBarHeightOverride {type=number, default=5}
 VehicleHealthBarWidthOverride {type=number, default=500}
]
widgets {
 info {
  inheritedWidgetCount=1
  namedSlotCount=3
  parentClass=/Script/FortniteUI.AthenaPlayerHitPointInfo
  rootWidgetClass=&SizeBox
  widgetCount=18
 }
 widgets [
  SizeBox_Bars {widgetClassPath=&SizeBox, bIsVariable=true, widget=SizeBox_Bars, properties {renderTransformPivot=<0 0.5>}}
  Overlay_0 {widgetClassPath=&Overlay, parent=SizeBox_Bars, widget=Overlay_0, slot {class=SizeBoxSlot}}
  Overlay_1 {
   widgetClassPath=&Overlay
   parent=Overlay_0
   widget=Overlay_1
   slot {class=OverlaySlot, properties {horizontalAlignment=HAlign_Fill, verticalAlignment=VAlign_Bottom}}
  }
  AthenaCurieStatusWidget {
   widgetClassPath=/Game/Athena/HUD/Curie/AthenaCurieStatusWidget
   bIsVariable=true
   parent=Overlay_1
   widget=AthenaCurieStatusWidget
   properties {
    elementTag {tagName=HUD.Athena.Player.Health}
    renderTransform {angle=0, scale=<1 1>, shear=<0 0>, translation=<12 -10>}
    renderTransformPivot=<0 0.8>
   }
   slot {class=OverlaySlot, properties {horizontalAlignment=HAlign_Fill, verticalAlignment=VAlign_Fill, padding=<ltrb -70 -100 -60 -30>}}
  }
  Overlay_LocalPlayerExtensions {
   widgetClassPath=/Game/UI/ExtensionWidgets/UnderLocalPlayerInfo
   bIsVariable=true
   parent=Overlay_1
   widget=Overlay_LocalPlayerExtensions
   slot {class=OverlaySlot, properties {horizontalAlignment=HAlign_Center, verticalAlignment=VAlign_Center}}
  }
  VerticalBox_23 {
   widgetClassPath=&VerticalBox
   parent=Overlay_1
   widget=VerticalBox_23
   slot {class=OverlaySlot, properties {horizontalAlignment=HAlign_Fill, verticalAlignment=VAlign_Fill}}
  }
  FortHUDElementWrapper_Vehicle {
   widgetClassPath=/Game/Athena/HUD/FortHUDElementWrapper
   bIsVariable=true
   parent=VerticalBox_23
   widget=FortHUDElementWrapper_Vehicle
   properties {
    visibility=Collapsed
    elementTag {tagName=HUD.Athena.Player.Vehicle}
    hUDElementTag {gameplayTags [{tagName=Athena.HideForCapture}, {tagName=Athena.Creative.HUDController.HideHealth}]}
   }
   slot {class=VerticalBoxSlot, properties {verticalAlignment=VAlign_Top, padding=<ltrb 0 0 0 16>}}
  }
  HitPointBarVehicleHealth {
   widgetClassPath=&/WBP_BRGameUI_PlayerVitalBar
   bIsVariable=true
   namedSlotHost=FortHUDElementWrapper_Vehicle
   widget=HitPointBarVehicleHealth
   properties {
    barHeight=16
    barType=VehicleHealth
    fillGlowSize=24
    flowDirectionPreference=LeftToRight
    visibility=Collapsed
    elementTag {tagName=HUD.BottomLeft}
   }
  }
  FortHUDElementWrapper_Shield {
   widgetClassPath=&FortHUDElementWrapper
   bIsVariable=true
   parent=VerticalBox_23
   widget=FortHUDElementWrapper_Shield
   properties {
    elementTag {tagName=HUD.Athena.Player.Shield}
    hUDElementTag {gameplayTags [{tagName=Athena.HideForCapture}, {tagName=Athena.Creative.HUDController.HideShields}]}
   }
   slot {class=VerticalBoxSlot, properties {horizontalAlignment=HAlign_Left}}
  }
  HitPointBarShield {
   widgetClassPath=&WBP_BRGameUI_PlayerVitalBar
   bIsVariable=true
   namedSlotHost=FortHUDElementWrapper_Shield
   widget=HitPointBarShield
   properties {
    bAllowSeparator=true
    barHeight=16
    barType=Shield
    fillGlowSize=24
    flowDirectionPreference=LeftToRight
    elementTag {tagName=HUD.BottomLeft}
    foregroundColor {colorUseRule=UseColor_Specified, specifiedColor=<rgba 1 0 1 1>}
    textHudElementTags {gameplayTags [{tagName=Athena.Creative.HUDController.HideShieldNumbers}]}
   }
  }
  FortHUDElementWrapper_Health {
   widgetClassPath=&FortHUDElementWrapper
   bIsVariable=true
   parent=VerticalBox_23
   widget=FortHUDElementWrapper_Health
   properties {
    elementTag {tagName=HUD.Athena.Player.Health}
    hUDElementTag {gameplayTags [{tagName=Athena.HideForCapture}, {tagName=Athena.Creative.HUDController.HideHealth}]}
   }
   slot {class=VerticalBoxSlot, properties {horizontalAlignment=HAlign_Left, verticalAlignment=VAlign_Top, padding=<ltrb 0 4 0 0>}}
  }
  HitPointBarHealth {
   widgetClassPath=&WBP_BRGameUI_PlayerVitalBar
   bIsVariable=true
   namedSlotHost=FortHUDElementWrapper_Health
   widget=HitPointBarHealth
   properties {
    bAllowSeparator=true
    bCanAddExtensionWidgets=true
    barHeight=40
    extensionSlotToUse=PlayerHealthbarOverlay
    fillGlowSize=40
    flowDirectionPreference=LeftToRight
    visibility=SelfHitTestInvisible
    elementTag {tagName=HUD.BottomLeft}
    textHudElementTags {gameplayTags [{tagName=Athena.Creative.HUDController.HideHealthNumbers}]}
   }
  }
  SizeBox {
   widgetClassPath=&SizeBox
   parent=VerticalBox_23
   widget=SizeBox
   properties {bOverride_HeightOverride=true, heightOverride=22}
   slot {class=VerticalBoxSlot, properties {verticalAlignment=VAlign_Top, padding=<ltrb 0 3 0 0>}}
  }
  AthenaHitPointBar_CNStorm2 {
   widgetClassPath=/Game/Athena/UI/GAPP/AthenaHitPointBar_CNStorm2
   bIsVariable=true
   parent=SizeBox
   widget=AthenaHitPointBar_CNStorm2
   properties {flowDirectionPreference=LeftToRight, visibility=Collapsed, elementTag {tagName=HUD.BottomLeft}}
   slot {class=SizeBoxSlot}
  }
  CanvasPanel_0 {widgetClassPath=&CanvasPanel, parent=VerticalBox_23, widget=CanvasPanel_0, slot {class=VerticalBoxSlot}}
  Overlay_2 {
   widgetClassPath=&Overlay
   parent=CanvasPanel_0
   widget=Overlay_2
   slot {class=CanvasPanelSlot, properties {layoutData {alignment=<0 0>, anchors {maximum=<1 1>, minimum=<0 0>}, offsets=<ltrb 0 0 0 0>}}}
  }
  TutorialHighlight {
   widgetClassPath=&TutorialHighlight
   bIsVariable=true
   parent=Overlay_2
   widget=TutorialHighlight
   properties {customHighlightWidget=/Game/Athena/HUD/Tutorial/Tutorial_Outline_Square_Static.Tutorial_Outline_Square_Static_C}
   slot {class=OverlaySlot, properties {horizontalAlignment=HAlign_Fill, verticalAlignment=VAlign_Fill}}
  }
  SizeBox_HealthbarOverlay {
   widgetClassPath=&SizeBox
   bIsVariable=true
   parent=Overlay_0
   widget=SizeBox_HealthbarOverlay
   properties {bOverride_HeightOverride=true, bOverride_WidthOverride=true, heightOverride=40}
   slot {class=OverlaySlot, properties {horizontalAlignment=HAlign_Fill, verticalAlignment=VAlign_Bottom, padding=<ltrb 0 0 0 3>}}
  }
 ]
}
