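# Touch-layout battle royale HUD widget that shows the local player's name and hit point info, with a tutorial highlight.
# Playlist extension slots sit behind, above and below the info, and Add Extension Widget appends centered widgets to MainBox.
# SB_Overflow, holding the behind slot, changes padding when the vehicle health bar's visibility changes.
# Async gameplay tag waits on the player controller also translate it down while riding and reset it afterward.
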WidgetBlueprint /BattleRoyaleGameUI/HUD/Touch/WBP_BRGameUI_LocalPlayerHitPointInfo_Touch
graphs [
 Add Extension Widget {nodes [
   n1 = Add Extension Widget!() -> (then, New Widget:User Widget Object Reference) {is=title K2Node_FunctionEntry}
   > n2 = AddChildToVerticalBox(self=Get MainBox() -> (MainBox:Vertical Box Object Reference), Content=@n1)
   > SetHorizontalAlignment(self:Vertical Box Slot Object Reference=@n2, InHorizontalAlignment=HAlign_Center)
   > SetVerticalAlignment(self:Vertical Box Slot Object Reference=@n2, InVerticalAlignment=VAlign_Center)
  ]}
 EventGraph {nodes [
   n1 = AdjustExtensionSlot!() -> (OutputDelegate, then, InVisibility:ESlateVisibility Enum=Visible) {is=title K2Node_CustomEvent}
   > n2 = Switch on ESlateVisibility!(execute, Selection:ESlateVisibility Enum=@n1.InVisibility) -> (Visible:Exec, Collapsed:Exec, Hidden:Exec, HitTestInvisible:Exec, SelfHitTestInvisible:Exec) {is=title K2Node_SwitchEnum}
   SetPadding!(
    execute=[@n2.Collapsed, @n2.Hidden]
    self:Size Box Slot Object Reference=SlotAsSizeBoxSlot!(Widget:Widget Object Reference=Get SB_Overflow() -> (SB_Overflow:Size Box Object Reference)) -> (ReturnValue:Size Box Slot Object Reference)
    InPadding_Left:float=0
    InPadding_Top:float=0
    InPadding_Right:float=-100
    InPadding_Bottom:float=0
   ) -> (then)
   SetPadding(
    execute=[@n2.HitTestInvisible, @n2.SelfHitTestInvisible, @n2.Visible]
    self=SlotAsSizeBoxSlot(Widget=Get SB_Overflow())
    InPadding_Left=0
    InPadding_Top=33
    InPadding_Right=-100
    InPadding_Bottom=-33
   )
   Event OnInitialized()
   > Bind Event to On Visibility Changed(
    self:User Widget Object Reference=Get HitPointBarVehicleHealth!(self:WBP BRGame UI Local Player Hit Point Info Object Reference=Get WBP_BRGameUI_LocalPlayerHitPointInfo() -> (WBP_BRGameUI_LocalPlayerHitPointInfo:WBP BRGame UI Local Player Hit Point Info Object Reference)) -> (HitPointBarVehicleHealth:WBP BRGame UI Player Vital Bar Object Reference)
    Delegate=@n1
   ) {is=title K2Node_AddDelegate}
   > n3 = WaitGameplayTagAddToActor!(
    execute:Exec
    TargetActor:Actor Object Reference=GetPlayerController!(PlayerIndex:int=0) -> (ReturnValue:Player Controller Object Reference)
    Tag:Gameplay Tag Structure=Get RiderTag() -> (RiderTag:Gameplay Tag Structure)
    OnlyTriggerOnce:bool=false
   ) -> (then:Exec, AsyncTaskProxy:Ability Async Wait Gameplay Tag Added Object Reference, Added:Exec) {is=title K2Node_AsyncAction}
   > n4 = WaitGameplayTagRemoveFromActor!(
    execute:Exec
    TargetActor:Actor Object Reference=GetPlayerController(PlayerIndex=0)
    Tag:Gameplay Tag Structure=Get RiderTag()
    OnlyTriggerOnce:bool=false
   ) -> (then:Exec, AsyncTaskProxy:Ability Async Wait Gameplay Tag Removed Object Reference, Removed:Exec) {is=title K2Node_AsyncAction}
   SetRenderTranslation(execute=@n4.Removed, self=Get SB_Overflow(), Translation=(X=0,Y=0))
   SetRenderTranslation(execute=@n3.Added, self=Get SB_Overflow(), Translation=(X=0,Y=46))
  ]}
]
variables [
 In Element Handle {type=ScriptTypedElementHandle, default {}}
 NewVar {type=Actor, default=null}
 RiderTag {type=GameplayTag, default {tagName=Granted.Riding.Player.IsRiding}}
]
widgets {
 info {inheritedWidgetCount=4, parentClass=/Script/FortniteUI.BacchusLocalPlayerInfoBase, rootWidgetClass=&ScaleBox, widgetCount=17}
 widgets [
  ScaleBox_HUDWidgetScale {
   widgetClassPath=&ScaleBox
   bInherited=true
   widget=ScaleBox_HUDWidgetScale
   properties {stretch=UserSpecified, visibility=HitTestInvisible}
  }
  SizeBox_2 {
   widgetClassPath=&SizeBox
   parent=ScaleBox_HUDWidgetScale
   widget=SizeBox_2
   properties {bOverride_WidthOverride=true, widthOverride=510}
   slot {class=ScaleBoxSlot, properties {horizontalAlignment=HAlign_Left, verticalAlignment=VAlign_Top}}
  }
  Overlay_0 {
   widgetClassPath=&Overlay
   parent=SizeBox_2
   widget=Overlay_0
   slot {class=SizeBoxSlot, properties {horizontalAlignment=HAlign_Left, padding=<ltrb 10 0 0 0>}}
  }
  SB_01 {
   widgetClassPath=&SizeBox
   bIsVariable=true
   parent=Overlay_0
   widget=SB_01
   properties {bOverride_HeightOverride=true, bOverride_WidthOverride=true, heightOverride=70, widthOverride=450}
   slot {class=OverlaySlot, properties {horizontalAlignment=HAlign_Fill, padding=<ltrb -30 25 -160 0>}}
  }
  SB_Overflow {
   widgetClassPath=&SizeBox
   bIsVariable=true
   parent=SB_01
   widget=SB_Overflow
   properties {widthOverride=600}
   slot {class=SizeBoxSlot}
  }
  FortPlaylistUIExtensionSlot_BehindMobilePlayerinfo {
   widgetClassPath=/Script/FortniteUI.FortPlaylistUIExtensionSlot
   parent=SB_Overflow
   widget=FortPlaylistUIExtensionSlot_BehindMobilePlayerinfo
   properties {
    entryHorizontalAlignment=HAlign_Left
    entryVerticalAlignment=VAlign_Center
    entryWidgetClass=&CommonUserWidget
    numDesignerPreviewEntries=1
    extensionSlots [BehindMobileLocalPlayerInfo]
   }
   slot {class=SizeBoxSlot}
  }
  SizeBox_0 {
   widgetClassPath=&SizeBox
   parent=Overlay_0
   widget=SizeBox_0
   properties {
    bOverride_HeightOverride=true
    bOverride_WidthOverride=true
    heightOverride=125
    maxDesiredHeight=125
    maxDesiredWidth=312
    minDesiredWidth=312
    widthOverride=312
   }
   slot {class=OverlaySlot, properties {horizontalAlignment=HAlign_Fill, padding=<ltrb 0 5 0 5>}}
  }
  MainBox {widgetClassPath=&VerticalBox, bIsVariable=true, parent=SizeBox_0, widget=MainBox, slot {class=SizeBoxSlot}}
  VerticalBox_112 {
   widgetClassPath=&VerticalBox
   parent=MainBox
   widget=VerticalBox_112
   slot {class=VerticalBoxSlot, properties {verticalAlignment=VAlign_Top, padding=<ltrb -25 0 0 0>}}
  }
  FortPlaylistUIExtensionSlot_65 {
   widgetClassPath=&FortPlaylistUIExtensionSlot
   parent=VerticalBox_112
   widget=FortPlaylistUIExtensionSlot_65
   properties {entryWidgetClass=&CommonUserWidget, extensionSlots [AboveMobileLocalPlayerInfo]}
   slot {class=VerticalBoxSlot}
  }
  AthenaPlayerNameInfo {
   widgetClassPath=/BattleRoyaleGameUI/HUD/SquadList/WBP_BRGameUI_PlayerNameInfo_Local
   bInherited=true
   bIsVariable=true
   parent=MainBox
   widget=AthenaPlayerNameInfo
   slot {class=VerticalBoxSlot}
  }
  WBP_BRGameUI_LocalPlayerHitPointInfo {
   widgetClassPath=/BattleRoyaleGameUI/HUD/PlayerHitPointInfo/WBP_BRGameUI_LocalPlayerHitPointInfo
   bIsVariable=true
   parent=MainBox
   widget=WBP_BRGameUI_LocalPlayerHitPointInfo
   properties {use Large Format=true, renderTransform {angle=0, scale=<1 1>, shear=<0 0>, translation=<0 5>}}
   slot {class=VerticalBoxSlot}
  }
  SB_PluginContainer {
   widgetClassPath=&SizeBox
   parent=MainBox
   widget=SB_PluginContainer
   properties {heightOverride=30, minDesiredHeight=30}
   slot {class=VerticalBoxSlot, properties {verticalAlignment=VAlign_Center, padding=<ltrb -22 0 0 0>}}
  }
  FortPlaylistUIExtensionSlot_60 {
   widgetClassPath=&FortPlaylistUIExtensionSlot
   parent=MainBox
   widget=FortPlaylistUIExtensionSlot_60
   properties {
    entryBoxType=Vertical
    entryHorizontalAlignment=HAlign_Center
    entryVerticalAlignment=VAlign_Center
    entryWidgetClass=&CommonUserWidget
    extensionSlots [UnderLocalPlayerInfo, UnderTeammateStatus, UnderMobileLocalPlayerInfo]
   }
   slot {class=VerticalBoxSlot}
  }
  SizeBox_1 {
   widgetClassPath=&SizeBox
   parent=Overlay_0
   widget=SizeBox_1
   properties {bOverride_HeightOverride=true, heightOverride=66, widthOverride=354.3339}
   slot {class=OverlaySlot, properties {horizontalAlignment=HAlign_Fill, verticalAlignment=VAlign_Fill}}
  }
  TutorialHighlight {
   widgetClassPath=&TutorialHighlight
   bIsVariable=true
   parent=SizeBox_1
   widget=TutorialHighlight
   properties {
    customHighlightWidget=/Game/Athena/HUD/Tutorial/Tutorial_Outline_Square_Static.Tutorial_Outline_Square_Static_C
    renderTransform {angle=0, scale=<1 1>, shear=<-12 0>, translation=<0 0>}
   }
   slot {class=SizeBoxSlot}
  }
  Image_Background {widgetClassPath=&Image, bInherited=true, widget=null}
 ]
}
