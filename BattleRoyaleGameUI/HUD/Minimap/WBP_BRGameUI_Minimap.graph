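# Battle Royale HUD minimap widget with a clickable frame that toggles the fullscreen map, inside a refresh-limiting retainer box.
# A map controller device event overrides rotation mode and swaps the frame's backing color for custom maps.
# Device override or player setting decides rotation, with a looping timer feeding player rotation and cardinal-direction visibility set on the frame's dynamic material.
# Mobile touch play collapses the minimap's widget slotting overlay.
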WidgetBlueprint /BattleRoyaleGameUI/HUD/Minimap/WBP_BRGameUI_Minimap
graphs [EventGraph {nodes [
   "On Clicked (Button_Map)"() {is=title K2Node_ComponentBoundEvent}
   > ToggleFullscreenMap(self:Fort HUDContext Object Reference=Get FortHUDContext!() -> (ReturnValue:Fort HUDContext Object Reference) {is=title K2Node_GetBlueprintContextEx})
   SetupForMapDeviceOverrides() {is=title K2Node_CustomEvent}
   > Set DefaultBackingColor(DefaultBackingColor:Linear Color Structure=Get BrushColor!(self:Border Object Reference=Get Border_MapFrame() -> (Border_MapFrame:Border Object Reference)) -> (BrushColor:Linear Color Structure)) -> (Output_Get:Linear Color Structure)
   > n1 = GetEventRouter!(execute, Scope:Object Class Reference=/Script/FortniteGame.FortPlayspace, ContextActor:Actor Object Reference=GetOwningPlayer()) -> (then, ReturnValue:Gameplay Event Router Component Object Reference)
   > Is Valid(InputObject=@n1)
   > n2 = StartListeningToEvent!(
    execute:Exec
    Target:Gameplay Event Router Component Object Reference=@n1
    EventType:Script Struct Object Reference=/Game/Athena/HUD/MiniMap/MapController/GER_MapControllerDevice_ActivePayload
    Context:Object Reference
    EventBubblingRule:EEventBubblingRule Enum
   ) -> (
    then:Exec
    AsyncTaskProxy:Async Action Start Listening To Event Object Reference
    OnEventReceived:Exec
    Payload:GER Map Controller Device Active Payload Structure
   )
   n3 = Break GER Map Controller Device Active Payload!(GER_MapControllerDevice_ActivePayload:"GER Map Controller Device Active Payload Structure (by ref)"=@n2.Payload) -> (
    CustomMapActive_1_#1:bool=~
    HideMapInfoPanel_4_#2:bool=~
    TransparentMinimap_12_#3:bool=~
    MapShape_10_#4:EMapController_MapShape Enum
    MapRotationMode_16_#5:EMapControllerDevice_RotationMode Enum
   ) {is=title K2Node_BreakStruct}
   Set RotationModeOverride(
    execute=@n2.OnEventReceived
    RotationModeOverride:EMapControllerDevice_RotationMode Enum=@n3.MapRotationMode_16_#5
   ) -> (Output_Get:EMapControllerDevice_RotationMode Enum=NewEnumerator0)
   > n4 = GetDynamicMaterial(self:Border Object Reference=Get Border_MapFrame())
   > SetVectorParameterValue(
    self=@n4
    ParameterName=BackgroundColor
    Value=SelectColor(
     A=Get CustomMapBackingColor() -> (CustomMapBackingColor:Linear Color Structure)
     B=Get DefaultBackingColor() -> (DefaultBackingColor:Linear Color Structure)
     bPickA=@n3
    )
   )
   OnMinimapRotationModeChanged!() -> (OutputDelegate, then, bMinimapRotationModeOn:bool=~) {is=title K2Node_CustomEvent}
   n5 = Get RotationModeOverride() -> (RotationModeOverride:EMapControllerDevice_RotationMode Enum=NewEnumerator0)
   n6 = "Equal (Enum)"(A:EMapControllerDevice_RotationMode Enum=@n5, B:EMapControllerDevice_RotationMode Enum=NewEnumerator2)
   > OverrideMinimapRotationMode!(
    execute
    self:Athena Minimap Object Reference=Get Minimap() -> (Minimap:Athena Minimap Object Reference)
    bOverrideActive:bool=AND Boolean(
     A="Not Equal (Enum)"!(A:EMapControllerDevice_RotationMode Enum=@n5, B:EMapControllerDevice_RotationMode Enum=NewEnumerator0) -> (ReturnValue:bool)
     B=@n6
    )
    bRotationEnabled:bool=@n6
   ) -> (then)
   n7 = Select!(
    NewEnumerator0:bool=IsPlayerAlwaysFaceNorthSettingEnabled!(self:Fort HUDContext Object Reference=Get FortHUDContext()) -> (ReturnValue:bool)
    NewEnumerator1:bool=false
    NewEnumerator2:bool=true
    Index:EMapControllerDevice_RotationMode Enum=@n5
   ) -> (ReturnValue:bool)
   > n8 = Branch(Condition=OR Boolean(A=Get bAlwaysShowCardinalDirections() -> (bAlwaysShowCardinalDirections:bool), B=@n7))
   n9 = Get Border_MapFrame()
   > n10 = GetDynamicMaterial(self=@n9)
   > SetScalarParameterValue(self=@n10, ParameterName=ShowCardinalDirections, Value=1)
   > n11 = Branch(Condition=@n7)
   n12 = OnApplyRotation() {is=title K2Node_CustomEvent}
   n13 = Set Timer by Event(execute=@n11.then, Delegate=@n12, Time=0.01, bLooping=true, bMaxOncePerFrame=true)
   n14 = GetDynamicMaterial(execute=@n8.else, self=@n9)
   SetScalarParameterValue(execute=[@n21.then, @n11.else], self=@n14, ParameterName=MapRotation, Value=0)
   > Clear and Invalidate Timer by Handle(Handle=@n13)
   Event Construct()
   n15 = Create Event()
   n16 = Get FortHUDContext()
   > Bind Event to On Minimap Rotation Mode Changed(self:Fort HUDContext Object Reference=@n16, Delegate=@n15) {is=title K2Node_AddDelegate}
   > SetupForMapDeviceOverrides() {is=title K2Node_CallFunction}
   > OnMinimapRotationModeChanged!(execute, self, bMinimapRotationModeOn:bool=false) -> (then) {is=title K2Node_CallFunction}
   n17 = GetOwningPlayer()
   > Branch(Condition=AND Boolean(A=IsMobileGame(), B=IsUsingTouch(PlayerController=@n17)))
   > SetVisibility(self=Get MinimapSlottingOverlay() -> (MinimapSlottingOverlay:Widget Slotting Host Overlay Object Reference), InVisibility=Collapsed)
   Event Destruct()
   > Unbind Event from On Minimap Rotation Mode Changed(self:Fort HUDContext Object Reference=@n16, Delegate=@n15) {is=title K2Node_RemoveDelegate}
   n18 = GetDynamicMaterial(execute=@n12.then, self=Get Border_MapFrame())
   n19 = GetFortPlayerControllerFromActor!(Actor:Actor Object Reference=GetOwningPlayer()) -> (ReturnValue:Fort Player Controller Object Reference)
   n20 = Get Minimap()
   > SetScalarParameterValue(
    self=@n18
    ParameterName=MapRotation
    Value=float * float(
     A=GetPlayerRotation!(self:Athena Minimap Object Reference=@n20, FortPlayerController:Fort Player Controller Object Reference=@n19) -> (ReturnValue:float)
     B=-1
    ) {is=title K2Node_PromotableOperator}
   )
   n21 = SetScalarParameterValue(execute=@n14.then, self=@n14, ParameterName=ShowCardinalDirections, Value=0)
  ]}]
variables [
 bAlwaysShowCardinalDirections {type=boolean}
 CustomMapBackingColor {type=LinearColor, default=<rgba 0 0 0 0.5>}
 DefaultBackingColor {type=LinearColor, default=<rgba 0 0.1384 0.3763 1>}
 RotationModeOverride {type=EMapControllerDevice_RotationMode, default=Default}
]
widgets {
 info {inheritedWidgetCount=1, parentClass=&FortHUDElementWidget, rootWidgetClass=&Overlay, widgetCount=8}
 widgets [
  Overlay_0 {widgetClassPath=&Overlay, widget=Overlay_0}
  RetainerBox_LimitRefreshRate {
   widgetClassPath=/Script/UMG.RetainerBox
   parent=Overlay_0
   widget=RetainerBox_LimitRefreshRate
   properties {phaseCount=2}
   slot {class=OverlaySlot}
  }
  Border_MapFrame {
   widgetClassPath=&Border
   bIsVariable=true
   parent=RetainerBox_LimitRefreshRate
   widget=Border_MapFrame
   properties {
    visibility=SelfHitTestInvisible
    background {
     bIsDynamicallyLoaded=false
     drawAs=Image
     imageType=FullColor
     mirroring=NoMirror
     resourceName=None
     resourceObject=/BattleRoyaleGameUI/HUD/Minimap/Assets/M_UI_MinimapFrame
     tiling=NoTile
     imageSize=<32 32>
     margin=<ltrb 0 0 0 0>
     outlineSettings {
      bUseBrushTransparency=false
      roundingType=HalfHeightRadius
      width=0
      color {colorUseRule=UseColor_Specified, specifiedColor=<rgba 0 0 0 0>}
      cornerRadii=<0 0 0 0>
     }
     tintColor {colorUseRule=UseColor_Specified, specifiedColor=<rgba 1 1 1 1>}
     uVRegion {bIsValid=false, max=<0 0>, min=<0 0>}
    }
    padding=<ltrb 0 0 0 0>
   }
   slot {class=PanelSlot}
  }
  Overlay_Map {
   widgetClassPath=&Overlay
   bIsVariable=true
   parent=Border_MapFrame
   widget=Overlay_Map
   slot {class=BorderSlot, properties {padding=<ltrb 0 0 0 0>}}
  }
  Button_Map {
   widgetClassPath=&Button
   parent=Overlay_Map
   widget=Button_Map
   properties {
    isFocusable=false
    backgroundColor=<rgba 1 1 1 0>
    colorAndOpacity=<rgba 1 1 1 0>
    widgetStyle {
     clickedSlateSound {resourceObject=null}
     disabled {
      bIsDynamicallyLoaded=false
      drawAs=NoDrawType
      imageType=NoImage
      mirroring=NoMirror
      resourceName=None
      resourceObject=null
      tiling=NoTile
      imageSize=<0 0>
      margin=<ltrb 0 0 0 0>
      outlineSettings {
       bUseBrushTransparency=false
       roundingType=HalfHeightRadius
       width=0
       color {colorUseRule=UseColor_Specified, specifiedColor=<rgba 0 0 0 0>}
       cornerRadii=<0 0 0 0>
      }
      tintColor {colorUseRule=UseColor_Specified, specifiedColor=<rgba 1 1 1 1>}
      uVRegion {bIsValid=false, max=<0 0>, min=<0 0>}
     }
     disabledForeground {colorUseRule=UseColor_Specified, specifiedColor=<rgba 0.5271 0.5271 0.5271 1>}
     hovered {
      bIsDynamicallyLoaded=false
      drawAs=RoundedBox
      imageType=NoImage
      mirroring=NoMirror
      resourceName=None
      resourceObject=null
      tiling=NoTile
      imageSize=<32 32>
      margin=<ltrb 0 0 0 0>
      outlineSettings {
       bUseBrushTransparency=true
       roundingType=FixedRadius
       width=1
       color {colorUseRule=UseColor_Specified, specifiedColor=<rgba 0.7243 0.7243 0.7243 1>}
       cornerRadii=<4 4 4 4>
      }
      tintColor {colorUseRule=UseColor_Specified, specifiedColor=<rgba 1 1 1 0>}
      uVRegion {bIsValid=false, max=<0 0>, min=<0 0>}
     }
     hoveredForeground {colorUseRule=UseColor_Specified, specifiedColor=<rgba 0 1 1 1>}
     hoveredSlateSound {resourceObject=null}
     normal {
      bIsDynamicallyLoaded=false
      drawAs=RoundedBox
      imageType=NoImage
      mirroring=NoMirror
      resourceName=None
      resourceObject=null
      tiling=NoTile
      imageSize=<32 32>
      margin=<ltrb 0 0 0 0>
      outlineSettings {
       bUseBrushTransparency=true
       roundingType=FixedRadius
       width=1
       color {colorUseRule=UseColor_Specified, specifiedColor=<rgba 0.6951 0.6951 0.6951 1>}
       cornerRadii=<4 4 4 4>
      }
      tintColor {colorUseRule=UseColor_Specified, specifiedColor=<rgba 1 1 1 0>}
      uVRegion {bIsValid=false, max=<0 0>, min=<0 0>}
     }
     normalForeground {colorUseRule=UseColor_Specified, specifiedColor=<rgba 0 1 1 1>}
     normalPadding=<ltrb 12 1.5 12 1.5>
     pressed {
      bIsDynamicallyLoaded=false
      drawAs=RoundedBox
      imageType=NoImage
      mirroring=NoMirror
      resourceName=None
      resourceObject=null
      tiling=NoTile
      imageSize=<32 32>
      margin=<ltrb 0 0 0 0>
      outlineSettings {
       bUseBrushTransparency=true
       roundingType=FixedRadius
       width=1
       color {colorUseRule=UseColor_Specified, specifiedColor=<rgba 0.7243 0.7243 0.7243 1>}
       cornerRadii=<4 4 4 4>
      }
      tintColor {colorUseRule=UseColor_Specified, specifiedColor=<rgba 1 1 1 0>}
      uVRegion {bIsValid=false, max=<0 0>, min=<0 0>}
     }
     pressedForeground {colorUseRule=UseColor_Specified, specifiedColor=<rgba 0 1 1 1>}
     pressedPadding=<ltrb 12 1.5 12 1.5>
     pressedSlateSound {resourceObject=null}
    }
   }
   slot {class=OverlaySlot, properties {horizontalAlignment=HAlign_Fill, verticalAlignment=VAlign_Fill}}
  }
  Minimap {
   widgetClassPath=/Script/FortniteUI.AthenaMinimap
   bIsVariable=true
   parent=Overlay_Map
   widget=Minimap
   properties {
    renderOpacity=0.95
    visibility=SelfHitTestInvisible
    desiredSize=<252 252>
    elevationArrowIconBrush {
     bIsDynamicallyLoaded=false
     drawAs=Image
     imageType=FullColor
     mirroring=NoMirror
     resourceName=None
     resourceObject=/Game/UI/CodeReferenced/T_UI_ElevationArrow
     tiling=NoTile
     imageSize=<32 32>
     margin=<ltrb 0 0 0 0>
     outlineSettings {
      bUseBrushTransparency=false
      roundingType=HalfHeightRadius
      width=0
      color {colorUseRule=UseColor_Specified, specifiedColor=<rgba 0 0 0 0>}
      cornerRadii=<0 0 0 0>
     }
     tintColor {colorUseRule=UseColor_Specified, specifiedColor=<rgba 1 1 1 1>}
     uVRegion {bIsValid=false, max=<0 0>, min=<0 0>}
    }
    playerIconBrush {
     bIsDynamicallyLoaded=false
     drawAs=Image
     imageType=FullColor
     mirroring=NoMirror
     resourceName=None
     resourceObject=/Game/UI/CodeReferenced/UI_Radar_AthenaPlayerArrow
     tiling=NoTile
     imageSize=<20 20>
     margin=<ltrb 0 0 0 0>
     outlineSettings {
      bUseBrushTransparency=false
      roundingType=HalfHeightRadius
      width=0
      color {colorUseRule=UseColor_Specified, specifiedColor=<rgba 0 0 0 0>}
      cornerRadii=<0 0 0 0>
     }
     tintColor {colorUseRule=UseColor_Specified, specifiedColor=<rgba 1 1 1 1>}
     uVRegion {bIsValid=false, max=<0 0>, min=<0 0>}
    }
    squadPinIconBrush {
     bIsDynamicallyLoaded=false
     drawAs=Image
     imageType=FullColor
     mirroring=NoMirror
     resourceName=None
     resourceObject=/Game/UI/CodeReferenced/UI_Squad_Pin
     tiling=NoTile
     imageSize=<35 35>
     margin=<ltrb 0 0 0 0>
     outlineSettings {
      bUseBrushTransparency=false
      roundingType=HalfHeightRadius
      width=0
      color {colorUseRule=UseColor_Specified, specifiedColor=<rgba 0 0 0 0>}
      cornerRadii=<0 0 0 0>
     }
     tintColor {colorUseRule=UseColor_Specified, specifiedColor=<rgba 1 1 1 1>}
     uVRegion {bIsValid=false, max=<0 0>, min=<0 0>}
    }
   }
   slot {class=OverlaySlot, properties {padding=<ltrb 24 24 24 24>}}
  }
  TutorialHighlight {
   widgetClassPath=&TutorialHighlight
   bIsVariable=true
   parent=Overlay_Map
   widget=TutorialHighlight
   properties {customHighlightWidget=/Game/Athena/HUD/Tutorial/Tutorial_Outline_Square.Tutorial_Outline_Square_C, visibility=SelfHitTestInvisible}
   slot {class=OverlaySlot, properties {horizontalAlignment=HAlign_Fill, verticalAlignment=VAlign_Fill, padding=<ltrb -6 -6 -6 -6>}}
  }
  MinimapSlottingOverlay {
   widgetClassPath=/Script/WidgetSlotting.WidgetSlottingHostOverlay
   bIsVariable=true
   parent=Overlay_0
   widget=MinimapSlottingOverlay
   properties {hostTag {tagName=Athena.HUD.MinimapOverlaySlot}}
   slot {class=OverlaySlot, properties {verticalAlignment=VAlign_Fill}}
  }
 ]
}
