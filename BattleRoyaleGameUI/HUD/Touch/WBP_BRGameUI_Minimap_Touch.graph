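# Touch HUD minimap cluster for Battle Royale, stacking the minimap, game-phase timer, kill and players-left counters, and playlist extension slots.
# Registers as a tagged widget host, placing content above the minimap or swapping ranked widgets into counter switchers, passing interface implementers the ranked bonuses VM.
# In compact touch UI, counters fade out while the game phase shows.
# Wins collapse the minimap; mobile touch overlays the respawn indicator on it.
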WidgetBlueprint /BattleRoyaleGameUI/HUD/Touch/WBP_BRGameUI_Minimap_Touch
animations [
 AnimHideGameStatus {
  binding_count=1
  length_seconds=0.3334
  object=&self:AnimHideGameStatus
  sequence {
   bindings [GameStatus_WithoutGamePhase {
     id=#1
     possessed_class=HorizontalBox
     tracks [{
       type=MovieSceneFloatTrack
       display_name=Render Opacity
       sections [{type=MovieSceneFloatSection, end_frame=20, start_frame=0, channels [None {type=MovieSceneScriptingFloatChannel, keys [[0,1],[15,0],[20,0]]}]}]
      }]
    }]
   display_rate {numerator=60}
  }
 }
 AnimShowGameStatus {
  binding_count=1
  length_seconds=0.3334
  object=&self:AnimShowGameStatus
  sequence {
   bindings [GameStatus_WithoutGamePhase {
     id=#1
     possessed_class=HorizontalBox
     tracks [{
       type=MovieSceneFloatTrack
       display_name=Render Opacity
       sections [{type=MovieSceneFloatSection, end_frame=20, start_frame=0, channels [None {type=MovieSceneScriptingFloatChannel, keys [[0,0],[5,0],[20,1]]}]}]
      }]
    }]
   display_rate {numerator=60}
  }
 }
]
graphs [
 AddWidgetSlotToCanvas {nodes [
   n1 = AddWidgetSlotToCanvas!() -> (then, Target:Widget Switcher Object Reference, Widget:Widget Object Reference) {is=title K2Node_FunctionEntry}
   > AddChild(self=@n1, Content=@n1.Widget)
   > SetActiveWidget(self:Widget Switcher Object Reference=@n1, Widget=@n1.Widget)
  ]}
 BindBonusVM {nodes [
   n1 = BindBonusVM!() -> (then, Athena Ranked Bonus VM:Athena Ranked Bonuses VM Object Reference) {is=title K2Node_FunctionEntry}
   > PrintString!(execute, InString=VM IS HERE, bPrintToScreen, bPrintToLog, TextColor=(R=1,G=0.7049,B=0,A=1), Duration=10, Key) -> (then)
   > Branch(Condition="Equal (Object)"(A=@n1, B=Get AthenaRankedBonusesVM() -> (AthenaRankedBonusesVM:Athena Ranked Bonuses VM Object Reference)))
  ]}
 EventGraph {nodes [
   n1 = "Updated Game Phase Step (WBP_BRGameUI_GamePhaseWidget)"!() -> (OutputDelegate, then, NewGamePhaseStep:EAthenaGamePhaseStep Enum=None, IsGamePhaseShown:bool) {is=title K2Node_ComponentBoundEvent}
   n2 = GetOwningPlayer()
   > Branch(Condition=AND Boolean(
     A=IsUsingCompactUI()
     B=IsUsingInputType!(InputType:ECommonInputType Enum=Touch, PlayerController:Player Controller Object Reference=@n2) -> (ReturnValue:bool)
    ))
   > PlayAnimation(InAnimation=Select(
     Option 0:Widget Animation Object Reference=Get AnimShowGameStatus() -> (AnimShowGameStatus:Widget Animation Object Reference)
     Option 1:Widget Animation Object Reference=Get AnimHideGameStatus() -> (AnimHideGameStatus:Widget Animation Object Reference)
     Index=@n1.IsGamePhaseShown
    ) -> (ReturnValue:Widget Animation Object Reference))
   n3 = On Local Team Won() {is=title K2Node_CustomEvent}
   SetVisibility(
    execute=[@n3.then, @n4.then]
    self=Get WBP_BRGameUI_Minimap() -> (WBP_BRGameUI_Minimap:WBP BRGame UI Minimap Object Reference)
    InVisibility=Collapsed
   )
   n4 = On Local Player Won() {is=title K2Node_CustomEvent}
   n5 = Event Add Content To Host Widget!() -> (
    OutputDelegate
    then
    WidgetHostTag:"Gameplay Tag Structure (by ref)"
    ContentWidget:Widget Object Reference
    Padding:"Margin Structure (by ref)"
    bFill:bool
   ) {is=title K2Node_Event}
   > n6 = Set SlottedContentWidget(SlottedContentWidget:Widget Object Reference=@n5.ContentWidget) -> (Output_Get:Widget Object Reference)
   > n7 = Sequence()
   > Branch(Condition="Equal (GameplayTag)"(A=@n5.WidgetHostTag, B=Get AboveMobileMinimapSlotTag() -> (AboveMobileMinimapSlotTag:Gameplay Tag Structure)))
   > AddChild(self=Get SizeBox_AboveMobileMinimapSlotHost() -> (SizeBox_AboveMobileMinimapSlotHost:Size Box Object Reference), Content=@n6)
   n8 = Branch(
    execute=@n7.then_1
    Condition=MatchesTag!(
     TagOne:Gameplay Tag Structure=@n5.WidgetHostTag
     TagTwo:Gameplay Tag Structure=Get RankedKillsWidgetSlotTag() -> (RankedKillsWidgetSlotTag:Gameplay Tag Structure)
     bExactMatch:bool=false
    ) -> (ReturnValue:bool)
   )
   > n9 = AddWidgetSlotToCanvas!(
    execute
    self
    Target:Widget Switcher Object Reference=Get WidgetSwitcherKills() -> (WidgetSwitcherKills:Common Activatable Widget Switcher Object Reference)
    Widget:Widget Object Reference=@n6
   ) -> (then) {is=title K2Node_CallFunction}
   Branch(
    execute=[@n9.then, @n10.then]
    Condition=Does Object Implement Interface!(
     TestObject:Object Reference=@n6
     Interface:Interface Interface=/BattleRoyaleGameUI/HUD/Minimap/BPI_InMatchFeedbackVMProvider.BPI_InMatchFeedbackVMProvider_C
    ) -> (ReturnValue:bool)
   )
   > ProvideVM!(execute:Exec, self:Object Reference=@n6, VM:Athena Ranked Bonuses VM Object Reference=Get AthenaRankedBonusesVM()) -> (then:Exec) {is=title K2Node_Message}
   Branch(
    execute=@n8.else
    Condition=MatchesTag(
     TagOne=@n5.WidgetHostTag
     TagTwo=Get RankedPlayersLeftWidgetSlotTag() -> (RankedPlayersLeftWidgetSlotTag:Gameplay Tag Structure)
     bExactMatch=false
    )
   )
   > n10 = AddWidgetSlotToCanvas(Target=Get WidgetSwitcherPlayersLeft() -> (WidgetSwitcherPlayersLeft:Common Activatable Widget Switcher Object Reference), Widget=@n6)
   Event Construct()
   n11 = Get RankedPlayersLeftWidgetSlotTag()
   > RegisterWidgetHost!(execute, WidgetHost:Widget Object Reference, WidgetHostTag:"Gameplay Tag Structure (by ref)"=@n11, IsPlayerSpecific:bool=true) -> (then)
   n12 = Get RankedKillsWidgetSlotTag()
   > RegisterWidgetHost(WidgetHostTag=@n12, IsPlayerSpecific=true)
   > RegisterWidgetHost(WidgetHostTag=Get AboveMobileMinimapSlotTag(), IsPlayerSpecific=true)
   > Cast To FortPlayerControllerAthena(Object=GetOwningPlayer())
   > SetRespawnMobileLayout()
   > n13 = Sequence()
   n14 = Get AthenaHUDContext!() -> (ReturnValue:Athena HUDContext Object Reference) {is=title K2Node_GetBlueprintContextEx}
   > Bind Event to On Local Player Won(self:Athena HUDContext Object Reference=@n14, Delegate=@n4) {is=title K2Node_AddDelegate}
   Bind Event to On Local Team Won(execute=@n13.then_1, self:Athena HUDContext Object Reference=@n14, Delegate=@n3) {is=title K2Node_AddDelegate}
   Event Destruct()
   > UnregisterWidgetHost!(execute, WidgetHost:Widget Object Reference, WidgetHostTag:"Gameplay Tag Structure (by ref)"=@n11) -> (then)
   > UnregisterWidgetHost(WidgetHostTag=@n12)
   > UnregisterWidgetHost(WidgetHostTag=Get AboveMobileMinimapSlotTag())
   > Cast To FortPlayerControllerAthena(Object=GetOwningPlayer())
   > n15 = Sequence()
   > Unbind Event from On Local Player Won(self:Athena HUDContext Object Reference=@n14, Delegate=@n4) {is=title K2Node_RemoveDelegate}
   Unbind Event from On Local Team Won(execute=@n15.then_1, self:Athena HUDContext Object Reference=@n14, Delegate=@n3) {is=title K2Node_RemoveDelegate}
  ]}
 RemoveContentFromWidgetHost {nodes [
   n1 = Remove Content From Host Widget!() -> (
    then
    WidgetHostTag:"Gameplay Tag Structure (by ref)"
    ContentWidget:Widget Object Reference
    RemovalType:EWidgetSlottingRemovalType Enum=Unparent
   ) {is=title K2Node_FunctionEntry}
   > n2 = Sequence()
   > Branch(Condition="Equal (GameplayTag)"(A=@n1, B=Get AboveMobileMinimapSlotTag()))
   > n3 = RemoveChild(self=Get SizeBox_AboveMobileMinimapSlotHost(), Content=@n1.ContentWidget)
   > Return Node!(execute, ReturnValue:bool=@n3)
   n4 = Branch(execute=@n2.then_1, Condition=MatchesTag(TagTwo=Get RankedKillsWidgetSlotTag(), bExactMatch=false))
   n5 = Reroute Node(InputPin:Widget Object Reference) -> (OutputPin:Widget Object Reference)
   n6 = Get WidgetSwitcherKills()
   > n7 = RemoveChild(self=@n6, Content=@n5)
   > n8 = Branch(Condition=@n7)
   > n9 = SetActiveWidgetIndex(self=@n6, Index=0)
   Return Node(execute=[@n9.then, @n8.else], ReturnValue=@n7)
   n10 = Branch(execute=@n4.else, Condition=MatchesTag(TagTwo=Get RankedPlayersLeftWidgetSlotTag(), bExactMatch=false))
   n11 = Get WidgetSwitcherPlayersLeft()
   > n12 = RemoveChild(self=@n11, Content=@n5)
   > n13 = Branch(Condition=@n12)
   > n14 = SetActiveWidgetIndex(self=@n11, Index=0)
   Return Node(execute=[@n14.then, @n13.else], ReturnValue=@n12)
   Return Node(execute=@n10.else, ReturnValue=false)
  ]}
 RemoveWidgetSlotFromCanvas {nodes [
   n1 = RemoveWidgetSlotFromCanvas!() -> (then, Target:Panel Widget Object Reference, Content:Widget Object Reference) {is=title K2Node_FunctionEntry}
   > RemoveChild(self=@n1, Content=@n1.Content)
  ]}
 SetRespawnMobileLayout {nodes [
   SetRespawnMobileLayout() {is=title K2Node_FunctionEntry}
   n1 = GetOwningPlayer()
   > Branch(Condition=AND Boolean(A=IsMobileGame(), B=IsUsingTouch(PlayerController=@n1)))
   n2 = Get AthenaRespawnIndicator_0() -> (AthenaRespawnIndicator_0:Athena Respawn Indicator Object Reference)
   > AddChildToOverlay!(
    execute
    self:Overlay Object Reference=Get Overlay_Minimap() -> (Overlay_Minimap:Overlay Object Reference)
    Content:Widget Object Reference=@n2
   ) -> (then, ReturnValue:Overlay Slot Object Reference)
   n3 = SlotAsOverlaySlot(Widget=@n2)
   > SetVerticalAlignment(self=@n3, InVerticalAlignment=VAlign_Bottom)
   > SetHorizontalAlignment(self=@n3, InHorizontalAlignment=HAlign_Fill)
  ]}
]
variables [
 AboveMobileMinimapSlotTag {type=GameplayTag, default {tagName=DynamicUI.HUD.Fort.Athena.AboveMobileMinimap.Slot}}
 RankedKillsWidgetSlotTag {type=GameplayTag, default {tagName=Hud.Athena.Habanero.RankedKills}}
 RankedPlayersLeftWidgetSlotTag {type=GameplayTag, default {tagName=HUD.Athena.Habanero.RankedPlayersLeft}}
 SlottedContentWidget {type=Widget, default=null}
]
widgets {
 info {
  inheritedWidgetCount=3
  namedSlotCount=1
  parentClass=/Script/FortniteUI.BacchusHUDElement
  rootWidgetClass=&ScaleBox
  widgetCount=33
 }
 widgets [
  ScaleBox_HUDWidgetScale {widgetClassPath=&ScaleBox, bInherited=true, widget=ScaleBox_HUDWidgetScale, properties {stretch=UserSpecified}}
  SizeBox_3 {
   widgetClassPath=&SizeBox
   parent=ScaleBox_HUDWidgetScale
   widget=SizeBox_3
   properties {
    bOverride_HeightOverride=true
    bOverride_WidthOverride=true
    heightOverride=340
    maxDesiredHeight=340
    maxDesiredWidth=300
    widthOverride=300
   }
   slot {class=ScaleBoxSlot, properties {horizontalAlignment=HAlign_Fill, verticalAlignment=VAlign_Top}}
  }
  VBox_TopRightContainer {widgetClassPath=&VerticalBox, parent=SizeBox_3, widget=VBox_TopRightContainer, slot {class=SizeBoxSlot}}
  VBox_MapAndGameInfo {widgetClassPath=&VerticalBox, parent=VBox_TopRightContainer, widget=VBox_MapAndGameInfo, slot {class=VerticalBoxSlot}}
  AthenaRespawnIndicator_0 {
   widgetClassPath=/Game/Athena/HUD/AthenaRespawnIndicator
   bIsVariable=true
   parent=VBox_MapAndGameInfo
   widget=AthenaRespawnIndicator_0
   properties {
    visibility=Collapsed
    elementTag {tagName=HUD.Special.RespawnIndicator}
    hUDElementTag {gameplayTags [
      {tagName=Athena.HideForCapture}
      {tagName=HUD.LTM.HideForTryHard}
      {tagName=HUD.Event.HideForSpecialEvent}
      {tagName=Athena.Creative.HideForLoading}
      {tagName=HUD.Hide.WhileLoading}
      {tagName=Athena.LTM.Mash.Title}
      {tagName=HUD.Athena.HideForPostGameSpectate}
      {tagName=HUD.Athena.HideForPostGame}
      {tagName=HUD.LTM.HideForCrucible}
      {tagName=HUD.Athena.HideforRadial}
     ]}
   }
   slot {class=VerticalBoxSlot}
  }
  Overlay_Minimap {
   widgetClassPath=&Overlay
   bIsVariable=true
   parent=VBox_MapAndGameInfo
   widget=Overlay_Minimap
   slot {class=VerticalBoxSlot, properties {horizontalAlignment=HAlign_Left, verticalAlignment=VAlign_Center}}
  }
  HorizontalBox_1 {widgetClassPath=&HorizontalBox, parent=Overlay_Minimap, widget=HorizontalBox_1, slot {class=OverlaySlot}}
  WBP_BRGameUI_Minimap {
   widgetClassPath=/BattleRoyaleGameUI/HUD/Minimap/WBP_BRGameUI_Minimap
   bIsVariable=true
   parent=HorizontalBox_1
   widget=WBP_BRGameUI_Minimap
   slot {class=HorizontalBoxSlot}
  }
  SizeBox_0 {
   widgetClassPath=&SizeBox
   parent=HorizontalBox_1
   widget=SizeBox_0
   properties {bOverride_WidthOverride=true, widthOverride=33}
   slot {class=HorizontalBoxSlot}
  }
  SizeBox_AboveMobileMinimapSlotHost {
   widgetClassPath=&SizeBox
   bIsVariable=true
   parent=Overlay_Minimap
   widget=SizeBox_AboveMobileMinimapSlotHost
   properties {heightOverride=300, visibility=Collapsed, widthOverride=300}
   slot {class=OverlaySlot, properties {verticalAlignment=VAlign_Bottom}}
  }
  FortHUDElementWrapper_GameStatus {
   widgetClassPath=/Game/Athena/HUD/FortHUDElementWrapper
   parent=VBox_MapAndGameInfo
   widget=FortHUDElementWrapper_GameStatus
   properties {elementTag {tagName=HUD.BelowMinimap.GameStatus}}
   slot {class=VerticalBoxSlot, properties {horizontalAlignment=HAlign_Left, verticalAlignment=VAlign_Top, padding=<ltrb 24 -24 0 0>}}
  }
  SizeBox_2 {
   widgetClassPath=&SizeBox
   namedSlotHost=FortHUDElementWrapper_GameStatus
   widget=SizeBox_2
   properties {
    bOverride_HeightOverride=true
    bOverride_MaxDesiredWidth=true
    bOverride_WidthOverride=true
    heightOverride=55
    maxDesiredWidth=400
    widthOverride=400
   }
  }
  VerticalBox_0 {
   widgetClassPath=&VerticalBox
   parent=SizeBox_2
   widget=VerticalBox_0
   slot {class=SizeBoxSlot, properties {padding=<ltrb 0 10 0 0>}}
  }
  SizeBox_6 {
   widgetClassPath=&SizeBox
   parent=VerticalBox_0
   widget=SizeBox_6
   properties {bOverride_WidthOverride=true, widthOverride=400}
   slot {class=VerticalBoxSlot, properties {size {sizeRule=Fill, value=1}}}
  }
  HorizontalBox_GameStatus {
   widgetClassPath=&HorizontalBox
   parent=SizeBox_6
   widget=HorizontalBox_GameStatus
   properties {renderTransform {angle=0, scale=<0.8 0.8>, shear=<0 0>, translation=<0 1>}, renderTransformPivot=<0 0>}
   slot {class=SizeBoxSlot}
  }
  Overlay_StormTimer {
   widgetClassPath=&Overlay
   parent=HorizontalBox_GameStatus
   widget=Overlay_StormTimer
   slot {class=HorizontalBoxSlot, properties {horizontalAlignment=HAlign_Left}}
  }
  Spacer_ForSurfTimer {
   widgetClassPath=&Spacer
   parent=Overlay_StormTimer
   widget=Spacer_ForSurfTimer
   properties {size=<100 0>}
   slot {class=OverlaySlot, properties {verticalAlignment=VAlign_Fill}}
  }
  WBP_BRGameUI_GamePhaseWidget {
   widgetClassPath=/BattleRoyaleGameUI/HUD/Minimap/WBP_BRGameUI_GamePhaseWidget
   bIsVariable=true
   parent=Overlay_StormTimer
   widget=WBP_BRGameUI_GamePhaseWidget
   properties {
    flowDirectionPreference=RightToLeft
    elementTag {tagName=MobileHUD.Stats.GamePhase}
    hUDElementTag {gameplayTags [
      {tagName=Athena.HideForCapture}
      {tagName=HUD.Athena.Hide.WithInventory}
      {tagName=HUD.Athena.Hide.WithMap}
      {tagName=HUD.LTM.HideForTryHard}
      {tagName=HUD.Event.HideForSpecialEvent}
      {tagName=Athena.Creative.HideForLoading}
      {tagName=HUD.Hide.WhileLoading}
      {tagName=HUD.LTM.HideForCrucible}
      {tagName=Athena.Creative.HUDController.HideStormTimer}
      {tagName=Athena.LTM.Mash.Title}
      {tagName=HUD.Athena.HideForPostGame}
      {tagName=HUD.Athena.HideforRadial}
      {tagName=HUD.Hide.CreativeJoiningLobby}
      {tagName=HUD.Athena.GamePhase.MinimapInfo}
     ]}
    renderTransformPivot=<0.5 0.5>
   }
   slot {class=OverlaySlot}
  }
  GameStatus_WithoutGamePhase {
   widgetClassPath=&HorizontalBox
   parent=HorizontalBox_GameStatus
   widget=GameStatus_WithoutGamePhase
   slot {class=HorizontalBoxSlot, properties {horizontalAlignment=HAlign_Left, verticalAlignment=VAlign_Center, padding=<ltrb 5 0 0 0>}}
  }
  WidgetSwitcherKills {
   widgetClassPath=&CommonActivatableWidgetSwitcher
   bIsVariable=true
   parent=GameStatus_WithoutGamePhase
   widget=WidgetSwitcherKills
   slot {class=HorizontalBoxSlot}
  }
  WBP_BRGameUI_KillsWidget {
   widgetClassPath=/BattleRoyaleGameUI/HUD/Minimap/WBP_BRGameUI_KillsWidget
   bIsVariable=true
   parent=WidgetSwitcherKills
   widget=WBP_BRGameUI_KillsWidget
   properties {
    flowDirectionPreference=RightToLeft
    elementTag {tagName=MobileHUD.Stats.Kills}
    hUDElementTag {gameplayTags [
      {tagName=Athena.HideForCapture}
      {tagName=HUD.Athena.Hide.WithInventory}
      {tagName=HUD.Athena.Hide.WithMap}
      {tagName=HUD.LTM.HideForTryHard}
      {tagName=HUD.Event.HideForSpecialEvent}
      {tagName=Athena.Creative.HideForLoading}
      {tagName=HUD.Hide.WhileLoading}
      {tagName=HUD.LTM.HideForCrucible}
      {tagName=Athena.Creative.HUDController.HideEliminationCounter}
      {tagName=Athena.LTM.Mash.Title}
      {tagName=HUD.Athena.HideForPostGame}
      {tagName=HUD.Athena.HideforRadial}
      {tagName=HUD.Athena.HideOnSpectate}
      {tagName=HUD.Hide.CreativeJoiningLobby}
      {tagName=HUD.Athena.Kills}
     ]}
   }
   slot {class=WidgetSwitcherSlot, properties {horizontalAlignment=HAlign_Left, verticalAlignment=VAlign_Top, padding=<ltrb 0 0 5 0>}}
  }
  WBP_BRGameUI_TeamKillsWidget {
   widgetClassPath=/BattleRoyaleGameUI/HUD/Minimap/WBP_BRGameUI_TeamKillsWidget
   bIsVariable=true
   parent=GameStatus_WithoutGamePhase
   widget=WBP_BRGameUI_TeamKillsWidget
   properties {
    flowDirectionPreference=RightToLeft
    visibility=Collapsed
    elementTag {tagName=MobileHUD.Stats.TeamKills}
    hUDElementTag {gameplayTags [
      {tagName=HUD.Athena.Hide.WithInventory}
      {tagName=HUD.Athena.Hide.WithMap}
      {tagName=HUD.LTM.HideForTryHard}
      {tagName=HUD.Event.HideForSpecialEvent}
      {tagName=HUD.Hide.WhileLoading}
      {tagName=HUD.LTM.HideForCrucible}
      {tagName=HUD.Athena.HideForPostGame}
      {tagName=HUD.Athena.HideforRadial}
      {tagName=HUD.Hide.CreativeJoiningLobby}
      {tagName=Athena.Creative.HideForLoading}
      {tagName=Athena.Creative.HUDController.HideTeamInfo}
      {tagName=Athena.HideForCapture}
      {tagName=Athena.LTM.Mash.Title}
      {tagName=HUD.Athena.TeamKills}
     ]}
   }
   slot {class=HorizontalBoxSlot, properties {horizontalAlignment=HAlign_Left, verticalAlignment=VAlign_Top, padding=<ltrb 0 0 5 0>}}
  }
  WidgetSwitcherPlayersLeft {
   widgetClassPath=&CommonActivatableWidgetSwitcher
   bIsVariable=true
   parent=GameStatus_WithoutGamePhase
   widget=WidgetSwitcherPlayersLeft
   slot {class=HorizontalBoxSlot}
  }
  WBP_BRGameUI_PlayersLeftWidget {
   widgetClassPath=/BattleRoyaleGameUI/HUD/Minimap/WBP_BRGameUI_PlayersLeftWidget
   bIsVariable=true
   parent=WidgetSwitcherPlayersLeft
   widget=WBP_BRGameUI_PlayersLeftWidget
   properties {
    flowDirectionPreference=RightToLeft
    elementTag {tagName=MobileHUD.Stats.PlayersLeft}
    hUDElementTag {gameplayTags [
      {tagName=Athena.HideForCapture}
      {tagName=HUD.Athena.Hide.WithInventory}
      {tagName=HUD.Athena.Hide.WithMap}
      {tagName=HUD.LTM.HideForTryHard}
      {tagName=HUD.Event.HideForSpecialEvent}
      {tagName=Athena.Creative.HideForLoading}
      {tagName=HUD.Hide.WhileLoading}
      {tagName=HUD.LTM.HideForCrucible}
      {tagName=Athena.Creative.HUDController.HidePlayerCount}
      {tagName=HUD.Athena.HideForPostGame}
      {tagName=HUD.Athena.HideforRadial}
      {tagName=HUD.Hide.CreativeJoiningLobby}
      {tagName=Athena.LTM.Mash.Title}
      {tagName=HUD.Athena.PlayersLeft}
     ]}
   }
   slot {class=WidgetSwitcherSlot, properties {horizontalAlignment=HAlign_Left, verticalAlignment=VAlign_Top}}
  }
  GameInfoBox_Bacchus {
   widgetClassPath=/Script/FortniteUI.FortPlaylistUIExtensionSlot
   bIsVariable=true
   parent=HorizontalBox_GameStatus
   widget=GameInfoBox_Bacchus
   properties {
    entryHorizontalAlignment=HAlign_Left
    entryVerticalAlignment=VAlign_Top
    entryWidgetClass=&CommonUserWidget
    numDesignerPreviewEntries=1
    extensionSlots [GameInfoBelowMap, GameInfoBox]
   }
   slot {class=HorizontalBoxSlot, properties {horizontalAlignment=HAlign_Center, verticalAlignment=VAlign_Center}}
  }
  HorizontalBox_2 {
   widgetClassPath=&HorizontalBox
   parent=VerticalBox_0
   widget=HorizontalBox_2
   slot {class=VerticalBoxSlot, properties {padding=<ltrb 0 5 0 5>}}
  }
  SizeBox_4 {
   widgetClassPath=&SizeBox
   parent=HorizontalBox_2
   widget=SizeBox_4
   slot {class=HorizontalBoxSlot, properties {size {sizeRule=Fill, value=1}}}
  }
  ExtensionSlotUnderSquadInfo {
   widgetClassPath=&FortPlaylistUIExtensionSlot
   bIsVariable=true
   parent=SizeBox_4
   widget=ExtensionSlotUnderSquadInfo
   properties {
    entryBoxType=Vertical
    entryWidgetClass=&CommonUserWidget
    maxElementSize=1
    entrySpacing=<0 3>
    extensionSlots [UnderSquadInfo, MobileHUDUnderMinimap]
   }
   slot {class=SizeBoxSlot}
  }
  VerticalBox_71 {
   widgetClassPath=&VerticalBox
   parent=VerticalBox_0
   widget=VerticalBox_71
   properties {visibility=Collapsed}
   slot {class=VerticalBoxSlot}
  }
  SizeBox_107 {widgetClassPath=&SizeBox, parent=VerticalBox_71, widget=SizeBox_107, slot {class=VerticalBoxSlot}}
  UIExtensionSlot_UnderMapMessagesContainer {
   widgetClassPath=&FortPlaylistUIExtensionSlot
   bIsVariable=true
   parent=SizeBox_107
   widget=UIExtensionSlot_UnderMapMessagesContainer
   properties {entryBoxType=Vertical, entryWidgetClass=&CommonUserWidget, extensionSlots [UnderMinimapMessageContainer]}
   slot {class=SizeBoxSlot, properties {horizontalAlignment=HAlign_Left}}
  }
  Image_Background {widgetClassPath=&Image, bInherited=true, widget=null}
  TutorialHighlight {widgetClassPath=&TutorialHighlight}
 ]
}
