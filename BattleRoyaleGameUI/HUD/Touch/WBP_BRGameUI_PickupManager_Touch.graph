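# Mobile HUD element that hosts the Battle Royale pickup manager in the touch layout, derived from FortMobileHUDElement.
# A user-specified ScaleBox under the inherited HUD scale box shrinks the shared WBP_BRGameUI_PickupManager, aligns it top-right and nudges it down.
# The pickup manager carries the Athena.HideForCapture HUD element tag.
# The graph adds no logic of its own, as Construct only calls the parent and the other events are empty.
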WidgetBlueprint /BattleRoyaleGameUI/HUD/Touch/WBP_BRGameUI_PickupManager_Touch
graphs [EventGraph {nodes [
   Event Construct()
   > Parent: Construct()
   Event PreConstruct()
   Event Tick()
  ]}]
variables []
widgets {
 info {inheritedWidgetCount=3, parentClass=/Script/FortniteUI.FortMobileHUDElement, rootWidgetClass=&ScaleBox, widgetCount=5}
 widgets [
  ScaleBox_HUDWidgetScale {
   widgetClassPath=&ScaleBox
   bInherited=true
   widget=ScaleBox_HUDWidgetScale
   properties {ignoreInheritedScale=true, stretch=UserSpecified}
  }
  ScaleBox_PickupManager {
   widgetClassPath=&ScaleBox
   parent=ScaleBox_HUDWidgetScale
   widget=ScaleBox_PickupManager
   properties {stretch=UserSpecified, userSpecifiedScale=0.75}
   slot {class=ScaleBoxSlot, properties {horizontalAlignment=HAlign_Right, verticalAlignment=VAlign_Top}}
  }
  WBP_BRGameUI_PickupManager {
   widgetClassPath=/BattleRoyaleGameUI/HUD/PickupManager/WBP_BRGameUI_PickupManager
   bIsVariable=true
   parent=ScaleBox_PickupManager
   widget=WBP_BRGameUI_PickupManager
   properties {hUDElementTag {gameplayTags [{tagName=Athena.HideForCapture}]}, renderTransform {angle=0, scale=<1 1>, shear=<0 0>, translation=<0 30>}}
   slot {class=ScaleBoxSlot, properties {horizontalAlignment=HAlign_Right, verticalAlignment=VAlign_Top}}
  }
  Image_Background {widgetClassPath=&Image, bInherited=true, widget=null}
  TutorialHighlight {widgetClassPath=&TutorialHighlight}
 ]
}
