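# Battle Royale HUD squad list widget showing local player and teammate info entries.
# In the CVar-gated new layout, the local player entry is collapsed, or moved into a visual attachment shown only while the HideOnBus tag is active.
# UIExtension events for the underneath-list slot async-load widget classes and toggle them, removing an existing instance or adding one.
# Non-local squad disconnects play a sound, with a variant when one or fewer members remain connected.
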WidgetBlueprint /BattleRoyaleGameUI/HUD/SquadList/WBP_BRGameUI_SquadList
graphs [
 AddExtensionWidget {nodes [
   n1 = AddExtensionWidget!() -> (then, ExtensionWidget:User Widget Object Reference) {is=title K2Node_FunctionEntry}
   > Is Valid(InputObject=@n1)
   > n2 = AddChildToVerticalBox(
    self=Get VBox_Extensions() -> (VBox_Extensions:Vertical Box Object Reference)
    Content=Get ExtensionWidget() -> (ExtensionWidget:User Widget Object Reference)
   )
   > SetVerticalAlignment(self:Vertical Box Slot Object Reference=@n2, InVerticalAlignment=VAlign_Fill)
   > SetHorizontalAlignment(self:Vertical Box Slot Object Reference=@n2, InHorizontalAlignment=HAlign_Fill)
   > SetPadding!(
    execute
    self:Vertical Box Slot Object Reference=@n2
    InPadding_Left:float=0
    InPadding_Top:float=0
    InPadding_Right:float=0
    InPadding_Bottom:float=8
   ) -> (then)
  ]}
 AttemptToRemoveExtensionWidget {nodes [
   AttemptToRemoveExtensionWidget!() -> (then, ExtensionWidgetClass:User Widget Class Reference) {is=title K2Node_FunctionEntry}
   > n1 = For Each Loop(Array:Array of Widget Object References=GetAllChildren(self=Get VBox_Extensions())) -> (Array Element:Widget Object Reference)
   Return Node!(execute=@n1.Completed, bExtensionWidgetRemoved:bool=false)
   > Branch(Condition=IsA!(
     Object:Object Reference=@n1
     ObjectClass:Object Class Reference=Get ExtensionWidgetClass() -> (ExtensionWidgetClass:User Widget Class Reference)
    ) -> (ReturnValue:bool))
   > RemoveChildAt!(execute, self:Panel Widget Object Reference=Get VBox_Extensions(), Index:int=@n1.Array Index) -> (then, ReturnValue:bool)
   > Return Node(bExtensionWidgetRemoved=true)
  ]}
 EventGraph {nodes [
   n1 = OnHUDElementVisibilityChanged!() -> (OutputDelegate, then, HiddenHUDElementTags:"Gameplay Tag Container Structure (by ref)") {is=title K2Node_CustomEvent}
   > n2 = Branch(Condition=HasTag!(
     TagContainer:"Gameplay Tag Container Structure (by ref)"=@n1.HiddenHUDElementTags
     Tag:Gameplay Tag Structure=(TagName="HUD.Athena.HideOnBus")
     bExactMatch:bool=true
    ) -> (ReturnValue:bool))
   n3 = Get VisAttachment_LocalPlayer() -> (VisAttachment_LocalPlayer:Fort Visual Attachment Object Reference)
   > SetVisibility(self=@n3, InVisibility=SelfHitTestInvisible)
   SetVisibility(execute=@n2.else, self=@n3, InVisibility=Collapsed)
   HandleNewHUDLayout() {is=title K2Node_CustomEvent}
   > Branch(Condition=IsNewHUDLayout!(self) -> (IsNewHUD:bool))
   > n4 = Branch(Condition=IsShowLocalPlayerInSquadListForBusPhaseEnabled!(self:Athena Squad List Base Object Reference) -> (ReturnValue:bool))
   > n5 = GetFortHUDContextByPlayer!(
    execute
    LocalPlayer:Local Player Object Reference=GetOwningLocalPlayer!(self:Widget Object Reference) -> (ReturnValue:Local Player Object Reference)
   ) -> (then, ReturnValue:Fort HUDContext Object Reference)
   > Bind Event to On HUDElement Visibility Changed(self:Fort HUDContext Object Reference=@n5, Delegate=Create Event()) {is=title K2Node_AddDelegate}
   > AddChild(self=Get VisAttachment_LocalPlayer(), Content=Get Overlay_LocalPlayer() -> (Overlay_LocalPlayer:Overlay Object Reference))
   > SetVisibility(self=Get VisAttachment_LocalPlayer(), InVisibility=Collapsed)
   Set LocalPlayerVisibility(execute=@n4.else, LocalPlayerVisibility:ESlateVisibility Enum=Collapsed) -> (Output_Get:ESlateVisibility Enum=Visible)
   Event Construct()
   > HandleNewHUDLayout() {is=title K2Node_CallFunction}
   n6 = GetGlobalEventRouter()
   > Is Valid(InputObject=@n6)
   > n7 = StartListeningToEvent!(
    execute:Exec
    Target:Gameplay Event Router Component Object Reference=@n6
    EventType:Script Struct Object Reference=/Script/FortniteGame.UIExtension
    Context:Object Reference=GetOwningPlayer()
    EventBubblingRule:EEventBubblingRule Enum
   ) -> (
    then:Exec
    AsyncTaskProxy:Async Action Start Listening To Event Object Reference
    OnEventReceived:Exec
    Payload_Slot:EUIExtensionSlot Enum
    Payload_WidgetClass:User Widget Soft Class Reference
   )
   > Set ListenForUIExtensionAsyncAction(ListenForUIExtensionAsyncAction:Async Action Start Listening To Event Object Reference=@n7) -> (Output_Get:Async Action Start Listening To Event Object Reference)
   Branch(
    execute=@n7.OnEventReceived
    Condition="Equal (Enum)"(A:EUIExtensionSlot Enum=@n7.Payload_Slot, B:EUIExtensionSlot Enum=UnderneathTeammateStatusList)
   )
   > Branch(Condition=IsValidSoftClassReference!(SoftClassReference:"Object Soft Class Reference (by ref)"=@n7.Payload_WidgetClass) -> (ReturnValue:bool))
   > n8 = Async Load Class Asset!(execute:Exec, AssetClass:Object Soft Class Reference=@n7.Payload_WidgetClass) -> (then:Exec, Completed:Exec, Class:Object Class Reference) {is=title K2Node_LoadAssetClass}
   n9 = Cast To UserWidget Class!(execute:Exec=@n8.Completed, Class:Object Class Reference=@n8) -> (then:Exec, CastFailed:Exec, AsUser Widget:User Widget Class Reference) {is=title K2Node_ClassDynamicCast}
   > n10 = AttemptToRemoveExtensionWidget!(execute, self, ExtensionWidgetClass:User Widget Class Reference=@n9) -> (then, bExtensionWidgetRemoved:bool) {is=title K2Node_CallFunction}
   > n11 = Branch(Condition=@n10)
   n12 = Create Widget!(execute:Exec=@n11.else, Class:User Widget Class Reference=@n9, OwningPlayer:Player Controller Object Reference=GetOwningPlayer()) -> (then:Exec, ReturnValue:User Widget Object Reference)
   > AddExtensionWidget!(execute, self, ExtensionWidget:User Widget Object Reference=@n12) -> (then) {is=title K2Node_CallFunction}
   Event Destruct()
   > n13 = GetFortHUDContextByPlayer(LocalPlayer=GetOwningLocalPlayer())
   > Unbind Event from On HUDElement Visibility Changed(self:Fort HUDContext Object Reference=@n13, Delegate=Create Event()) {is=title K2Node_RemoveDelegate}
   > n14 = Get!(execute:Exec) -> (then:Exec, else:Exec, ListenForUIExtensionAsyncAction:Async Action Start Listening To Event Object Reference)
   > StopListeningToEvent(self:Async Action Start Listening To Event Object Reference=@n14)
   n15 = Event HandleSquadMemberDisconnected!() -> (OutputDelegate, then, Player:Fort Player State Athena Object Reference) {is=title K2Node_Event}
   > n16 = Cast To PlayerController!(execute, Object=GetOwner(self:Actor Object Reference=@n15.Player)) -> (then, CastFailed, AsPlayer Controller:Player Controller Object Reference) {is=title K2Node_DynamicCast}
   > n17 = Branch(Condition=IsLocalController!(self:Controller Object Reference=@n16) -> (ReturnValue:bool))
   n18 = GetSquadConnectedCount!(self:Athena HUDContext Object Reference=Get AthenaHUDContext!() -> (ReturnValue:Athena HUDContext Object Reference) {is=title K2Node_GetBlueprintContextEx}) -> (ReturnValue:int=0)
   PlaySound2D(
    execute=[@n16.CastFailed, @n17.else]
    Sound=Select(
     Option 0:Sound Base Object Reference=Get Sound_SquadMemberDisconnected() -> (Sound_SquadMemberDisconnected:Sound Base Object Reference)
     Option 1:Sound Base Object Reference=Get Sound_LastSquadMemberDisconnected() -> (Sound_LastSquadMemberDisconnected:Sound Base Object Reference)
     Index="integer <= integer"!(A:int=@n18, B:int=1) -> (ReturnValue:bool) {is=title K2Node_CallFunction}
    ) -> (ReturnValue:Sound Base Object Reference)
   )
  ]}
 IsNewHUDLayout {nodes [
   IsNewHUDLayout() {is=title K2Node_FunctionEntry}
   > n1 = GetConsoleVariableBoolValue!(execute, VariableName:String=BattleRoyalGameUI.SquadListPositionAtBottom.Enabled) -> (then, ReturnValue:bool)
   > n2 = Branch(Condition=@n1)
   > n3 = Branch(Condition=IsCreativeOrPlaygroundModePlaylist!() -> (ReturnValue:bool))
   > Return Node!(
    execute
    IsNewHUD:bool=GetValueAsBoolWithContext!(
     Input:"Scalable Float Structure (by ref)"=Get Class Defaults!(Class:Object Class Reference=/FortniteCoreUI/Settings/ExperienceSettingsComponent.ExperienceSettingsComponent_C) -> (CreativeGameUINewHUD:Scalable Float Structure) {is=title K2Node_GetClassDefaults}
     Level:float=0
    ) -> (ReturnValue:bool)
   )
   n4 = GetConsoleVariableBoolValue(execute=@n3.else, VariableName=BattleRoyalGameUI.NewHUD.Enabled)
   > Return Node(IsNewHUD=@n4)
   Return Node(execute=@n2.else, IsNewHUD=false)
  ]}
]
variables [
 ListenForUIExtensionAsyncAction {type=AsyncAction_StartListeningToEvent, default=null}
 Sound_LastSquadMemberDisconnected {type=SoundBase, default=/Game/Sounds/Fort_UI_Sounds/Gameplay/Fort_SquadList_LastPlayerLeft_Cue}
 Sound_SquadMemberDisconnected {type=SoundBase, default=/Game/Sounds/Fort_UI_Sounds/Gameplay/Fort_SquadList_PlayerLeft_Cue}
]
widgets {
 info {inheritedWidgetCount=3, parentClass=/Script/FortniteUI.AthenaSquadListBase, rootWidgetClass=&Overlay, widgetCount=8}
 widgets [
  Overlay_0 {widgetClassPath=&Overlay, widget=Overlay_0}
  VertBox {widgetClassPath=&VerticalBox, bIsVariable=true, parent=Overlay_0, widget=VertBox, slot {class=OverlaySlot}}
  Overlay_LocalPlayer {
   widgetClassPath=&Overlay
   bIsVariable=true
   parent=VertBox
   widget=Overlay_LocalPlayer
   slot {class=VerticalBoxSlot, properties {padding=<ltrb 0 0 0 8>}}
  }
  LocalPlayerInfo {
   widgetClassPath=&/WBP_BRGameUI_TeamMemberInfo
   bInherited=true
   bIsVariable=true
   parent=Overlay_LocalPlayer
   widget=LocalPlayerInfo
   properties {bIsLocalPlayer=true, flowDirectionPreference=Inherit, elementTag {tagName=HUD.Player.TeamEntry}}
   slot {class=OverlaySlot}
  }
  EntryBox_SquadList {
   widgetClassPath=/Script/FortniteUI.FortDynamicEntryBox
   bInherited=true
   bIsVariable=true
   parent=VertBox
   widget=EntryBox_SquadList
   properties {
    entryBoxType=Vertical
    entryWidgetClass=&/WBP_BRGameUI_TeamMemberInfo.WBP_BRGameUI_TeamMemberInfo_C
    entrySpacing=<0 8>
   }
   slot {class=VerticalBoxSlot, properties {horizontalAlignment=HAlign_Left, padding=<ltrb 0 0 0 8>}}
  }
  VisAttachment_LocalPlayer {
   widgetClassPath=&FortVisualAttachment
   bIsVariable=true
   parent=VertBox
   widget=VisAttachment_LocalPlayer
   properties {visibility=Hidden}
   slot {class=VerticalBoxSlot}
  }
  VBox_Extensions {
   widgetClassPath=&VerticalBox
   bIsVariable=true
   parent=VertBox
   widget=VBox_Extensions
   slot {class=VerticalBoxSlot, properties {padding=<ltrb 0 0 0 4>}}
  }
  TutorialHighlight {widgetClassPath=&TutorialHighlight}
 ]
}
