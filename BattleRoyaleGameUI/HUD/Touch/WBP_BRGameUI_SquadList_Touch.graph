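# Touch HUD element for Battle Royale, built on BacchusHUDElement, that wraps the shared squad list widget with the local player's entry collapsed.
# A fixed-size box holds a vertical stack with the slightly downscaled squad list above a playlist UI extension slot for content underneath teammate status.
# On PreConstruct, SetMobileStyles nudges the squad list left when the owning player is using touch.
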WidgetBlueprint /BattleRoyaleGameUI/HUD/Touch/WBP_BRGameUI_SquadList_Touch
graphs [
 EventGraph {nodes [
   Event PreConstruct()
   > SetMobileStyles()
  ]}
 SetMobileStyles {nodes [
   SetMobileStyles() {is=title K2Node_FunctionEntry}
   > Branch(Condition=IsUsingTouch(PlayerController=GetOwningPlayer()))
   > SetRenderTranslation(self=Get WBP_BRGameUI_SquadList() -> (WBP_BRGameUI_SquadList:WBP BRGame UI Squad List Object Reference), Translation=(X=-15,Y=0))
  ]}
]
variables []
widgets {
 info {inheritedWidgetCount=3, parentClass=/Script/FortniteUI.BacchusHUDElement, rootWidgetClass=&ScaleBox, widgetCount=8}
 widgets [
  ScaleBox_HUDWidgetScale {
   widgetClassPath=&ScaleBox
   bInherited=true
   bIsVariable=true
   widget=ScaleBox_HUDWidgetScale
   properties {stretch=UserSpecified, visibility=HitTestInvisible}
  }
  SizeBox_0 {
   widgetClassPath=&SizeBox
   parent=ScaleBox_HUDWidgetScale
   widget=SizeBox_0
   properties {
    bOverride_HeightOverride=true
    bOverride_WidthOverride=true
    heightOverride=170
    maxDesiredHeight=170
    maxDesiredWidth=336
    minDesiredHeight=195
    widthOverride=336
   }
   slot {class=ScaleBoxSlot, properties {horizontalAlignment=HAlign_Fill, verticalAlignment=VAlign_Top}}
  }
  VerticalBox_0 {
   widgetClassPath=&VerticalBox
   parent=SizeBox_0
   widget=VerticalBox_0
   properties {renderTransform {angle=0, scale=<1 1>, shear=<0 0>, translation=<-16 -1>}}
   slot {class=SizeBoxSlot, properties {horizontalAlignment=HAlign_Left, verticalAlignment=VAlign_Top}}
  }
  ScaleBox_1 {
   widgetClassPath=&ScaleBox
   parent=VerticalBox_0
   widget=ScaleBox_1
   properties {stretch=UserSpecified, stretchDirection=DownOnly, userSpecifiedScale=0.9}
   slot {class=VerticalBoxSlot, properties {horizontalAlignment=HAlign_Left, verticalAlignment=VAlign_Top}}
  }
  WBP_BRGameUI_SquadList {
   widgetClassPath=/BattleRoyaleGameUI/HUD/SquadList/WBP_BRGameUI_SquadList
   bIsVariable=true
   parent=ScaleBox_1
   widget=WBP_BRGameUI_SquadList
   properties {localPlayerVisibility=Collapsed}
   slot {class=ScaleBoxSlot, properties {horizontalAlignment=HAlign_Left}}
  }
  ExtensionSlot_UnderneathTeammateStatus {
   widgetClassPath=/Script/FortniteUI.FortPlaylistUIExtensionSlot
   bIsVariable=true
   parent=VerticalBox_0
   widget=ExtensionSlot_UnderneathTeammateStatus
   properties {
    entryWidgetClass=/Game/Athena/HUD/FortHUDElementWrapper.FortHUDElementWrapper_C
    numDesignerPreviewEntries=1
    extensionSlots [UnderneathTeammateStatusList]
   }
   slot {class=VerticalBoxSlot, properties {horizontalAlignment=HAlign_Left, verticalAlignment=VAlign_Bottom}}
  }
  Image_Background {widgetClassPath=&Image, bInherited=true, widget=null}
  TutorialHighlight {widgetClassPath=&TutorialHighlight}
 ]
}
